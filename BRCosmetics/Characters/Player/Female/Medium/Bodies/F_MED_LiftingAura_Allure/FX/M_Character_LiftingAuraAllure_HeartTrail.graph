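# Translucent, unlit, two-sided material for the scrolling heart trail effect on the LiftingAura Allure character.
# WPO combines SplineThicken, with vertex-color-driven widths, and V-masked noise wobble and VelX drag from particle dynamic parameters.
# Noise-distorted heart UVs, tiled and time-offset by vertex color, drive a two-pink gradient plus a camera-space normal matcap highlight, tinted by particle color.
# A smoothstepped heart mask minus a V gradient erodes hearts toward one end.
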Material /BRCosmetics/Characters/Player/Female/Medium/Bodies/F_MED_LiftingAura_Allure/FX/M_Character_LiftingAuraAllure_HeartTrail
graph {
 properties {
  bUsedWithInstancedStaticMeshes=true
  bUsedWithMeshDeformer=true
  bUsedWithNanite=true
  bUsedWithNiagaraMeshParticles=true
  bUsedWithSkeletalMesh=true
  blendMode=BLEND_Translucent
  pixelDepthOffsetMode=PDOM_Legacy
  shadingModel=MSM_Unlit
  twoSided=true
  usedShadingModels=Unlit
 }
 nodes [
  n1 = DynamicParameter() {paramNames [RandomWobble,VelX,VelY,VelZ]}
  n2 = MaterialFunctionCall(R=@n1.VelX, G=@n1.VelX, B=@n1.VelX) {materialFunction=&AppendMany}
  n3 = OneMinus(None=ComponentMask(None=TextureCoordinate()) {g=true})
  n4 = VertexColor()
  n5 = Multiply(A=Constant() {r=6}, B=@n4.B)
  n6 = Multiply(A=Constant() {r=10}, B=@n4.G)
  n7 = Multiply(A=TextureCoordinate()) {constB=0.1}
  n8 = Multiply(A=Add(A=Time(), B=@n4.A), B=Constant2Vector() {g=0.15, r=0.1})
  n9 = Multiply(A=TextureSample(UVs=Add(A=@n7, B=@n8)) {samplerType=SAMPLERTYPE_Normal, texture=/Game/Effects/Fort_Effects/Emotes/BirdNestBlue/T_LowResSimplePerlin}.RGB) {constB=2}
  n10 = Multiply(A=Multiply(A=Subtract(A=@n9)) {constB=2}, B=@n3)
  n11 = DynamicParameter() {paramNames [RandomWobble,VelX,VelY,VelZ]}
  Add(
   A=Multiply(A=Multiply(A=@n2.RGB) {constB=-0.02}, B=@n3)
   B=Add(
    A=MaterialFunctionCall(WidthBase=@n5, WidthTip=@n6) {materialFunction=&SplineThicken}.WorldPositionOffset
    B=Multiply(A=@n10, B=@n11.RandomWobble)
   )
  )
  n12 = TextureCoordinate()
  n13 = VertexColor()
  n14 = AppendVector(A=Constant() {r=1}, B=@n13.R)
  n15 = Add(A=Time(), B=Multiply(A=@n13.A) {constB=3})
  n16 = Constant2Vector() {g=0.2}
  n17 = Panner(Coordinate=Multiply(A=TextureCoordinate()), Time=@n15) {speedX=0.3, speedY=0.2}
  n18 = Add(
   A=Add(A=Multiply(A=@n12, B=@n14), B=Multiply(A=@n15, B=@n16))
   B=Multiply(A=TextureSample(UVs=@n17) {samplerType=SAMPLERTYPE_Grayscale, texture=/FortUILibrary/UIKit/Buttons/Textures/LowResBlurredNoise}.R) {constB=0.05}
  )
  n19 = TextureSample(UVs=@n18) {
   samplerType=SAMPLERTYPE_Normal
   texture=&/T_Character_LiftingAura_Heart_Normal
  }
  n20 = ComponentMask(None=Transform(None=Transform(None=@n19.RGB)) {transformSourceType=TRANSFORMSOURCE_World, transformType=TRANSFORM_Camera}) {g=true, r=true}
  n21 = Constant2Vector() {g=-0.5, r=0.5}
  n22 = TextureSample(UVs=Add(A=Multiply(A=@n20, B=@n21)) {constB=0.5}) {
   samplerType=SAMPLERTYPE_LinearGrayscale
   texture=/BRCosmetics/Athena/Effects/Particles/Skydiving/NS_Contrail_FNCS_S26/T_Contrail_FNCS_S26_Matcap
  }
  n23 = TextureSample(UVs=@n18) {texture=&/T_Character_LiftingAura_Heart}
  n24 = LinearInterpolate(
   A=VectorParameter() {defaultValue=<rgba 1 0.0437 0.5271 1>}.RGB
   B=VectorParameter() {parameterName=Param2, defaultValue=<rgba 0.5104 0 0.1764 1>}.RGB
   Alpha=Power(Base=@n23.R) {constExponent=1}
  )
  n25 = Multiply(A=Multiply(A=@n22.R, B=Multiply(A=@n24) {constB=3})) {constB=0.3}
  MaterialFunctionCall(Emissive=Multiply(A=Add(A=@n25, B=@n24), B=ParticleColor().RGB)) {materialFunction=&MF_FX_EmissiveAdjustments}
  n26 = OneMinus(None=ComponentMask(None=@n12) {g=true})
  n27 = Saturate(None=Multiply(A=Subtract(A=@n26) {constB=0.7}) {constB=3.5})
  SmoothStep(Value=Saturate(None=Subtract(A=@n23.R, B=@n27))) {constMax=0.12, constMin=0.1}
 ]
}
