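# Niagara cosmetic effect for the LiftingAura Allure character body that attaches persistent heart mesh particles to the skeletal mesh.
# A single Hearts emitter burst-spawns particles for two mesh renderers, with a rotation tweak for the front end.
# Each update follows the sampled mesh location and tints color by time of day.
# Dynamic material parameters receive a camera-facing lerp, whose rotated vector is zero, and owner X/Y velocity, zeroed while skydiving.
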NiagaraSystem /BRCosmetics/Characters/Player/Female/Medium/Bodies/F_MED_LiftingAura_Allure/FX/NS_Character_LiftingAura_Heart
dependencies {
 usedDataInterfaces [&NiagaraDataInterfaceCamera]
 usedDynamicInputs [&Boolean_NotOperation]
 usedModules [
  &SystemState
  &self:SystemSpawnScript.SystemScriptSource.SystemScriptGraph.NiagaraNodeAssignment_0.SetVariables_#1
 ]
}
emitters {Hearts {
  emitterUpdateScript [
   &EmitterState(Life Cycle Mode=::NewEnumerator1, Loop Behavior=::NewEnumerator1, Loop Duration=1, UseLoopDelay=false)
   &SpawnBurst_Instantaneous(Use Spawn Probability=false, UseLoopCountLimit=false)
  ]
  particleSpawnScript [
   &InitializeParticle(Lifetime=5, UsePositionOffset=false, Write Material Random=true)
   &InitialMeshOrientation(
    Mesh Orientation Mode=::NewEnumerator2
    Rotation=&FN_FrontEndVector3Modifier(Gameplay=<0 0 0>, Front End=<-0.03 0.025 0.02>)
    Use Rotation Vector=true
   )
  ]
  particleUpdateScript [
   &ParticleState(Kill Particles When Lifetime Has Elapsed=false)
   &SkeletalMeshLocation(Bone Sampling Mode=::NewEnumerator4, MaskSpawnGroup=false, Skeletal Mesh=$User.SkeletalMesh)
   &ScaleColor(
    Scale Mode=::NewEnumerator1
    Scale RGBA=&MakeVector4_FromLinearColor(LinearColor=&FN_Select_Color_from_Time_Of_Day(Morning=<rgba 0.6979 0.229 0.2925 1>, Day=<rgba 0.9955 0.5336 1 1>, Evening=<rgba 0.9955 0.5336 1 1>, Night=<rgba 0.2168 0.0924 0.474 1>))
    ScaleRGBA=true
   )
   /SetVariables(Particles.HeartNormalizedDirection:Vector3f=&NormalizeVector(
     Vector To Normalize=$Particles.SkeletalMeshLocation.SampledVelocity
     Safe Normalize=true
     Safe Fallback Threshold=0
     Safe Fallback Vector=<0 0 0.001>
    )) {isSetParameters=true}
   &DynamicMaterialParameters(
    Index 0 Param 1=&Lerp_Float(
     A=1
     B=1.5
     Alpha=&DotProduct(
      Vector A=&ReturnCameraVector(Camera Vectors=::NewEnumerator1, Coordinate Space=::World)
      Vector B=&RotateVectorByQuaternion(Vector To Rotate=<0 0 0>, Quaternion=$Particles.SkeletalMeshLocation.SampledRotation)
     )
    )
    Index 0 Param 2=HlslExpression() {hlslExpression=User.IsSkydiving ? 0 : Engine.Owner.Velocity.x}
    Index 0 Param 3=HlslExpression() {hlslExpression=User.IsSkydiving ? 0 : Engine.Owner.Velocity.y}
    Module.Index0Param0Write=true
    Module.Index0Param1Write=true
    Module.Index0Param2Write=true
    Module.Index0Param3Write=true
   )
  ]
  rendererClasses [&NiagaraMeshRendererProperties]
  renderers [
   {rendererClass=&NiagaraMeshRendererProperties, rendererIndex=0}
   {rendererClass=&NiagaraMeshRendererProperties, rendererIndex=1}
  ]
 }}
summary {
 systemName=NS_Character_LiftingAura_Heart
 emitters [{emitterName=Hearts, rendererClasses [&NiagaraMeshRendererProperties]}]
 userVariables [
  User.bIsFrontEnd {type=NiagaraBool}
  User.bIsFrontEndPreview {type=NiagaraBool}
  User.IsSkydiving {type=NiagaraBool}
  User.SkeletalMesh {
   type=NiagaraDataInterfaceSkeletalMesh
   defaultValue {
    struct=NiagaraExt_VariableValue_DataInterface
    value {dataInterface=&self:NiagaraDataInterfaceSkeletalMesh_0}
   }
  }
 ]
}
