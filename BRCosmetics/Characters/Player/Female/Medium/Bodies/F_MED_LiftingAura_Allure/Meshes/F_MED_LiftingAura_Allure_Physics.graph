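# Physics asset for the F_MED_LiftingAura_Allure player cosmetic body, covering its dynamic skirt and left sleeve bones.
# Capsule bodies sit on pelvis, legs, left arm and cloth bones, plus an unconstrained floorCOL body whose huge offset capsule approximates ground.
# Branching skirt chains and short sleeve chains have limited swing and tight twist.
# Their master bodies attach to pelvis, thighs or left upper arm with free rotation.
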PhysicsAsset /BRCosmetics/Characters/Player/Female/Medium/Bodies/F_MED_LiftingAura_Allure/Meshes/F_MED_LiftingAura_Allure_Physics
bodies {
 calf_l [{shapeType=Capsule, length=43.5206, radius=15.2936, shapeName=calf_r_capsule, center=<-25.4537 2.2109 0>, rotation=<rot 90 0 0>}]
 calf_r [{shapeType=Capsule, length=47.2367, radius=11.5775, shapeName=calf_r_capsule, center=<25.1671 2.2109 0>, rotation=<rot 90 0 0>}]
 dyn_skirtBack_01 [{shapeType=Capsule, length=2.585, radius=8, shapeName=dyn_skirtBack_01_capsule, center=<6.2268 0 5>, rotation=<rot 90 0 0>}]
 dyn_skirtBack_01_l [{shapeType=Capsule, length=7.6652, radius=6.3376, shapeName=dyn_skirtBack_01_l_capsule, center=<9.485 0 6>, rotation=<rot 90 0 0>}]
 dyn_skirtBack_01_r [{shapeType=Capsule, length=1.684, radius=12.8955, shapeName=dyn_skirtBack_01_r_capsule, center=<5.8464 0 9>, rotation=<rot 90 0 0>}]
 dyn_skirtBack_02 [{shapeType=Capsule, length=0, radius=15.1419, shapeName=dyn_skirtBack_02_capsule, center=<6.3853 0 9>, rotation=<rot 90 0 0>}]
 dyn_skirtBack_02_l [{shapeType=Capsule, length=6.5049, radius=6.5386, shapeName=dyn_skirtBack_02_l_capsule, center=<8.7927 0 6>, rotation=<rot 90 0 0>}]
 dyn_skirtBack_02_r [
  {shapeType=Capsule, length=0, radius=12.8955, shapeName=dyn_skirtBack_02_r_capsule, center=<6.5317 0 9>, rotation=<rot 90 0 0>}
  {shapeType=Capsule, length=16.5986, radius=13.4014, shapeName=dyn_skirtBack_02_r_capsule, center=<8.52 16.6925 5.5339>, rotation=<rot 90 0 0>}
 ]
 dyn_skirtBack_03 [{shapeType=Capsule, length=5.9997, radius=13.6397, shapeName=dyn_skirtBack_03_capsule, center=<13.3744 0 5>, rotation=<rot 90 0 0>}]
 dyn_skirtBack_03_a [{shapeType=Capsule, length=2.3585, radius=7.6415, shapeName=dyn_skirtBack_03_a_capsule, center=<6.4436 0 5>, rotation=<rot 90 0 0>}]
 dyn_skirtBack_03_l [
  {shapeType=Capsule, length=2.9824, radius=11.7171, shapeName=dyn_skirtBack_03_l_capsule, center=<9.411 0 5>, rotation=<rot 90 0 0>}
  {shapeType=Capsule, length=7.138, radius=7.5615, shapeName=dyn_skirtBack_03_l_capsule, center=<20.5902 -12.1222 5>, rotation=<rot 90 0 0>}
 ]
 dyn_skirtBack_03_r [
  {shapeType=Capsule, length=6.5655, radius=9.1315, shapeName=dyn_skirtBack_03_r_capsule, center=<10.3147 0 9>, rotation=<rot 90 0 0>}
  {shapeType=Capsule, length=16.5986, radius=13.4014, shapeName=dyn_skirtBack_03_r_capsule, center=<8.52 16.6925 5.5339>, rotation=<rot 90 0 0>}
 ]
 dyn_skirtBack_04 [{shapeType=Capsule, length=4.545, radius=14.1087, shapeName=dyn_skirtBack_04_capsule, center=<11.0439 0 5>, rotation=<rot 90 0 0>}]
 dyn_skirtBack_04_a [{shapeType=Capsule, length=2.3585, radius=7.6415, shapeName=dyn_skirtBack_04_a_capsule, center=<6.4436 0 5>, rotation=<rot 90 0 0>}]
 dyn_skirtBack_04_a_02 [{shapeType=Capsule, length=2.3585, radius=7.6415, shapeName=dyn_skirtBack_04_a_02_capsule, center=<6.4436 0 5>, rotation=<rot 90 0 0>}]
 dyn_skirtBack_04_l [{shapeType=Capsule, length=15, radius=8, shapeName=dyn_skirtBack_04_l_capsule, center=<15.0268 0 5>, rotation=<rot 90 0 0>}]
 dyn_skirtBack_04_r [
  {shapeType=Capsule, length=5.4208, radius=9.1315, shapeName=dyn_skirtBack_04_r_capsule, center=<7.8461 0 9>, rotation=<rot 90 0 0>}
  {
   shapeType=Capsule
   length=10.4793
   radius=11.2969
   shapeName=dyn_skirtBack_04_r_capsule
   center=<9.8176 11.9397 2.9678>
   rotation=<rot 90 0 0>
  }
 ]
 dyn_skirtBack_05_l [{shapeType=Capsule, length=15, radius=8, shapeName=dyn_skirtBack_05_l_capsule, center=<15.0268 0 5>, rotation=<rot 90 0 0>}]
 dyn_skirtBack_05_r [{shapeType=Capsule, length=6.2638, radius=9.1315, shapeName=dyn_skirtBack_05_r_capsule, center=<9.3496 0 9>, rotation=<rot 90 0 0>}]
 dyn_skirtBack_06_l [{shapeType=Capsule, length=15, radius=8, shapeName=dyn_skirtBack_06_l_capsule, center=<15.0268 0 5>, rotation=<rot 90 0 0>}]
 dyn_skirtBack_06_r [{shapeType=Capsule, length=13.8685, radius=9.1315, shapeName=dyn_skirtBack_06_r_capsule, center=<15.0268 0 9>, rotation=<rot 90 0 0>}]
 dyn_skirtBack_l_a_01 [{shapeType=Capsule, length=15, radius=8, shapeName=dyn_skirtBack_l_a_01_capsule, center=<15.0268 0 5>, rotation=<rot 90 0 0>}]
 dyn_skirtBack_l_a_02 [{shapeType=Capsule, length=0.3315, radius=8, shapeName=dyn_skirtBack_l_a_02_capsule, center=<7.5833 0 5>, rotation=<rot 90 0 0>}]
 dyn_skirtBack_l_b_01 [{shapeType=Capsule, length=15, radius=8, shapeName=dyn_skirtBack_l_b_01_capsule, center=<15.0268 0 5>, rotation=<rot 90 0 0>}]
 dyn_skirtBack_l_b_02 [{shapeType=Capsule, length=15, radius=8, shapeName=dyn_skirtBack_l_b_02_capsule, center=<15.0268 0 5>, rotation=<rot 90 0 0>}]
 dyn_skirtBack_l_b_03 [{shapeType=Capsule, length=15, radius=8, shapeName=dyn_skirtBack_l_b_03_capsule, center=<15.0268 0 5>, rotation=<rot 90 0 0>}]
 dyn_skirtBack_l_c_01 [{shapeType=Capsule, length=5, radius=5, shapeName=dyn_skirtBack_l_c_01_capsule, center=<6.6413 0 0>, rotation=<rot 90 0 0>}]
 dyn_skirtBack_l_c_02 [{shapeType=Capsule, length=5, radius=5, shapeName=dyn_skirtBack_l_c_02_capsule, center=<6.6413 0 0>, rotation=<rot 90 0 0>}]
 dyn_skirtBack_l_c_03 [{shapeType=Capsule, length=5, radius=5, shapeName=dyn_skirtBack_l_c_03_capsule, center=<6.6413 0 0>, rotation=<rot 90 0 0>}]
 dyn_skirtBack_l_d_01 [{shapeType=Capsule, length=0.505, radius=0.505, shapeName=dyn_skirtBack_l_d_01_capsule, rotation=<rot 0 -90 0>}]
 dyn_skirtBack_l_flyoff_01 [{shapeType=Capsule, length=1.5989, radius=9.9413, shapeName=dyn_skirtBack_l_flyoff_01_capsule, center=<6.9765 0 10>, rotation=<rot 90 0 0>}]
 dyn_skirtBack_l_flyoff_02 [{shapeType=Capsule, length=1.5989, radius=9.9413, shapeName=dyn_skirtBack_l_flyoff_02_capsule, center=<6.9765 0 10>, rotation=<rot 90 0 0>}]
 dyn_skirtBack_l_flyoff_03 [{shapeType=Capsule, length=1.5989, radius=9.9413, shapeName=dyn_skirtBack_l_flyoff_03_capsule, center=<6.9765 0 10>, rotation=<rot 90 0 0>}]
 dyn_skirtBack_l_flyoff_04 [{shapeType=Capsule, length=1.5989, radius=9.9413, shapeName=dyn_skirtBack_l_flyoff_04_capsule, center=<6.9765 0 10>, rotation=<rot 90 0 0>}]
 dyn_skirtBack_l_flyoff_MSTR [{shapeType=Capsule, length=0.505, radius=0.505, shapeName=dyn_skirtBack_l_flyoff_MSTR_capsule, rotation=<rot 0 -90 0>}]
 dyn_skirtBack_mstr [{shapeType=Capsule, length=0.505, radius=0.505, shapeName=dyn_skirtBack_mstr_capsule, rotation=<rot 0 -90 0>}]
 dyn_skirtBack_mstr_l [{shapeType=Capsule, length=0.505, radius=0.505, shapeName=dyn_skirtBack_mstr_l_capsule, rotation=<rot 0 -90 0>}]
 dyn_skirtBack_mstr_r [{shapeType=Capsule, length=0.505, radius=0.505, shapeName=dyn_skirtBack_mstr_r_capsule, rotation=<rot 0 -90 0>}]
 dyn_skirtSideFront_l_01 [{shapeType=Capsule, length=4.4289, radius=6.603, shapeName=dyn_skirtSideFront_l_01_capsule, center=<7.1171 0 8>, rotation=<rot 90 0 0>}]
 dyn_skirtSideFront_l_02 [{shapeType=Capsule, length=0.839, radius=7.4212, shapeName=dyn_skirtSideFront_l_02_capsule, center=<4.4898 0 8>, rotation=<rot 90 0 0>}]
 dyn_skirtSideFront_l_03 [{shapeType=Capsule, length=5.345, radius=6.7806, shapeName=dyn_skirtSideFront_l_03_capsule, center=<8.3351 0 8>, rotation=<rot 90 0 0>}]
 dyn_skirtSideFront_l_MSTR [{shapeType=Capsule, length=0.505, radius=0.505, shapeName=dyn_skirtSideFront_l_MSTR_capsule, rotation=<rot 0 -90 0>}]
 dyn_skirtSideFront_r_01 [{shapeType=Capsule, length=16.1146, radius=6.8854, shapeName=dyn_skirtSideFront_r_01_capsule, center=<15.0268 0 5>, rotation=<rot 90 0 0>}]
 dyn_skirtSideFront_r_02 [{
   shapeType=Capsule
   length=16.1146
   radius=6.8854
   shapeName=dyn_skirtSideFront_r_02_capsule
   center=<15.0268 0 5>
   rotation=<rot 90 174.8764 174.7472>
  }]
 dyn_skirtSideFront_r_03 [{shapeType=Capsule, length=16.1146, radius=6.8854, shapeName=dyn_skirtSideFront_r_02_capsule, center=<15.0268 0 5>, rotation=<rot 90 0 0>}]
 dyn_skirtSideFront_r_MSTR [{shapeType=Capsule, length=0.505, radius=0.505, shapeName=dyn_skirtSideFront_r_MSTR_capsule, rotation=<rot 0 -90 0>}]
 dyn_skirtSide_01_l [{shapeType=Capsule, length=7.3012, radius=6.7017, shapeName=dyn_skirtSide_01_l_capsule, center=<9.485 0 5>, rotation=<rot 90 0 0>}]
 dyn_skirtSide_01_r [{shapeType=Capsule, length=9.2733, radius=12.8263, shapeName=dyn_skirtSide_01_r_capsule, center=<7.4609 0 12>, rotation=<rot 90 0 0>}]
 dyn_skirtSide_02_l [{shapeType=Capsule, length=3.6797, radius=8, shapeName=dyn_skirtSide_02_l_capsule, center=<7.7177 0 5>, rotation=<rot 90 0 0>}]
 dyn_skirtSide_02_r [
  {
   shapeType=Capsule
   length=9.2733
   radius=12.8263
   shapeName=dyn_skirtSide_02_r_capsule
   center=<7.6938 0 12>
   rotation=<rot 90 41.0354 40.9062>
  }
  {shapeType=Capsule, length=15, radius=15, shapeName=dyn_skirtSide_02_r_capsule, center=<0 -14.138 4.2866>, rotation=<rot 90 0 0>}
 ]
 dyn_skirtSide_03_l [{shapeType=Capsule, length=3.6901, radius=8, shapeName=dyn_skirtSide_03_l_capsule, center=<7.3002 0 5>, rotation=<rot 90 0 0>}]
 dyn_skirtSide_03_r [
  {shapeType=Capsule, length=9.2733, radius=12.8263, shapeName=dyn_skirtSide_03_r_capsule, center=<8.9883 0 12>, rotation=<rot 90 0 0>}
  {shapeType=Capsule, length=15, radius=15, shapeName=dyn_skirtSide_03_r_capsule, center=<0 -14.138 4.2866>, rotation=<rot 90 0 0>}
 ]
 dyn_skirtSide_04_l [{shapeType=Capsule, length=5.1485, radius=8, shapeName=dyn_skirtSide_04_l_capsule, center=<8.7899 0 5>, rotation=<rot 90 0 0>}]
 dyn_skirtSide_04_r [{shapeType=Capsule, length=9.2733, radius=12.8263, shapeName=dyn_skirtSide_04_r_capsule, center=<7.848 0 12>, rotation=<rot 90 0 0>}]
 dyn_skirtSide_05_r [{shapeType=Capsule, length=9.2733, radius=12.8263, shapeName=dyn_skirtSide_05_r_capsule, center=<10.54 0 12>, rotation=<rot 90 0 0>}]
 dyn_skirtSide_mstr_l [{shapeType=Capsule, length=0.505, radius=0.505, shapeName=dyn_skirtSide_mstr_l_capsule, rotation=<rot 0 -90 0>}]
 dyn_skirtSide_mstr_r [{shapeType=Capsule, length=0.505, radius=0.505, shapeName=dyn_skirtSide_mstr_r_capsule, rotation=<rot 0 -90 0>}]
 dyn_sleeve_l_a_01 [{shapeType=Capsule, length=4.5815, radius=6.4597, center=<6.8352 0 5>, rotation=<rot 90 0 0>}]
 dyn_sleeve_l_a_MSTR [{shapeType=Capsule, length=0.505, radius=0.505, shapeName=dyn_sleeve_l_a_MSTR_capsule}]
 dyn_sleeve_l_b_01 [{shapeType=Capsule, length=4.5815, radius=6.4597, center=<6.8352 0 5>, rotation=<rot 90 0 0>}]
 dyn_sleeve_l_b_MSTR [{shapeType=Capsule, length=0.505, radius=0.505, shapeName=dyn_sleeve_l_b_MSTR_capsule}]
 floorCOL [
  {shapeType=Sphere, radius=33.7982, shapeName=floorCOL_sphere, center=<0 0.5113 -6.6497>}
  {shapeType=Capsule, length=0.505, radius=500, shapeName=floorCOL_sphere, center=<0 0 -495>, rotation=<rot 0 -90 0>}
 ]
 lowerarm_l [{shapeType=Capsule, length=16.1047, radius=3.5717, shapeName=lowerarm_l_capsule, center=<12.1492 0 0>, rotation=<rot 90 0 0>}]
 pelvis [{shapeType=Capsule, length=8.3951, radius=7.8597, shapeName=pelvis_capsule}]
 thigh_l [{shapeType=Capsule, length=31.6358, radius=11.345, shapeName=thigh_r_capsule, center=<-20.3127 0 0>, rotation=<rot 90 0 0>}]
 thigh_r [{shapeType=Capsule, length=40.2453, radius=6.7497, shapeName=thigh_r_capsule, center=<24.175 0 0>, rotation=<rot 90 0 0>}]
 upperarm_l [{shapeType=Capsule, length=16.6662, radius=2.6144, shapeName=upperarm_l_capsule, center=<16.961 0 0>, rotation=<rot 90 0 0>}]
}
constraints [
 {bone1Name=dyn_skirtBack_01, bone2Name=dyn_skirtBack_mstr, swing1LimitDegrees=10, swing2LimitDegrees=45, twistLimitDegrees=5}
 {bone1Name=dyn_skirtBack_02, bone2Name=dyn_skirtBack_01, swing1LimitDegrees=10, swing2LimitDegrees=45, twistLimitDegrees=5}
 {bone1Name=dyn_skirtBack_03, bone2Name=dyn_skirtBack_02, swing1LimitDegrees=10, swing2LimitDegrees=45, twistLimitDegrees=5}
 {bone1Name=dyn_skirtBack_04, bone2Name=dyn_skirtBack_03, swing1LimitDegrees=10, swing2LimitDegrees=45, twistLimitDegrees=5}
 {bone1Name=dyn_skirtBack_01_l, bone2Name=dyn_skirtBack_mstr_l, swing1LimitDegrees=2, swing2LimitDegrees=20, twistLimitDegrees=5}
 {bone1Name=dyn_skirtBack_02_l, bone2Name=dyn_skirtBack_01_l, swing1LimitDegrees=2, swing2LimitDegrees=20, twistLimitDegrees=5}
 {bone1Name=dyn_skirtBack_03_l, bone2Name=dyn_skirtBack_02_l, swing1LimitDegrees=2, swing2LimitDegrees=20, twistLimitDegrees=5}
 {bone1Name=dyn_skirtBack_04_l, bone2Name=dyn_skirtBack_03_l, swing1LimitDegrees=2, swing2LimitDegrees=20, twistLimitDegrees=5}
 {bone1Name=dyn_skirtBack_05_l, bone2Name=dyn_skirtBack_04_l, swing1LimitDegrees=2, swing2LimitDegrees=20, twistLimitDegrees=5}
 {bone1Name=dyn_skirtBack_06_l, bone2Name=dyn_skirtBack_05_l, swing1LimitDegrees=2, swing2LimitDegrees=20, twistLimitDegrees=5}
 {bone1Name=dyn_skirtBack_l_a_01, bone2Name=dyn_skirtBack_04_l, swing1LimitDegrees=2, swing2LimitDegrees=20, twistLimitDegrees=5}
 {bone1Name=dyn_skirtBack_l_a_02, bone2Name=dyn_skirtBack_l_a_01, swing1LimitDegrees=2, swing2LimitDegrees=20, twistLimitDegrees=5}
 {bone1Name=dyn_skirtBack_l_b_01, bone2Name=dyn_skirtBack_04_l, swing1LimitDegrees=2, swing2LimitDegrees=20, twistLimitDegrees=5}
 {bone1Name=dyn_skirtBack_l_b_02, bone2Name=dyn_skirtBack_l_b_01, swing1LimitDegrees=2, swing2LimitDegrees=20, twistLimitDegrees=5}
 {bone1Name=dyn_skirtBack_l_b_03, bone2Name=dyn_skirtBack_l_b_02, swing1LimitDegrees=2, swing2LimitDegrees=20, twistLimitDegrees=5}
 {bone1Name=dyn_skirtBack_01_r, bone2Name=dyn_skirtBack_mstr_r, swing1LimitDegrees=10, swing2LimitDegrees=45, twistLimitDegrees=5}
 {bone1Name=dyn_skirtBack_02_r, bone2Name=dyn_skirtBack_01_r, swing1LimitDegrees=10, swing2LimitDegrees=45, twistLimitDegrees=5}
 {bone1Name=dyn_skirtBack_03_r, bone2Name=dyn_skirtBack_02_r, swing1LimitDegrees=10, swing2LimitDegrees=45, twistLimitDegrees=5}
 {bone1Name=dyn_skirtBack_04_r, bone2Name=dyn_skirtBack_03_r, swing1LimitDegrees=10, swing2LimitDegrees=45, twistLimitDegrees=5}
 {bone1Name=dyn_skirtBack_05_r, bone2Name=dyn_skirtBack_04_r, swing1LimitDegrees=10, swing2LimitDegrees=45, twistLimitDegrees=5}
 {bone1Name=dyn_skirtBack_06_r, bone2Name=dyn_skirtBack_05_r, swing1LimitDegrees=10, swing2LimitDegrees=45, twistLimitDegrees=5}
 {bone1Name=dyn_skirtSide_01_l, bone2Name=dyn_skirtSide_mstr_l, swing1LimitDegrees=10, swing2LimitDegrees=45, twistLimitDegrees=5}
 {bone1Name=dyn_skirtSide_02_l, bone2Name=dyn_skirtSide_01_l, swing1LimitDegrees=10, swing2LimitDegrees=45, twistLimitDegrees=5}
 {bone1Name=dyn_skirtSide_03_l, bone2Name=dyn_skirtSide_02_l, swing1LimitDegrees=10, swing2LimitDegrees=45, twistLimitDegrees=5}
 {bone1Name=dyn_skirtSide_04_l, bone2Name=dyn_skirtSide_03_l, swing1LimitDegrees=10, swing2LimitDegrees=45, twistLimitDegrees=5}
 {bone1Name=dyn_skirtSide_01_r, bone2Name=dyn_skirtSide_mstr_r, swing1LimitDegrees=10, swing2LimitDegrees=45, twistLimitDegrees=5}
 {bone1Name=dyn_skirtSide_02_r, bone2Name=dyn_skirtSide_01_r, swing1LimitDegrees=10, swing2LimitDegrees=45, twistLimitDegrees=5}
 {bone1Name=dyn_skirtSide_03_r, bone2Name=dyn_skirtSide_02_r, swing1LimitDegrees=10, swing2LimitDegrees=45, twistLimitDegrees=5}
 {bone1Name=dyn_skirtSide_04_r, bone2Name=dyn_skirtSide_03_r, swing1LimitDegrees=10, swing2LimitDegrees=45, twistLimitDegrees=5}
 {bone1Name=dyn_skirtSide_05_r, bone2Name=dyn_skirtSide_04_r, swing1LimitDegrees=10, swing2LimitDegrees=45, twistLimitDegrees=5}
 {bone1Name=dyn_skirtSideFront_l_01, bone2Name=dyn_skirtSideFront_l_MSTR, swing1LimitDegrees=10, swing2LimitDegrees=45, twistLimitDegrees=5}
 {bone1Name=dyn_skirtSideFront_l_02, bone2Name=dyn_skirtSideFront_l_01, swing1LimitDegrees=10, swing2LimitDegrees=45, twistLimitDegrees=5}
 {bone1Name=dyn_skirtSideFront_l_03, bone2Name=dyn_skirtSideFront_l_02, swing1LimitDegrees=10, swing2LimitDegrees=45, twistLimitDegrees=5}
 {bone1Name=dyn_skirtSideFront_r_01, bone2Name=dyn_skirtSideFront_r_MSTR, swing1LimitDegrees=45, swing2LimitDegrees=45, twistLimitDegrees=5}
 {bone1Name=dyn_skirtSideFront_r_02, bone2Name=dyn_skirtSideFront_r_01, swing1LimitDegrees=45, swing2LimitDegrees=45, twistLimitDegrees=5}
 {
  bone1Name=dyn_skirtSideFront_l_MSTR
  bone2Name=thigh_l
  swing1LimitDegrees=45
  swing1Motion=Free
  swing2LimitDegrees=45
  swing2Motion=Free
  twistLimitDegrees=45
  twistMotion=Free
 }
 {
  bone1Name=dyn_skirtSideFront_r_MSTR
  bone2Name=thigh_r
  swing1LimitDegrees=45
  swing1Motion=Free
  swing2LimitDegrees=45
  swing2Motion=Free
  twistLimitDegrees=45
  twistMotion=Free
 }
 {
  bone1Name=thigh_l
  bone2Name=pelvis
  swing1LimitDegrees=45
  swing1Motion=Free
  swing2LimitDegrees=45
  swing2Motion=Free
  twistLimitDegrees=45
  twistMotion=Free
 }
 {
  bone1Name=thigh_r
  bone2Name=pelvis
  swing1LimitDegrees=45
  swing1Motion=Free
  swing2LimitDegrees=45
  swing2Motion=Free
  twistLimitDegrees=45
  twistMotion=Free
 }
 {
  bone1Name=dyn_skirtBack_mstr_l
  bone2Name=pelvis
  swing1LimitDegrees=45
  swing1Motion=Free
  swing2LimitDegrees=45
  swing2Motion=Free
  twistLimitDegrees=45
  twistMotion=Free
 }
 {
  bone1Name=dyn_skirtSide_mstr_r
  bone2Name=pelvis
  swing1LimitDegrees=45
  swing1Motion=Free
  swing2LimitDegrees=45
  swing2Motion=Free
  twistLimitDegrees=45
  twistMotion=Free
 }
 {
  bone1Name=dyn_skirtBack_mstr_r
  bone2Name=pelvis
  swing1LimitDegrees=45
  swing1Motion=Free
  swing2LimitDegrees=45
  swing2Motion=Free
  twistLimitDegrees=45
  twistMotion=Free
 }
 {
  bone1Name=dyn_skirtBack_mstr
  bone2Name=pelvis
  swing1LimitDegrees=45
  swing1Motion=Free
  swing2LimitDegrees=45
  swing2Motion=Free
  twistLimitDegrees=45
  twistMotion=Free
 }
 {
  bone1Name=dyn_skirtSide_mstr_l
  bone2Name=pelvis
  swing1LimitDegrees=45
  swing1Motion=Free
  swing2LimitDegrees=45
  swing2Motion=Free
  twistLimitDegrees=45
  twistMotion=Free
 }
 {
  bone1Name=calf_l
  bone2Name=thigh_l
  swing1LimitDegrees=45
  swing1Motion=Free
  swing2LimitDegrees=45
  swing2Motion=Free
  twistLimitDegrees=45
  twistMotion=Free
 }
 {
  bone1Name=calf_r
  bone2Name=thigh_r
  swing1LimitDegrees=45
  swing1Motion=Free
  swing2LimitDegrees=45
  swing2Motion=Free
  twistLimitDegrees=45
  twistMotion=Free
 }
 {bone1Name=dyn_skirtSideFront_r_03, bone2Name=dyn_skirtSideFront_r_02, swing1LimitDegrees=45, swing2LimitDegrees=45, twistLimitDegrees=5}
 {
  bone1Name=dyn_skirtBack_l_flyoff_01
  bone2Name=dyn_skirtBack_l_flyoff_MSTR
  swing1LimitDegrees=45
  swing2LimitDegrees=45
  twistLimitDegrees=5
 }
 {bone1Name=dyn_skirtBack_l_flyoff_02, bone2Name=dyn_skirtBack_l_flyoff_01, swing1LimitDegrees=45, swing2LimitDegrees=45, twistLimitDegrees=5}
 {bone1Name=dyn_skirtBack_l_flyoff_03, bone2Name=dyn_skirtBack_l_flyoff_02, swing1LimitDegrees=45, swing2LimitDegrees=45, twistLimitDegrees=5}
 {bone1Name=dyn_skirtBack_l_flyoff_04, bone2Name=dyn_skirtBack_l_flyoff_03, swing1LimitDegrees=45, swing2LimitDegrees=45, twistLimitDegrees=5}
 {bone1Name=dyn_skirtBack_l_d_01, bone2Name=dyn_skirtBack_04_l, swing1LimitDegrees=2, swing2LimitDegrees=20, twistLimitDegrees=5}
 {bone1Name=dyn_sleeve_l_b_01, bone2Name=dyn_sleeve_l_b_MSTR, swing1LimitDegrees=14, swing2LimitDegrees=14, twistLimitDegrees=5}
 {bone1Name=dyn_sleeve_l_a_01, bone2Name=dyn_sleeve_l_a_MSTR, swing1LimitDegrees=14, swing2LimitDegrees=14, twistLimitDegrees=5}
 {bone1Name=dyn_skirtBack_03_a, bone2Name=dyn_skirtBack_04_a_02, swing1LimitDegrees=2, swing2LimitDegrees=20, twistLimitDegrees=5}
 {bone1Name=dyn_skirtBack_04_a_02, bone2Name=dyn_skirtBack_04_a, swing1LimitDegrees=2, swing2LimitDegrees=20, twistLimitDegrees=5}
 {bone1Name=dyn_skirtBack_04_a, bone2Name=dyn_skirtBack_03_l, swing1LimitDegrees=2, swing2LimitDegrees=20, twistLimitDegrees=5}
 {
  bone1Name=dyn_sleeve_l_a_MSTR
  bone2Name=upperarm_l
  swing1LimitDegrees=45
  swing1Motion=Free
  swing2LimitDegrees=45
  swing2Motion=Free
  twistLimitDegrees=45
  twistMotion=Free
 }
 {
  bone1Name=dyn_sleeve_l_b_MSTR
  bone2Name=upperarm_l
  swing1LimitDegrees=45
  swing1Motion=Free
  swing2LimitDegrees=45
  swing2Motion=Free
  twistLimitDegrees=45
  twistMotion=Free
 }
 {bone1Name=dyn_skirtBack_l_c_01, bone2Name=dyn_skirtBack_03_l, swing1LimitDegrees=2, swing2LimitDegrees=20, twistLimitDegrees=5}
 {bone1Name=dyn_skirtBack_l_c_02, bone2Name=dyn_skirtBack_l_c_01, swing1LimitDegrees=2, swing2LimitDegrees=20, twistLimitDegrees=5}
 {bone1Name=dyn_skirtBack_l_c_03, bone2Name=dyn_skirtBack_l_c_02, swing1LimitDegrees=2, swing2LimitDegrees=20, twistLimitDegrees=5}
 {
  bone1Name=dyn_skirtBack_l_flyoff_MSTR
  bone2Name=thigh_l
  swing1LimitDegrees=45
  swing1Motion=Free
  swing2LimitDegrees=45
  swing2Motion=Free
  twistLimitDegrees=45
  twistMotion=Free
 }
 {
  bone1Name=upperarm_l
  bone2Name=pelvis
  swing1LimitDegrees=45
  swing1Motion=Free
  swing2LimitDegrees=45
  swing2Motion=Free
  twistLimitDegrees=45
  twistMotion=Free
 }
 {
  bone1Name=lowerarm_l
  bone2Name=upperarm_l
  swing1LimitDegrees=45
  swing1Motion=Free
  swing2LimitDegrees=45
  swing2Motion=Free
  twistLimitDegrees=45
  twistMotion=Free
 }
]
