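# Post-copy AnimBlueprint for the F_MED_LiftingAura_Allure cosmetic body, animating skirt, sleeve and chest feather bones over the parent mesh's copied pose.
# A Control Rig takes rifle and pistol aiming flags, alongside constraints and additive bone offsets that follow limbs and torso.
# The FullSim path adds RigidBody physics with a toggleable floor-collider offset, plus AnimDynamics on skirt flyoff and feathers.
# bRigidBodyEnabled selects between cached FullSim and RBANDisabled poses.
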AnimBlueprint /BRCosmetics/Characters/Player/Female/Medium/Bodies/F_MED_LiftingAura_Allure/Meshes/F_MED_LiftingAura_Allure_AnimBP
target_skeleton=/Game/Characters/Player/Female/Medium/Bodies/F_MED_LiftingAura_Allure/Meshes/F_MED_LiftingAura_Allure_Skeleton
graphs [
 AnimGraph {nodes [
   n1 = "Constraint - dyn_chestFeather_MSTR (RR)"!(
    ComponentPose=Local To Component(LocalPose=Copy Pose From Mesh() {bCopyCurves=true, bUseAttachedParent=true})
    Alpha=1
    ConstraintWeights_0=1
    ConstraintWeights_1:float=0
   ) -> (Pose) {
    alphaBoolBlend {blendOption=Linear}
    boneToModify {boneName=dyn_chestFeather_MSTR}
    constraintSetup [{transformType=Rotation, targetBone {boneName=clavicle_l}}, {transformType=Rotation, targetBone {boneName=upperarm_l}}]
    constraintWeights [1,1]
   }
   n2 = Control Rig!(
    Source=Component To Local(ComponentPose=@n1)
    Alpha
    Rifle:double=Get isAimingRifle() -> (isAimingRifle:double)
    Pistol:double=Get isAimingPistol() -> (isAimingPistol:double)
   ) -> (Pose) {
    alphaBoolBlend {blendOption=Linear}
    controlRigAssetReference {blueprintRigClass=&/F_MED_LiftingAura_Allure_CtrlRig.F_MED_LiftingAura_Allure_CtrlRig_C}
   }
   Blend Poses by bool(BlendPose_0=@n2, bActiveValue=Get JumpingFalling() -> (JumpingFalling:bool)) {blendType=Linear}
   n3 = "Constraint - dyn_skirtBack_03 (RR)"!(ComponentPose, Alpha=1, ConstraintWeights_0=0.5, ConstraintWeights_1:float=0.5) -> (Pose) {
    alphaBoolBlend {blendOption=Linear}
    boneToModify {boneName=dyn_skirtBack_03}
    constraintSetup [{transformType=Rotation, targetBone {boneName=calf_l}}, {transformType=Rotation, targetBone {boneName=calf_r}}]
    constraintWeights [1,1]
   }
   "Constraint - dyn_skirtBack_04 (RR)"!(ComponentPose=@n3, Alpha=1, ConstraintWeights_0=0.5, ConstraintWeights_1:float=0.5) -> (Pose) {
    alphaBoolBlend {blendOption=Linear}
    boneToModify {boneName=dyn_skirtBack_04}
    constraintSetup [{transformType=Rotation, targetBone {boneName=calf_l}}, {transformType=Rotation, targetBone {boneName=calf_r}}]
    constraintWeights [1,1]
   }
   "Constraint - dyn_skirtBack_l_c_MSTR (R)"(Alpha=1, ConstraintWeights_0=1) {
    alphaBoolBlend {blendOption=Linear}
    boneToModify {boneName=dyn_skirtBack_l_c_MSTR}
    constraintSetup [{transformType=Rotation, targetBone {boneName=calf_r}}]
    constraintWeights [1]
   }
   "Constraint - dyn_skirtBack_01 (RR)"!(ComponentPose, Alpha=1, ConstraintWeights_0=1, ConstraintWeights_1:float=0) -> (Pose) {
    alphaBoolBlend {blendOption=Linear}
    boneToModify {boneName=dyn_skirtBack_01}
    constraintSetup [{transformType=Rotation, targetBone {boneName=thigh_l}}, {transformType=Rotation, targetBone {boneName=thigh_r}}]
    constraintWeights [1,1]
   }
   n4 = "Transform (Modify) Bone - Bone: dyn_skirtSide_mstr_r"!(ComponentPose, bAlphaBoolEnabled:bool=True, Translation="34.017,18.287,3.216", Rotation="15.735,0.279,6.003", Scale="0.126,1.000,1.000") -> (Pose) {
    alphaInputType=Bool
    rotationMode=BMM_Additive
    rotationSpace=BCS_BoneSpace
    scaleMode=BMM_Additive
    scaleSpace=BCS_BoneSpace
    translationMode=BMM_Additive
    translationSpace=BCS_BoneSpace
    alphaBoolBlend {blendInTime=1, blendOption=CircularOut, blendOutTime=1}
    alphaScaleBiasClamp {bInterpResult=true}
    boneToModify {boneName=dyn_skirtSide_mstr_r}
    rotation=<rot 15.735 0.279 6.003>
    scale {x=0.126}
    translation=<34.017 18.287 3.216>
   }
   n5 = "Transform (Modify) Bone - Bone: dyn_skirtSideFront_r_MSTR"!(ComponentPose=@n4, bAlphaBoolEnabled:bool=True, Translation="2.104,0.071,1.910", Rotation="6.572,-32.215,-1.664", Scale="1,1,1") -> (Pose) {
    alphaInputType=Bool
    rotationMode=BMM_Additive
    rotationSpace=BCS_BoneSpace
    scaleMode=BMM_Additive
    scaleSpace=BCS_BoneSpace
    translationMode=BMM_Additive
    translationSpace=BCS_BoneSpace
    alphaBoolBlend {blendInTime=1, blendOption=CircularOut, blendOutTime=1}
    alphaScaleBiasClamp {bInterpResult=true}
    boneToModify {boneName=dyn_skirtSideFront_r_MSTR}
    rotation=<rot 6.572 -32.215 -1.664>
    translation=<2.104 0.071 1.91>
   }
   n6 = "Transform (Modify) Bone - Bone: dyn_skirtSide_03_r"!(ComponentPose=@n5, bAlphaBoolEnabled:bool=True, Translation="0,0,0", Rotation="0,0,0", Scale="1,1,1") -> (Pose) {
    alphaInputType=Bool
    rotationMode=BMM_Additive
    rotationSpace=BCS_BoneSpace
    scaleMode=BMM_Additive
    scaleSpace=BCS_BoneSpace
    translationMode=BMM_Additive
    translationSpace=BCS_BoneSpace
    alphaBoolBlend {blendInTime=1, blendOption=CircularOut, blendOutTime=1}
    alphaScaleBiasClamp {bInterpResult=true}
    boneToModify {boneName=dyn_skirtSide_03_r}
   }
   "Transform (Modify) Bone - Bone: dyn_skirtBack_03_r"!(ComponentPose=@n6, bAlphaBoolEnabled:bool=True, Translation="-2.957,-6.382,0.094", Rotation="12.351,-24.729,0.815", Scale="1,1,1") -> (Pose) {
    alphaInputType=Bool
    rotationMode=BMM_Additive
    rotationSpace=BCS_BoneSpace
    scaleMode=BMM_Additive
    scaleSpace=BCS_BoneSpace
    translationMode=BMM_Additive
    translationSpace=BCS_BoneSpace
    alphaBoolBlend {blendInTime=1, blendOption=CircularOut, blendOutTime=1}
    alphaScaleBiasClamp {bInterpResult=true}
    boneToModify {boneName=dyn_skirtBack_03_r}
    rotation=<rot 12.351 -24.729 0.815>
    translation=<-2.957 -6.382 0.094>
   }
   Output Pose(Result=Blend Poses by bool(BlendPose_0=Use cached pose 'FullSim'(), BlendPose_1=Use cached pose 'RBANDisabled'(), bActiveValue=Get bRigidBodyEnabled()) {blendType=Linear}) {layerGroup=None}
   n7 = "Constraint - dyn_chestFeather_MSTR (RR)"(
    ComponentPose=Local To Component(LocalPose=Copy Pose From Mesh() {bCopyCurves=true, bUseAttachedParent=true})
    Alpha=1
    ConstraintWeights_0=1
    ConstraintWeights_1=0
   ) {
    alphaBoolBlend {blendOption=Linear}
    boneToModify {boneName=dyn_chestFeather_MSTR}
    constraintSetup [{transformType=Rotation, targetBone {boneName=clavicle_l}}, {transformType=Rotation, targetBone {boneName=upperarm_l}}]
    constraintWeights [1,1]
   }
   n8 = "Constraint - dyn_skirtBack_mstr_l (P)"!(ComponentPose=@n7, bAlphaBoolEnabled:bool=True, ConstraintWeights_0=1) -> (Pose) {
    alphaInputType=Bool
    alphaBoolBlend {blendOption=Linear}
    boneToModify {boneName=dyn_skirtBack_mstr_l}
    constraintSetup [{targetBone {boneName=spine_01}}]
    constraintWeights [1]
   }
   n9 = "Constraint - dyn_skirtBack_mstr (P)"!(ComponentPose=@n8, bAlphaBoolEnabled:bool=True, ConstraintWeights_0=1) -> (Pose) {
    alphaInputType=Bool
    alphaBoolBlend {blendOption=Linear}
    boneToModify {boneName=dyn_skirtBack_mstr}
    constraintSetup [{targetBone {boneName=spine_01}}]
    constraintWeights [1]
   }
   n10 = "Constraint - dyn_skirtBack_mstr_r (P)"!(ComponentPose=@n9, bAlphaBoolEnabled:bool=True, ConstraintWeights_0=1) -> (Pose) {
    alphaInputType=Bool
    alphaBoolBlend {blendOption=Linear}
    boneToModify {boneName=dyn_skirtBack_mstr_r}
    constraintSetup [{targetBone {boneName=spine_01}}]
    constraintWeights [1]
   }
   n11 = "Transform (Modify) Bone - Bone: dyn_skirtBack_mstr_l"!(ComponentPose=@n10, bAlphaBoolEnabled:bool=True, Translation="0.010,0.404,0.227", Rotation="10.236,-17.607,-31.775", Scale="1,1,1") -> (Pose) {
    alphaInputType=Bool
    rotationMode=BMM_Additive
    rotationSpace=BCS_BoneSpace
    scaleMode=BMM_Additive
    scaleSpace=BCS_BoneSpace
    translationMode=BMM_Additive
    translationSpace=BCS_BoneSpace
    alphaBoolBlend {blendInTime=1, blendOption=CircularOut, blendOutTime=1}
    alphaScaleBiasClamp {bInterpResult=true}
    boneToModify {boneName=dyn_skirtBack_mstr_l}
    rotation=<rot 10.236 -17.607 -31.775>
    translation=<0.01 0.404 0.227>
   }
   n12 = "Transform (Modify) Bone - Bone: dyn_skirtBack_mstr"!(ComponentPose=@n11, bAlphaBoolEnabled:bool=True, Translation="-2.403,-0.001,4.329", Rotation="74.994,-76.764,-76.412", Scale="1,1,1") -> (Pose) {
    alphaInputType=Bool
    rotationMode=BMM_Additive
    rotationSpace=BCS_BoneSpace
    scaleMode=BMM_Additive
    scaleSpace=BCS_BoneSpace
    translationMode=BMM_Additive
    translationSpace=BCS_BoneSpace
    alphaBoolBlend {blendInTime=1, blendOption=CircularOut, blendOutTime=1}
    alphaScaleBiasClamp {bInterpResult=true}
    boneToModify {boneName=dyn_skirtBack_mstr}
    rotation=<rot 74.994 -76.764 -76.412>
    translation=<-2.403 -0.001 4.329>
   }
   n13 = "Transform (Modify) Bone - Bone: dyn_skirtBack_mstr_r"!(ComponentPose=@n12, bAlphaBoolEnabled:bool=True, Translation="-0.690,13.399,3.432", Rotation="28.336,152.758,114.657", Scale="1,1,1") -> (Pose) {
    alphaInputType=Bool
    rotationMode=BMM_Additive
    rotationSpace=BCS_BoneSpace
    scaleMode=BMM_Additive
    scaleSpace=BCS_BoneSpace
    translationMode=BMM_Additive
    translationSpace=BCS_BoneSpace
    alphaBoolBlend {blendInTime=1, blendOption=CircularOut, blendOutTime=1}
    alphaScaleBiasClamp {bInterpResult=true}
    boneToModify {boneName=dyn_skirtBack_mstr_r}
    rotation=<rot 28.336 152.758 114.657>
    translation=<-0.69 13.399 3.432>
   }
   n14 = "Transform (Modify) Bone - Bone: dyn_skirtSide_mstr_r"(ComponentPose=@n13, bAlphaBoolEnabled=True, Translation="-10.811,-3.882,-0.869", Rotation="-51.462,48.232,-63.161", Scale="0.2,1,1") {
    alphaInputType=Bool
    rotationMode=BMM_Additive
    rotationSpace=BCS_BoneSpace
    scaleMode=BMM_Additive
    scaleSpace=BCS_BoneSpace
    translationMode=BMM_Additive
    translationSpace=BCS_BoneSpace
    alphaBoolBlend {blendInTime=1, blendOption=CircularOut, blendOutTime=1}
    alphaScaleBiasClamp {bInterpResult=true}
    boneToModify {boneName=dyn_skirtSide_mstr_r}
    rotation=<rot -51.462 48.232 -63.161>
    scale {x=0.2}
    translation=<-10.811 -3.882 -0.869>
   }
   n15 = Control Rig(Source=Component To Local(ComponentPose=@n14), Rifle=Get isAimingRifle(), Pistol=Get isAimingPistol()) {
    alphaBoolBlend {blendOption=Linear}
    controlRigAssetReference {blueprintRigClass=&F_MED_LiftingAura_Allure_CtrlRig_C}
   }
   n16 = "Transform (Modify) Bone - Bone: floorCOL"!(
    ComponentPose=Local To Component(LocalPose=@n15)
    bAlphaBoolEnabled:bool=Get FloorCOL() -> (FloorCOL:bool)
    Translation="0,0,-100"
    Rotation="0,0,0"
    Scale="1,1,1"
   ) -> (Pose) {
    alphaInputType=Bool
    rotationMode=BMM_Additive
    rotationSpace=BCS_BoneSpace
    translationMode=BMM_Additive
    translationSpace=BCS_BoneSpace
    alphaBoolBlend {blendInTime=0.1, blendOption=Linear, blendOutTime=0.5}
    alphaScaleBiasClamp {bInterpResult=true}
    boneToModify {boneName=floorCOL}
    translation {z=-100}
   }
   n17 = RigidBody!(ComponentPose=@n16, Alpha, OverrideWorldGravity:Vector="0,0,-980", ExternalForce) -> (Pose) {
    bClampLinearTranslationLimitToRefPose=true
    bOverrideWorldGravity=true
    overridePhysicsAsset=&/F_MED_LiftingAura_Allure_Physics
    simulationSpace=BaseBoneSpace
    alphaBoolBlend {blendOption=Linear}
    baseBoneRef {boneName=root}
    componentAppliedLinearAccClamp=<2500 2500 2000>
    componentLinearAccScale=<0.35 0.35 0.6>
    overrideWorldGravity {z=-980}
    simSpaceSettings {
     maxAngularAcceleration=500
     maxAngularVelocity=3
     maxLinearAcceleration=400
     maxLinearVelocity=500
     velocityScaleZ=0
     externalLinearDragV=<0.3 0.3>
    }
   }
   n18 = "Constraint - dyn_skirtBack_01_l (R)"(ComponentPose=@n17, Alpha=1, ConstraintWeights_0=1) {
    alphaBoolBlend {blendOption=Linear}
    boneToModify {boneName=dyn_skirtBack_01_l}
    constraintSetup [{transformType=Rotation, targetBone {boneName=thigh_l}}]
    constraintWeights [1]
   }
   n19 = "Constraint - dyn_skirtBack_02_l (R)"(ComponentPose=@n18, Alpha=1, ConstraintWeights_0=1) {
    alphaBoolBlend {blendOption=Linear}
    boneToModify {boneName=dyn_skirtBack_02_l}
    constraintSetup [{transformType=Rotation, targetBone {boneName=thigh_l}}]
    constraintWeights [1]
   }
   n20 = "Constraint - dyn_skirtBack_03_l (R)"(ComponentPose=@n19, Alpha=1, ConstraintWeights_0=1) {
    alphaBoolBlend {blendOption=Linear}
    boneToModify {boneName=dyn_skirtBack_03_l}
    constraintSetup [{transformType=Rotation, targetBone {boneName=calf_l}}]
    constraintWeights [1]
   }
   n21 = "Constraint - dyn_skirtBack_02 (RR)"!(ComponentPose=@n20, Alpha=1, ConstraintWeights_0=1, ConstraintWeights_1:float=0) -> (Pose) {
    alphaBoolBlend {blendOption=Linear}
    boneToModify {boneName=dyn_skirtBack_02}
    constraintSetup [{transformType=Rotation, targetBone {boneName=thigh_l}}, {transformType=Rotation, targetBone {boneName=thigh_r}}]
    constraintWeights [1,1]
   }
   n22 = "Constraint - dyn_skirtBack_01_r (R)"(ComponentPose=@n21, Alpha=1, ConstraintWeights_0=1) {
    alphaBoolBlend {blendOption=Linear}
    boneToModify {boneName=dyn_skirtBack_01_r}
    constraintSetup [{transformType=Rotation, targetBone {boneName=thigh_r}}]
    constraintWeights [1]
   }
   n23 = "Constraint - dyn_skirtBack_02_r (R)"(ComponentPose=@n22, Alpha=1, ConstraintWeights_0=1) {
    alphaBoolBlend {blendOption=Linear}
    boneToModify {boneName=dyn_skirtBack_02_r}
    constraintSetup [{transformType=Rotation, targetBone {boneName=thigh_r}}]
    constraintWeights [1]
   }
   n24 = "Constraint - dyn_skirtBack_03_r (P)"(ComponentPose=@n23, Alpha=1, ConstraintWeights_0=1) {
    alphaBoolBlend {blendOption=Linear}
    boneToModify {boneName=dyn_skirtBack_03_r}
    constraintSetup [{targetBone {boneName=thigh_r}}]
    constraintWeights [1]
   }
   n25 = "Constraint - dyn_skirtBack_03_r (R)"(ComponentPose=@n24, Alpha=1, ConstraintWeights_0=1) {
    alphaBoolBlend {blendOption=Linear}
    boneToModify {boneName=dyn_skirtBack_03_r}
    constraintSetup [{transformType=Rotation, targetBone {boneName=calf_r}}]
    constraintWeights [1]
   }
   n26 = "Constraint - dyn_skirtSide_mstr_r (R)"(ComponentPose=@n25, Alpha=1, ConstraintWeights_0=1) {
    alphaBoolBlend {blendOption=Linear}
    boneToModify {boneName=dyn_skirtSide_mstr_r}
    constraintSetup [{transformType=Rotation, targetBone {boneName=thigh_r}}]
    constraintWeights [1]
   }
   n27 = "Constraint - dyn_skirtSide_01_r (R)"(ComponentPose=@n26, Alpha=1, ConstraintWeights_0=1) {
    alphaBoolBlend {blendOption=Linear}
    boneToModify {boneName=dyn_skirtSide_01_r}
    constraintSetup [{transformType=Rotation, targetBone {boneName=thigh_r}}]
    constraintWeights [1]
   }
   n28 = "Constraint - dyn_skirtSide_02_r (R)"(ComponentPose=@n27, Alpha=1, ConstraintWeights_0=1) {
    alphaBoolBlend {blendOption=Linear}
    boneToModify {boneName=dyn_skirtSide_02_r}
    constraintSetup [{transformType=Rotation, targetBone {boneName=thigh_r}}]
    constraintWeights [1]
   }
   n29 = "Constraint - dyn_skirtSide_03_r (R)"(ComponentPose=@n28, Alpha=1, ConstraintWeights_0=1) {
    alphaBoolBlend {blendOption=Linear}
    boneToModify {boneName=dyn_skirtSide_03_r}
    constraintSetup [{transformType=Rotation, targetBone {boneName=thigh_r}}]
    constraintWeights [1]
   }
   n30 = "Constraint - dyn_skirtBack_l_flyoff_01 (PPPPPPPPP)"!(
    ComponentPose=@n29
    Alpha=1
    ConstraintWeights_0=0.0899
    ConstraintWeights_1:float=0.6513
    ConstraintWeights_2:float=0.0067
    ConstraintWeights_3:float=0.0201
    ConstraintWeights_4:float=0
    ConstraintWeights_5:float=0.0386
    ConstraintWeights_6:float=0.0218
    ConstraintWeights_7:float=0.1375
    ConstraintWeights_8:float=0.0341
   ) -> (Pose) {
    alphaBoolBlend {blendOption=Linear}
    boneToModify {boneName=dyn_skirtBack_l_flyoff_01}
    constraintSetup [
     {targetBone {boneName=thigh_twist_01_l}}
     {targetBone {boneName=thigh_l}}
     {targetBone {boneName=dyn_skirtBack_02_l}}
     {targetBone {boneName=dyn_skirtBack_03_l}}
     {targetBone {boneName=dyn_skirtBack_04_l}}
     {targetBone {boneName=dyn_skirtSide_02_l}}
     {targetBone {boneName=dyn_skirtSide_03_l}}
     {targetBone {boneName=dyn_skirtBack_l_flyoff_01}}
     {targetBone {boneName=dyn_skirtBack_l_flyoff_02}}
    ]
    constraintWeights [1,1,1,1,1,1,1,1,1]
   }
   n31 = "Constraint - dyn_skirtBack_l_flyoff_02 (PPPPPPPP)"!(
    ComponentPose=@n30
    Alpha=1
    ConstraintWeights_0=0.0708
    ConstraintWeights_1:float=0.8112
    ConstraintWeights_2:float=0.0048
    ConstraintWeights_3:float=0.029
    ConstraintWeights_4:float=0.0359
    ConstraintWeights_5:float=0.0178
    ConstraintWeights_6:float=0.0235
    ConstraintWeights_7:float=0.0069
   ) -> (Pose) {
    alphaBoolBlend {blendOption=Linear}
    boneToModify {boneName=dyn_skirtBack_l_flyoff_02}
    constraintSetup [
     {targetBone {boneName=thigh_twist_01_l}}
     {targetBone {boneName=thigh_l}}
     {targetBone {boneName=dyn_skirtBack_02_l}}
     {targetBone {boneName=dyn_skirtBack_03_l}}
     {targetBone {boneName=dyn_skirtBack_04_l}}
     {targetBone {boneName=dyn_skirtSide_02_l}}
     {targetBone {boneName=dyn_skirtSide_03_l}}
     {targetBone {boneName=dyn_skirtBack_l_flyoff_01}}
    ]
    constraintWeights [1,1,1,1,1,1,1,1]
   }
   n32 = "Constraint - dyn_skirtBack_l_flyoff_03 (RR)"!(ComponentPose=@n31, Alpha=1, ConstraintWeights_0=0.7, ConstraintWeights_1:float=0.3) -> (Pose) {
    alphaBoolBlend {blendOption=Linear}
    boneToModify {boneName=dyn_skirtBack_l_flyoff_03}
    constraintSetup [{transformType=Rotation, targetBone {boneName=dyn_skirtBack_03_l}}, {transformType=Rotation, targetBone {boneName=thigh_l}}]
    constraintWeights [1,1]
   }
   n33 = "Transform (Modify) Bone - Bone: dyn_skirtBack_01_l"(ComponentPose=@n32, Alpha=1, Translation="0.001,-2.429,5.000", Rotation="-12.826,-0.000,-0.000", Scale="1,1,1") {
    rotationMode=BMM_Additive
    rotationSpace=BCS_BoneSpace
    scaleSpace=BCS_BoneSpace
    translationMode=BMM_Additive
    translationSpace=BCS_BoneSpace
    alphaBoolBlend {blendInTime=0.1, blendOption=Linear, blendOutTime=0.5}
    alphaScaleBiasClamp {bInterpResult=true}
    boneToModify {boneName=dyn_skirtBack_01_l}
    rotation=<rot -12.826 0 0>
    translation=<0.001 -2.429 5>
   }
   n34 = "Transform (Modify) Bone - Bone: dyn_skirtBack_03_l"(ComponentPose=@n33, Alpha=1, Translation="0,20,5", Rotation="0,0,0", Scale="1,1,1") {
    rotationMode=BMM_Additive
    rotationSpace=BCS_BoneSpace
    scaleSpace=BCS_BoneSpace
    translationMode=BMM_Additive
    translationSpace=BCS_BoneSpace
    alphaBoolBlend {blendInTime=0.1, blendOption=Linear, blendOutTime=0.5}
    alphaScaleBiasClamp {bInterpResult=true}
    boneToModify {boneName=dyn_skirtBack_03_l}
    translation {y=20, z=5}
   }
   n35 = "Transform (Modify) Bone - Bone: dyn_skirtBack_l_flyoff_MSTR"(ComponentPose=@n34, Alpha=1, Translation="-1.74,-0.648,0", Rotation="28.673,12.024,43.289", Scale="1,1,1") {
    rotationMode=BMM_Additive
    rotationSpace=BCS_BoneSpace
    scaleSpace=BCS_BoneSpace
    translationMode=BMM_Additive
    translationSpace=BCS_BoneSpace
    alphaBoolBlend {blendInTime=0.1, blendOption=Linear, blendOutTime=0.5}
    alphaScaleBiasClamp {bInterpResult=true}
    boneToModify {boneName=dyn_skirtBack_l_flyoff_MSTR}
    rotation=<rot 28.673 12.024 43.289>
    translation=<-1.74 -0.648>
   }
   n36 = "Transform (Modify) Bone - Bone: dyn_skirtBack_mstr"!(ComponentPose=@n35, Alpha:float=1, Translation="-1.133,-0.000,3.335", Rotation="-28.414,-0.000,0.000", Scale="1,1,1") -> (Pose) {
    rotationMode=BMM_Additive
    rotationSpace=BCS_BoneSpace
    scaleSpace=BCS_BoneSpace
    translationMode=BMM_Additive
    translationSpace=BCS_BoneSpace
    alphaBoolBlend {blendInTime=0.1, blendOption=Linear, blendOutTime=0.5}
    alphaScaleBiasClamp {bInterpResult=true}
    boneToModify {boneName=dyn_skirtBack_mstr}
    rotation=<rot -28.414 0 0>
    translation {x=-1.133, z=3.335}
   }
   n37 = "Transform (Modify) Bone - Bone: dyn_skirtBack_01_r"(ComponentPose=@n36, Alpha=1, Translation="0.000,14.285,4.351", Rotation="-0.000,-23.422,-0.000", Scale="1,1,1") {
    rotationMode=BMM_Additive
    rotationSpace=BCS_BoneSpace
    scaleSpace=BCS_BoneSpace
    translationMode=BMM_Additive
    translationSpace=BCS_BoneSpace
    alphaBoolBlend {blendInTime=0.1, blendOption=Linear, blendOutTime=0.5}
    alphaScaleBiasClamp {bInterpResult=true}
    boneToModify {boneName=dyn_skirtBack_01_r}
    rotation=<rot 0 -23.422 0>
    translation {y=14.285, z=4.351}
   }
   n38 = "Transform (Modify) Bone - Bone: dyn_skirtBack_03_r"!(ComponentPose=@n37, Alpha:float=1, Translation="0.081,-5.182,-0.477", Rotation="0,0,0", Scale="1,1,1") -> (Pose) {
    rotationMode=BMM_Additive
    rotationSpace=BCS_BoneSpace
    scaleSpace=BCS_BoneSpace
    translationMode=BMM_Additive
    translationSpace=BCS_BoneSpace
    alphaBoolBlend {blendInTime=0.1, blendOption=Linear, blendOutTime=0.5}
    alphaScaleBiasClamp {bInterpResult=true}
    boneToModify {boneName=dyn_skirtBack_03_r}
    translation=<0.081 -5.182 -0.477>
   }
   n39 = Anim Dynamics - Bone: dyn_skirtBack_l_flyoff_02(ComponentPose=@n38, Alpha=0.4, ExternalForce="0,0,0") {
    angularDampingOverride=1
    angularSpringConstant=500
    bAngularSpring=true
    bOverrideAngularDamping=true
    bUseGravityOverride=true
    bUsePlanarLimit=false
    bUseSphericalLimits=true
    simulationSpace=BoneRelative
    alphaBoolBlend {blendOption=Linear}
    boundBone {boneName=dyn_skirtBack_l_flyoff_02}
    componentAppliedLinearAccClamp=<3000 3000 3000>
    componentLinearAccScale=<5 5 8>
    componentLinearVelScale=<5 5 8>
    gravityOverride {z=-980}
    physicsBodyDefinitions [{
      collisionType=CustomSphere
      sphereCollisionRadius=1
      boundBone {boneName=dyn_skirtBack_l_flyoff_02}
      boxExtents=<10 25 25>
      constraintSetup {angularLimitsMax=<5 0 30>, angularLimitsMin=<-5 -30 -10>, angularTarget=<0 0 0>}
      localJointOffset=<10 0 0>
     }]
    relativeSpaceBone {boneName=root}
    sphericalLimits [
     {limitRadius=130, drivingBone {boneName=root}, sphereLocalOffset=<0 0 -100>}
     {limitRadius=22, drivingBone {boneName=calf_l}, sphereLocalOffset=<0 0 0>}
     {limitRadius=30, drivingBone {boneName=calf_twist_02_l}, sphereLocalOffset=<0 0 0>}
    ]
   }
   n40 = Anim Dynamics - Bone: dyn_skirtBack_l_flyoff_03(ComponentPose=@n39, Alpha=0.6, ExternalForce="0,0,0") {
    angularDampingOverride=1
    angularSpringConstant=500
    bAngularSpring=true
    bOverrideAngularDamping=true
    bUseGravityOverride=true
    bUsePlanarLimit=false
    bUseSphericalLimits=true
    simulationSpace=BoneRelative
    alphaBoolBlend {blendOption=Linear}
    boundBone {boneName=dyn_skirtBack_l_flyoff_03}
    componentAppliedLinearAccClamp=<3000 3000 3000>
    componentLinearAccScale=<5 5 8>
    componentLinearVelScale=<5 5 8>
    gravityOverride {z=-980}
    physicsBodyDefinitions [{
      collisionType=CustomSphere
      sphereCollisionRadius=1
      boundBone {boneName=dyn_skirtBack_l_flyoff_03}
      boxExtents=<10 25 25>
      constraintSetup {angularLimitsMax=<5 0 30>, angularLimitsMin=<-5 -30 -10>, angularTarget=<0 0 0>}
      localJointOffset=<10 0 0>
     }]
    relativeSpaceBone {boneName=root}
    sphericalLimits [
     {limitRadius=130, drivingBone {boneName=root}, sphereLocalOffset=<0 0 -100>}
     {limitRadius=22, drivingBone {boneName=calf_l}, sphereLocalOffset=<0 0 0>}
     {limitRadius=30, drivingBone {boneName=calf_twist_02_l}, sphereLocalOffset=<0 0 0>}
    ]
   }
   n41 = Anim Dynamics - Bone: dyn_skirtBack_l_flyoff_04(ComponentPose=@n40, Alpha=0.8, ExternalForce="0,0,0") {
    angularDampingOverride=1
    angularSpringConstant=500
    bAngularSpring=true
    bOverrideAngularDamping=true
    bUseGravityOverride=true
    bUsePlanarLimit=false
    bUseSphericalLimits=true
    simulationSpace=BoneRelative
    alphaBoolBlend {blendOption=Linear}
    boundBone {boneName=dyn_skirtBack_l_flyoff_04}
    componentAppliedLinearAccClamp=<3000 3000 3000>
    componentLinearAccScale=<5 5 8>
    componentLinearVelScale=<5 5 8>
    gravityOverride {z=-980}
    physicsBodyDefinitions [{
      collisionType=CustomSphere
      sphereCollisionRadius=1
      boundBone {boneName=dyn_skirtBack_l_flyoff_04}
      boxExtents=<10 25 25>
      constraintSetup {angularLimitsMax=<5 0 30>, angularLimitsMin=<-5 -30 -10>, angularTarget=<0 0 0>}
      localJointOffset=<10 0 0>
     }]
    relativeSpaceBone {boneName=root}
    sphericalLimits [
     {limitRadius=130, drivingBone {boneName=root}, sphereLocalOffset=<0 0 -100>}
     {limitRadius=22, drivingBone {boneName=calf_l}, sphereLocalOffset=<0 0 0>}
     {limitRadius=30, drivingBone {boneName=calf_twist_02_l}, sphereLocalOffset=<0 0 0>}
    ]
   }
   n42 = "Transform (Modify) Bone - Bone: dyn_skirtBack_mstr_r"(ComponentPose=@n41, bAlphaBoolEnabled=True, Translation="0,0,5", Rotation="0,0,0", Scale="1,1,1") {
    alphaInputType=Bool
    rotationMode=BMM_Additive
    rotationSpace=BCS_BoneSpace
    translationMode=BMM_Additive
    translationSpace=BCS_BoneSpace
    alphaBoolBlend {blendInTime=0.1, blendOption=Linear, blendOutTime=0.1}
    alphaScaleBiasClamp {bInterpResult=true}
    boneToModify {boneName=dyn_skirtBack_mstr_r}
    translation {z=5}
   }
   n43 = "Transform (Modify) Bone - Bone: dyn_skirtBack_mstr"!(ComponentPose=@n42, bAlphaBoolEnabled:bool=True, Translation="0,0,4", Rotation="0,0,0", Scale="1,1,1") -> (Pose) {
    alphaInputType=Bool
    rotationMode=BMM_Additive
    rotationSpace=BCS_BoneSpace
    translationMode=BMM_Additive
    translationSpace=BCS_BoneSpace
    alphaBoolBlend {blendInTime=0.1, blendOption=Linear, blendOutTime=0.1}
    alphaScaleBiasClamp {bInterpResult=true}
    boneToModify {boneName=dyn_skirtBack_mstr}
    translation {z=4}
   }
   n44 = "Transform (Modify) Bone - Bone: dyn_skirtBack_mstr_l"(ComponentPose=@n43, bAlphaBoolEnabled=True, Translation="0,0,4", Rotation="0,0,0", Scale="1,1,1") {
    alphaInputType=Bool
    rotationMode=BMM_Additive
    rotationSpace=BCS_BoneSpace
    translationMode=BMM_Additive
    translationSpace=BCS_BoneSpace
    alphaBoolBlend {blendInTime=0.1, blendOption=Linear, blendOutTime=0.1}
    alphaScaleBiasClamp {bInterpResult=true}
    boneToModify {boneName=dyn_skirtBack_mstr_l}
    translation {z=4}
   }
   n45 = "Transform (Modify) Bone - Bone: dyn_skirtBack_04_a_MSTR"!(ComponentPose=@n44, bAlphaBoolEnabled:bool=True, Translation="0,0,0", Rotation="0,-20,0", Scale="1,1,1") -> (Pose) {
    alphaInputType=Bool
    rotationMode=BMM_Additive
    rotationSpace=BCS_BoneSpace
    translationMode=BMM_Additive
    translationSpace=BCS_BoneSpace
    alphaBoolBlend {blendInTime=0.1, blendOption=Linear, blendOutTime=0.5}
    alphaScaleBiasClamp {bInterpResult=true}
    boneToModify {boneName=dyn_skirtBack_04_a_MSTR}
    rotation {yaw=-20}
   }
   n46 = "Transform (Modify) Bone - Bone: dyn_skirtBack_l_c_MSTR"!(ComponentPose=@n45, bAlphaBoolEnabled:bool=True, Translation="0,0,0", Rotation="0,-8,0", Scale="1,1,1") -> (Pose) {
    alphaInputType=Bool
    rotationMode=BMM_Additive
    rotationSpace=BCS_BoneSpace
    translationMode=BMM_Additive
    translationSpace=BCS_BoneSpace
    alphaBoolBlend {blendInTime=0.1, blendOption=Linear, blendOutTime=0.5}
    alphaScaleBiasClamp {bInterpResult=true}
    boneToModify {boneName=dyn_skirtBack_l_c_MSTR}
    rotation {yaw=-8}
   }
   n47 = "Transform (Modify) Bone - Bone: dyn_skirtSideFront_r_MSTR"(ComponentPose=@n46, bAlphaBoolEnabled=True, Translation="0.188,0.726,1.440", Rotation="-11.748,21.946,-8.859", Scale="1,1,1") {
    alphaInputType=Bool
    rotationMode=BMM_Additive
    rotationSpace=BCS_BoneSpace
    translationMode=BMM_Additive
    translationSpace=BCS_BoneSpace
    alphaBoolBlend {blendInTime=0.1, blendOption=Linear, blendOutTime=0.5}
    alphaScaleBiasClamp {bInterpResult=true}
    boneToModify {boneName=dyn_skirtSideFront_r_MSTR}
    rotation=<rot -11.748 21.946 -8.859>
    translation=<0.188 0.726 1.44>
   }
   n48 = "Transform (Modify) Bone - Bone: dyn_skirtBack_mstr_l"(ComponentPose=@n47, bAlphaBoolEnabled=True, Translation="-0.112,10.741,3.591", Rotation="-0.002,8.573,0.001", Scale="1,1,1") {
    alphaInputType=Bool
    rotationMode=BMM_Additive
    rotationSpace=BCS_BoneSpace
    translationMode=BMM_Additive
    translationSpace=BCS_BoneSpace
    alphaBoolBlend {blendInTime=0.1, blendOption=Linear, blendOutTime=0.5}
    alphaScaleBiasClamp {bInterpResult=true}
    boneToModify {boneName=dyn_skirtBack_mstr_l}
    rotation=<rot -0.002 8.573 0.001>
    translation=<-0.112 10.741 3.591>
   }
   n49 = "Transform (Modify) Bone - Bone: dyn_skirtSideFront_r_MSTR"(ComponentPose=@n48, bAlphaBoolEnabled=True, Translation="0.188,0.726,1.440", Rotation="-11.748,21.946,-8.859", Scale="1,1,1") {
    alphaInputType=Bool
    rotationMode=BMM_Additive
    rotationSpace=BCS_BoneSpace
    translationMode=BMM_Additive
    translationSpace=BCS_BoneSpace
    alphaBoolBlend {blendInTime=0.1, blendOption=Linear, blendOutTime=0.5}
    alphaScaleBiasClamp {bInterpResult=true}
    boneToModify {boneName=dyn_skirtSideFront_r_MSTR}
    rotation=<rot -11.748 21.946 -8.859>
    translation=<0.188 0.726 1.44>
   }
   n50 = "Constraint - dyn_sleeve_l_a_MSTR (R)"(ComponentPose=@n49, Alpha=0.3, ConstraintWeights_0=1) {
    alphaBoolBlend {blendOption=Linear}
    boneToModify {boneName=dyn_sleeve_l_a_MSTR}
    constraintSetup [{transformType=Rotation, targetBone {boneName=lowerarm_l}}]
    constraintWeights [1]
   }
   n51 = "Constraint - dyn_sleeve_l_b_MSTR (R)"(ComponentPose=@n50, Alpha=1, ConstraintWeights_0=1) {
    alphaBoolBlend {blendOption=Linear}
    boneToModify {boneName=dyn_sleeve_l_b_MSTR}
    constraintSetup [{transformType=Rotation, targetBone {boneName=lowerarm_l}}]
    constraintWeights [1]
   }
   n52 = Anim Dynamics - Bone: dyn_chestFeather_a_01(ComponentPose=@n51, Alpha=0.4, ExternalForce="0,0,0") {
    angularDampingOverride=0.999
    angularSpringConstant=500
    bAngularSpring=true
    bOverrideAngularDamping=true
    bUseGravityOverride=true
    bUsePlanarLimit=false
    bUseSphericalLimits=true
    simulationSpace=BoneRelative
    alphaBoolBlend {blendOption=Linear}
    boundBone {boneName=dyn_chestFeather_a_01}
    componentAppliedLinearAccClamp=<3000 3000 3000>
    componentLinearAccScale=<5 5 8>
    componentLinearVelScale=<5 5 8>
    gravityOverride {z=-300}
    physicsBodyDefinitions [{
      collisionType=CustomSphere
      sphereCollisionRadius=1
      boundBone {boneName=dyn_chestFeather_a_01}
      boxExtents=<25 25 25>
      constraintSetup {angularLimitsMax=<10 10 10>, angularLimitsMin=<-10 -10 -10>, angularTarget=<0 0 0>}
      localJointOffset=<10 0 0>
     }]
    relativeSpaceBone {boneName=root}
   }
   n53 = Anim Dynamics - Bone: dyn_chestFeather_b_01(ComponentPose=@n52, Alpha=0.3, ExternalForce="0,0,0") {
    angularDampingOverride=0.9999
    angularSpringConstant=500
    bAngularSpring=true
    bOverrideAngularDamping=true
    bUseGravityOverride=true
    bUsePlanarLimit=false
    bUseSphericalLimits=true
    simulationSpace=BoneRelative
    alphaBoolBlend {blendOption=Linear}
    boundBone {boneName=dyn_chestFeather_b_01}
    componentAppliedLinearAccClamp=<3000 3000 3000>
    componentLinearAccScale=<5 5 8>
    componentLinearVelScale=<5 5 8>
    gravityOverride {z=-300}
    physicsBodyDefinitions [{
      collisionType=CustomSphere
      sphereCollisionRadius=1
      boundBone {boneName=dyn_chestFeather_b_01}
      boxExtents=<30 30 30>
      constraintSetup {angularLimitsMax=<10 10 10>, angularLimitsMin=<-10 -10 -10>, angularTarget=<0 0 0>}
      localJointOffset=<8 0 0>
     }]
    relativeSpaceBone {boneName=root}
   }
   n54 = Anim Dynamics - Bone: dyn_chestFeather_c_01(ComponentPose=@n53, Alpha=0.25, ExternalForce="0,0,0") {
    angularDampingOverride=0.999
    angularSpringConstant=500
    bAngularSpring=true
    bOverrideAngularDamping=true
    bUseGravityOverride=true
    bUsePlanarLimit=false
    bUseSphericalLimits=true
    simulationSpace=BoneRelative
    alphaBoolBlend {blendOption=Linear}
    boundBone {boneName=dyn_chestFeather_c_01}
    componentAppliedLinearAccClamp=<3000 3000 3000>
    componentLinearAccScale=<5 5 8>
    componentLinearVelScale=<5 5 8>
    gravityOverride {z=-300}
    physicsBodyDefinitions [{
      collisionType=CustomSphere
      sphereCollisionRadius=1
      boundBone {boneName=dyn_chestFeather_c_01}
      boxExtents=<20 20 20>
      constraintSetup {angularLimitsMax=<10 10 10>, angularLimitsMin=<-10 -10 -10>, angularTarget=<0 0 0>}
      localJointOffset=<11 0 0>
     }]
    relativeSpaceBone {boneName=root}
   }
   n55 = Anim Dynamics - Bone: dyn_chestFeather_d_01(ComponentPose=@n54, Alpha=0.5, ExternalForce="0,0,0") {
    angularDampingOverride=0.9
    angularSpringConstant=500
    bAngularSpring=true
    bOverrideAngularDamping=true
    bUseGravityOverride=true
    bUsePlanarLimit=false
    bUseSphericalLimits=true
    simulationSpace=BoneRelative
    alphaBoolBlend {blendOption=Linear}
    boundBone {boneName=dyn_chestFeather_d_01}
    componentAppliedLinearAccClamp=<3000 3000 3000>
    componentLinearAccScale=<5 5 8>
    componentLinearVelScale=<5 5 8>
    gravityOverride {z=-300}
    physicsBodyDefinitions [{
      collisionType=CustomSphere
      sphereCollisionRadius=1
      boundBone {boneName=dyn_chestFeather_d_01}
      boxExtents=<23 23 23>
      constraintSetup {angularLimitsMax=<10 10 10>, angularLimitsMin=<-10 -10 -10>, angularTarget=<0 0 0>}
      localJointOffset=<6 0 0>
     }]
    relativeSpaceBone {boneName=root}
   }
   n56 = "Transform (Modify) Bone - Bone: dyn_skirtSide_mstr_r"(ComponentPose=@n55, bAlphaBoolEnabled=True, Translation="16.350,18.000,5.231", Rotation="16.700,0.387,6.033", Scale="0.126,1.000,1.000") {
    alphaInputType=Bool
    rotationMode=BMM_Additive
    rotationSpace=BCS_BoneSpace
    scaleMode=BMM_Additive
    scaleSpace=BCS_BoneSpace
    translationMode=BMM_Additive
    translationSpace=BCS_BoneSpace
    alphaBoolBlend {blendInTime=1, blendOption=CircularOut, blendOutTime=1}
    alphaScaleBiasClamp {bInterpResult=true}
    boneToModify {boneName=dyn_skirtSide_mstr_r}
    rotation=<rot 16.7 0.387 6.033>
    scale {x=0.126}
    translation=<16.35 18 5.231>
   }
   n57 = "Transform (Modify) Bone - Bone: dyn_skirtSideFront_r_MSTR"(ComponentPose=@n56, bAlphaBoolEnabled=True, Translation="2.104,0.071,1.910", Rotation="6.572,-32.215,-1.664", Scale="1,1,1") {
    alphaInputType=Bool
    rotationMode=BMM_Additive
    rotationSpace=BCS_BoneSpace
    scaleMode=BMM_Additive
    scaleSpace=BCS_BoneSpace
    translationMode=BMM_Additive
    translationSpace=BCS_BoneSpace
    alphaBoolBlend {blendInTime=1, blendOption=CircularOut, blendOutTime=1}
    alphaScaleBiasClamp {bInterpResult=true}
    boneToModify {boneName=dyn_skirtSideFront_r_MSTR}
    rotation=<rot 6.572 -32.215 -1.664>
    translation=<2.104 0.071 1.91>
   }
   n58 = "Transform (Modify) Bone - Bone: dyn_skirtSide_03_r"(ComponentPose=@n57, bAlphaBoolEnabled=True, Translation="-0.000,-0.000,-7.780", Rotation="-1.232,2.807,0.937", Scale="1,1,1") {
    alphaInputType=Bool
    rotationMode=BMM_Additive
    rotationSpace=BCS_BoneSpace
    scaleMode=BMM_Additive
    scaleSpace=BCS_BoneSpace
    translationMode=BMM_Additive
    translationSpace=BCS_BoneSpace
    alphaBoolBlend {blendInTime=1, blendOption=CircularOut, blendOutTime=1}
    alphaScaleBiasClamp {bInterpResult=true}
    boneToModify {boneName=dyn_skirtSide_03_r}
    rotation=<rot -1.2321 2.807 0.9371>
    translation {z=-7.78}
   }
   n59 = "Transform (Modify) Bone - Bone: dyn_skirtBack_03_r"!(ComponentPose=@n58, bAlphaBoolEnabled:bool=True, Translation="-2.957,-6.382,0.094", Rotation="12.351,-24.729,0.815", Scale="1,1,1") -> (Pose) {
    alphaInputType=Bool
    rotationMode=BMM_Additive
    rotationSpace=BCS_BoneSpace
    scaleMode=BMM_Additive
    scaleSpace=BCS_BoneSpace
    translationMode=BMM_Additive
    translationSpace=BCS_BoneSpace
    alphaBoolBlend {blendInTime=1, blendOption=CircularOut, blendOutTime=1}
    alphaScaleBiasClamp {bInterpResult=true}
    boneToModify {boneName=dyn_skirtBack_03_r}
    rotation=<rot 12.351 -24.729 0.815>
    translation=<-2.957 -6.382 0.094>
   }
   Save cached pose 'FullSim'(Pose=Component To Local(ComponentPose=@n59)) {cacheName=FullSim}
   n60 = "Constraint - dyn_chestFeather_MSTR (RR)"(
    ComponentPose=Local To Component(LocalPose=Copy Pose From Mesh() {bCopyCurves=true, bUseAttachedParent=true})
    Alpha=1
    ConstraintWeights_0=1
    ConstraintWeights_1=0
   ) {
    alphaBoolBlend {blendOption=Linear}
    boneToModify {boneName=dyn_chestFeather_MSTR}
    constraintSetup [{transformType=Rotation, targetBone {boneName=clavicle_l}}, {transformType=Rotation, targetBone {boneName=upperarm_l}}]
    constraintWeights [1,1]
   }
   n61 = Control Rig(Source=Component To Local(ComponentPose=@n60), Rifle=Get isAimingRifle(), Pistol=Get isAimingPistol()) {
    alphaBoolBlend {blendOption=Linear}
    controlRigAssetReference {blueprintRigClass=&F_MED_LiftingAura_Allure_CtrlRig_C}
   }
   n62 = "Constraint - dyn_skirtBack_mstr_l (RR)"!(ComponentPose=Local To Component(LocalPose=@n61), Alpha=1, ConstraintWeights_0=0.5, ConstraintWeights_1:float=0.5) -> (Pose) {
    alphaBoolBlend {blendOption=Linear}
    alphaScaleBiasClamp {bInterpResult=true}
    boneToModify {boneName=dyn_skirtBack_mstr_l}
    constraintSetup [{transformType=Rotation, targetBone {boneName=thigh_l}}, {transformType=Rotation, targetBone {boneName=pelvis}}]
    constraintWeights [1,1]
   }
   n63 = "Constraint - dyn_skirtBack_mstr_r (RR)"!(ComponentPose=@n62, Alpha=1, ConstraintWeights_0=0.5, ConstraintWeights_1:float=0.5) -> (Pose) {
    alphaBoolBlend {blendOption=Linear}
    alphaScaleBiasClamp {bInterpResult=true}
    boneToModify {boneName=dyn_skirtBack_mstr_r}
    constraintSetup [{transformType=Rotation, targetBone {boneName=thigh_r}}, {transformType=Rotation, targetBone {boneName=pelvis}}]
    constraintWeights [1,1]
   }
   n64 = "Constraint - dyn_skirtBack_mstr_l (R)"!(ComponentPose=@n63, bAlphaBoolEnabled:bool=True, ConstraintWeights_0=1) -> (Pose) {
    alphaInputType=Bool
    alphaBoolBlend {blendInTime=0.1, blendOption=Linear, blendOutTime=0.5}
    boneToModify {boneName=dyn_skirtBack_mstr_l}
    constraintSetup [{transformType=Rotation, targetBone {boneName=pelvis}}]
    constraintWeights [1]
   }
   n65 = "Constraint - dyn_skirtBack_mstr_r (R)"!(ComponentPose=@n64, bAlphaBoolEnabled:bool=True, ConstraintWeights_0=1) -> (Pose) {
    alphaInputType=Bool
    alphaBoolBlend {blendInTime=0.1, blendOption=Linear, blendOutTime=0.5}
    boneToModify {boneName=dyn_skirtBack_mstr_r}
    constraintSetup [{transformType=Rotation, targetBone {boneName=pelvis}}]
    constraintWeights [1]
   }
   n66 = "Transform (Modify) Bone - Bone: dyn_skirtBack_mstr_r"(ComponentPose=@n65, bAlphaBoolEnabled=True, Translation="0,0,0", Rotation="58.592,64.074,36.209", Scale="1,1,1") {
    alphaInputType=Bool
    rotationMode=BMM_Additive
    rotationSpace=BCS_BoneSpace
    scaleMode=BMM_Additive
    scaleSpace=BCS_BoneSpace
    translationMode=BMM_Additive
    translationSpace=BCS_BoneSpace
    alphaBoolBlend {blendInTime=0.5, blendOption=CircularOut, blendOutTime=0.5}
    alphaScaleBiasClamp {bInterpResult=true}
    boneToModify {boneName=dyn_skirtBack_mstr_r}
    rotation=<rot 58.592 64.074 36.209>
   }
   n67 = "Transform (Modify) Bone - Bone: dyn_skirtBack_mstr_l"(ComponentPose=@n66, bAlphaBoolEnabled=True, Translation="0,0,0", Rotation="30.684,-35.266,-38.784", Scale="1,1,1") {
    alphaInputType=Bool
    rotationMode=BMM_Additive
    rotationSpace=BCS_BoneSpace
    scaleMode=BMM_Additive
    scaleSpace=BCS_BoneSpace
    translationMode=BMM_Additive
    translationSpace=BCS_BoneSpace
    alphaBoolBlend {blendInTime=0.5, blendOption=CircularOut, blendOutTime=0.5}
    alphaScaleBiasClamp {bInterpResult=true}
    boneToModify {boneName=dyn_skirtBack_mstr_l}
    rotation=<rot 30.684 -35.266 -38.784>
   }
   n68 = "Transform (Modify) Bone - Bone: dyn_skirtBack_mstr_l"(ComponentPose=@n67, bAlphaBoolEnabled=True, Translation="-7.891,7.725,-4.175", Rotation="-0.172,23.505,0.054", Scale="1,1,1") {
    alphaInputType=Bool
    rotationMode=BMM_Additive
    rotationSpace=BCS_BoneSpace
    scaleMode=BMM_Additive
    scaleSpace=BCS_BoneSpace
    translationMode=BMM_Additive
    translationSpace=BCS_BoneSpace
    alphaBoolBlend {blendInTime=0.1, blendOption=CircularOut, blendOutTime=0.5}
    alphaScaleBiasClamp {bInterpResult=true}
    boneToModify {boneName=dyn_skirtBack_mstr_l}
    rotation=<rot -0.172 23.505 0.054>
    translation=<-7.891 7.725 -4.175>
   }
   n69 = "Transform (Modify) Bone - Bone: dyn_skirtBack_mstr_r"(ComponentPose=@n68, bAlphaBoolEnabled=True, Translation="0,0,0", Rotation="-21.599,-28.674,-0.151", Scale="1,1,1") {
    alphaInputType=Bool
    rotationMode=BMM_Additive
    rotationSpace=BCS_BoneSpace
    scaleMode=BMM_Additive
    scaleSpace=BCS_BoneSpace
    translationMode=BMM_Additive
    translationSpace=BCS_BoneSpace
    alphaBoolBlend {blendInTime=0.1, blendOption=CircularOut, blendOutTime=0.5}
    alphaScaleBiasClamp {bInterpResult=true}
    boneToModify {boneName=dyn_skirtBack_mstr_r}
    rotation=<rot -21.599 -28.674 -0.151>
   }
   n70 = "Transform (Modify) Bone - Bone: dyn_skirtBack_mstr_l"(ComponentPose=@n69, bAlphaBoolEnabled=True, Translation="0,0,0", Rotation="-18.755,-3.613,1.074", Scale="1,1,1") {
    alphaInputType=Bool
    rotationMode=BMM_Additive
    rotationSpace=BCS_BoneSpace
    scaleMode=BMM_Additive
    scaleSpace=BCS_BoneSpace
    translationMode=BMM_Additive
    translationSpace=BCS_BoneSpace
    alphaBoolBlend {blendInTime=0.5, blendOption=CircularOut, blendOutTime=0.5}
    alphaScaleBiasClamp {bInterpResult=true}
    boneToModify {boneName=dyn_skirtBack_mstr_l}
    rotation=<rot -18.755 -3.613 1.074>
   }
   n71 = "Transform (Modify) Bone - Bone: dyn_skirtBack_mstr_r"(ComponentPose=@n70, bAlphaBoolEnabled=True, Translation="-1.053,0.525,-3.737", Rotation="-31.689,-23.786,35.520", Scale="1,1,1") {
    alphaInputType=Bool
    rotationMode=BMM_Additive
    rotationSpace=BCS_BoneSpace
    scaleMode=BMM_Additive
    scaleSpace=BCS_BoneSpace
    translationMode=BMM_Additive
    translationSpace=BCS_BoneSpace
    alphaBoolBlend {blendInTime=0.5, blendOption=CircularOut, blendOutTime=0.5}
    alphaScaleBiasClamp {bInterpResult=true}
    boneToModify {boneName=dyn_skirtBack_mstr_r}
    rotation=<rot -31.689 -23.786 35.52>
    translation=<-1.053 0.525 -3.737>
   }
   n72 = "Transform (Modify) Bone - Bone: dyn_skirtBack_mstr_l"(ComponentPose=@n71, bAlphaBoolEnabled=True, Translation="0,0,0", Rotation="21.441,-10.949,-3.962", Scale="1,1,1") {
    alphaInputType=Bool
    rotationMode=BMM_Additive
    rotationSpace=BCS_BoneSpace
    scaleMode=BMM_Additive
    scaleSpace=BCS_BoneSpace
    translationMode=BMM_Additive
    translationSpace=BCS_BoneSpace
    alphaBoolBlend {blendInTime=0.5, blendOption=CircularOut, blendOutTime=0.5}
    alphaScaleBiasClamp {bInterpResult=true}
    boneToModify {boneName=dyn_skirtBack_mstr_l}
    rotation=<rot 21.441 -10.949 -3.962>
   }
   n73 = "Transform (Modify) Bone - Bone: dyn_skirtBack_mstr_r"(ComponentPose=@n72, bAlphaBoolEnabled=True, Translation="0,0,0", Rotation="6.553,18.075,-0.193", Scale="1,1,1") {
    alphaInputType=Bool
    rotationMode=BMM_Additive
    rotationSpace=BCS_BoneSpace
    scaleMode=BMM_Additive
    scaleSpace=BCS_BoneSpace
    translationMode=BMM_Additive
    translationSpace=BCS_BoneSpace
    alphaBoolBlend {blendInTime=0.5, blendOption=CircularOut, blendOutTime=0.5}
    alphaScaleBiasClamp {bInterpResult=true}
    boneToModify {boneName=dyn_skirtBack_mstr_r}
    rotation=<rot 6.553 18.075 -0.193>
   }
   n74 = "Constraint - dyn_skirtBack_04_r (R)"!(ComponentPose=@n73, bAlphaBoolEnabled:bool=True, ConstraintWeights_0=1) -> (Pose) {
    alphaInputType=Bool
    alphaBoolBlend {blendOption=Linear}
    alphaScaleBiasClamp {bInterpResult=true}
    boneToModify {boneName=dyn_skirtBack_04_r}
    constraintSetup [{transformType=Rotation, targetBone {boneName=calf_r}}]
    constraintWeights [1]
   }
   n75 = "Constraint - dyn_skirtBack_mstr_r (R)"(ComponentPose=@n74, bAlphaBoolEnabled=True, ConstraintWeights_0=1) {
    alphaInputType=Bool
    alphaBoolBlend {blendOption=Linear}
    alphaScaleBiasClamp {bInterpResult=true}
    boneToModify {boneName=dyn_skirtBack_mstr_r}
    constraintSetup [{transformType=Rotation, targetBone {boneName=thigh_r}}]
    constraintWeights [1]
   }
   n76 = "Transform (Modify) Bone - Bone: dyn_skirtBack_mstr_r"(ComponentPose=@n75, bAlphaBoolEnabled=True, Translation="-0.000,0.000,2.806", Rotation="31.155,28.693,-3.043", Scale="1,1,1") {
    alphaInputType=Bool
    rotationMode=BMM_Additive
    rotationSpace=BCS_BoneSpace
    scaleMode=BMM_Additive
    scaleSpace=BCS_BoneSpace
    translationMode=BMM_Additive
    translationSpace=BCS_BoneSpace
    alphaBoolBlend {blendInTime=0.5, blendOption=CircularOut, blendOutTime=0.5}
    alphaScaleBiasClamp {bInterpResult=true}
    boneToModify {boneName=dyn_skirtBack_mstr_r}
    rotation=<rot 31.155 28.693 -3.043>
    translation {z=2.806}
   }
   n77 = "Transform (Modify) Bone - Bone: dyn_skirtBack_mstr_r"!(ComponentPose=@n76, bAlphaBoolEnabled=True, Rotation="-21,-10,5") -> (Pose) {
    alphaInputType=Bool
    rotationMode=BMM_Additive
    rotationSpace=BCS_BoneSpace
    scaleMode=BMM_Additive
    scaleSpace=BCS_BoneSpace
    translationMode=BMM_Additive
    translationSpace=BCS_BoneSpace
    alphaBoolBlend {blendInTime=0.5, blendOption=CircularOut, blendOutTime=0.5}
    alphaScaleBiasClamp {bInterpResult=true}
    boneToModify {boneName=dyn_skirtBack_mstr_r}
    rotation=<rot -21 -10 5>
   }
   Save cached pose 'RBANDisabled'(Pose=Component To Local(ComponentPose=@n77)) {cacheName=RBANDisabled}
  ]}
 EventGraph {nodes [
   Event BlueprintUpdateAnimation()
   > n1 = Sequence!(execute) -> (then_0, then_1, then_2:Exec, then_3:Exec, then_4:Exec, then_5:Exec, then_6:Exec, then_7:Exec)
   n2 = Cast To FortPlayerAnimInstance(execute=@n1.then_2, Object=GetAnimInstance(self=GetOwnerSkeletalMeshComponent()))
   > Set isAimingPistol(isAimingPistol:double=0) -> (Output_Get:double)
   > Set isPickaxe(isPickaxe:double=0) -> (Output_Get:double)
   > Set isAimingRifle(isAimingRifle:double=0) -> (Output_Get:double)
   > Set isPickaxeBoo(isPickaxeBoo:bool=false) -> (Output_Get:bool)
   > n3 = Switch on EFortWeaponCoreAnimation!(execute, Selection:EFortWeaponCoreAnimation Enum=Get WeaponCoreAnim(self=@n2)) -> (
    Melee:Exec
    Pistol:Exec
    Shotgun:Exec
    PaperBlueprint:Exec
    Rifle:Exec
    MeleeOneHand:Exec
    MachinePistol:Exec
    RocketLauncher:Exec
    GrenadeLauncher:Exec
    GoingCommando:Exec
    AssaultRifle:Exec
    TacticalShotgun:Exec
    SniperRifle:Exec
    TrapPlacement:Exec
    ShoulderLauncher:Exec
    AbilityDecoTool:Exec
    Crossbow:Exec
    C4:Exec
    RemoteControl:Exec
    DualWield:Exec
    AR_BullPup:Exec
    AR_ForwardGrip:Exec
    MedPackPaddles:Exec
    SMG_P90:Exec
    AR_DrumGun:Exec
    Consumable_Small:Exec
    Consumable_Large:Exec
    Balloon:Exec
    MountedTurret:Exec
    CreativeTool:Exec
    ExplosiveBow:Exec
    AshtonIndigo:Exec
    AshtonChicago:Exec
    MeleeDualWield:Exec
    Unarmed:Exec
   )
   Branch(execute=[@n3.AssaultRifle, @n3.Crossbow, @n3.GrenadeLauncher, @n3.Rifle, @n3.RocketLauncher, @n3.Shotgun, @n3.SniperRifle, @n3.TacticalShotgun])
   > Set isAimingRifle(isAimingRifle=1)
   Branch(execute=[@n3.MachinePistol, @n3.Pistol])
   > Set isAimingPistol(isAimingPistol=1)
   Branch(execute=@n3.Melee)
   > Set isPickaxe(isPickaxe=1)
   > Set isPickaxeBoo(isPickaxeBoo=true)
   n4 = Branch(execute=@n1.then_3, Condition=AND Boolean(A=Get bIsCrouching(), B=Get bIsAccelerating2D()))
   > Set CrouchWalking(CrouchWalking:bool=true) -> (Output_Get:bool)
   Set CrouchWalking(execute=@n4.else, CrouchWalking=false)
   n5 = GetAnimInstance(self=GetOwnerSkeletalMeshComponent())
   Is Valid(exec=@n1.then_4, InputObject=@n5)
   n6 = OR Boolean!(
    A=Montage_IsPlaying(self=@n5, Montage=/BRCosmetics/Animation/Game/MainPlayer/Emotes/LiftingAura/Emote_LiftingAura_CMF_M)
    B=false
    C:bool
    D:bool
    E:bool
   ) -> (ReturnValue)
   > Set isCustomEmotePlaying(isCustomEmotePlaying:double=SelectFloat(A=0, B=1, bPickA=@n6)) -> (Output_Get:double)
   > Set isCustomEmotePlaying2(isCustomEmotePlaying2:double=SelectFloat(A=1, B=0, bPickA=@n6)) -> (Output_Get:double)
   n7 = Branch(execute=@n1.then_5, Condition=NOT Boolean(A=Get bIsCrouching()))
   > Set NotCrouching(NotCrouching:bool=true) -> (Output_Get:bool)
   Set NotCrouching(execute=@n7.else, NotCrouching=false)
   n8 = Branch(execute=@n1.then_7, Condition=OR Boolean!(A=false, B=Get bIsJumping(), C=Get bIsFalling()) -> (ReturnValue))
   > Set JumpingFalling(JumpingFalling:bool=true) -> (Output_Get:bool)
   Set JumpingFalling(execute=@n8.else, JumpingFalling=false)
   n9 = Branch(execute=@n1.then_1, Condition=OR Boolean!(A=Get bIsJumping(), B=Get bIsFalling(), C=Get bIsOnGrindRail(), D:bool) -> (ReturnValue))
   > Set FloorCOL(FloorCOL:bool=true) -> (Output_Get:bool)
   Set FloorCOL(execute=@n9.else, FloorCOL=false)
   n10 = NOT Boolean(A=Get bIsCrouching())
   n11 = Get bIsAccelerating2D()
   n12 = Branch(execute=@n1.then_6, Condition=OR Boolean!(A=AND Boolean(A=@n10, B=@n11), B=Get bIsJumping(), C=Get bIsFalling()) -> (ReturnValue))
   > Set MovingNotCrouching(MovingNotCrouching:bool=true) -> (Output_Get:bool)
   Set MovingNotCrouching(execute=@n12.else, MovingNotCrouching=false)
   n13 = Gravity Override State Function DEPRECATED(
    execute=@n1.then_0
    CharacterPartAnimInstance=Self-Reference()
    GameTimeInSeconds=GetGameTimeInSeconds()
    GravityOverrideParams=Get Gravity Override Params()
   )
   > Set Gravity Override(Gravity Override=@n13)
   Get bIsDBNO()
   Get bIsSliding()
  ]}
]
variables [
 CrouchWalking {type=boolean}
 FloorCOL {type=boolean}
 Gravity Override {type=Vector}
 Gravity Override Params {type=GravityOverrideParamsStruct, default {skydive_GoingUp {gravityOverride=<0 -300 -980>}}}
 isAimingPistol {type=number}
 isAimingRifle {type=number}
 isCustomEmotePlaying {type=number}
 isCustomEmotePlaying2 {type=number}
 isPickaxe {type=number}
 isPickaxeBoo {type=boolean}
 JumpingFalling {type=boolean}
 MovingNotCrouching {type=boolean}
 NotCrouching {type=boolean}
]
