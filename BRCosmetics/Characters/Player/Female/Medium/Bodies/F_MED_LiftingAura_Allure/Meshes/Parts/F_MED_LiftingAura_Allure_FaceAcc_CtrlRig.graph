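# Control Rig for the F_MED_LiftingAura_Allure face accessory that procedurally corrects hair bones as the head and right arm move.
# Spherical pose readers on head, right clavicle and upper arm store alphas, some multiplied into combined look-and-rifle-pose weights.
# These alphas weight additive back and right hair offsets, while parent constraints blend hair and bangs between head, spine, clavicles and back-hair bones.
# Several chains, including MSTR-bone constraints and a DeformRig function, are unconnected to execution.
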ControlRigBlueprint /BRCosmetics/Characters/Player/Female/Medium/Bodies/F_MED_LiftingAura_Allure/Meshes/Parts/F_MED_LiftingAura_Allure_FaceAcc_CtrlRig
graphs [
 ContainedGraph {nodes [
   n1 = RigVMFunctionEntryNode()
   SequenceExecution_2 = &RigVMFunction_Sequence(ExecuteContext=@n1.ExecuteContext)
   SequenceExecution_2_1 = &RigVMFunction_Sequence(ExecuteContext=@SequenceExecution_2.B)
   SequenceExecution_2_2 = &RigVMFunction_Sequence(ExecuteContext=@SequenceExecution_2_1.B)
   SequenceExecution_2_3 = &RigVMFunction_Sequence(ExecuteContext=@SequenceExecution_2_2.B)
   RigVMFunctionReturnNode(A=@SequenceExecution_2.A, B=@SequenceExecution_2_1.A, C=@SequenceExecution_2_2.A, D=@SequenceExecution_2_3.A, E=@SequenceExecution_2_3.B)
  ]}
 RigVMFunctionLibrary {links [], nodes []}
 RigVMFunction_Sequence_1_ContainedGraph {nodes [
   n1 = RigVMFunctionEntryNode()
   RigVMFunction_Sequence = &RigVMFunction_Sequence(ExecuteContext=@n1.ExecuteContext)
   n2 = &RigVMFunction_Sequence(ExecuteContext=@RigVMFunction_Sequence.B)
   n3 = &RigVMFunction_Sequence(ExecuteContext=@n2.B)
   RigVMFunctionReturnNode(A=@RigVMFunction_Sequence.A, B=@n2.A, C=@n3.A, D=@n3.B)
  ]}
 RigVMModel {nodes [
   n1 = &RigVMFunction_MathQuaternionToEuler(
    RotationOrder=ZYX
    Value=&RigUnit_GetRelativeTransformForItem(Child=(Type=Bone,Name="head"), Parent=(Type=Bone,Name="spine_05")).RelativeTransform.Rotation
   )
   n2 = &RigVMFunction_MathFloatRemap(SourceMaximum=-43, SourceMinimum=0, TargetMinimum=0, bClamp=true, Value=@n1.Result.Z)
   &RigUnit_ModifyTransforms(
    ExecutePin=()
    ItemToModify=((Item=(Type=Bone,Name="dyn_head"),Transform=(Rotation=(X=0,Y=0.1736,Z=0,W=0.9848),Translation=(X=-7,Y=0,Z=0),Scale3D=(X=1,Y=1,Z=1))))
    WeightMinimum=0
    Weight=@n2.Result
   )
   > &RigUnit_ModifyTransforms(
    ItemToModify=((Item=(Type=Bone,Name="dyn_hairBangs_H_MSTR"),Transform=(Rotation=(X=0,Y=0,Z=-0.4226,W=0.9063),Translation=(X=0,Y=0,Z=0),Scale3D=(X=1,Y=1,Z=1))))
    WeightMinimum=0
    Weight=@n2.Result
   )
   > &RigUnit_ModifyTransforms(
    ItemToModify=((Item=(Type=Bone,Name="dyn_hairBangs_A_MSTR"),Transform=(Rotation=(X=0,Y=0,Z=0.4226,W=0.9063),Translation=(X=0,Y=0,Z=0),Scale3D=(X=1,Y=1,Z=1))))
    WeightMinimum=0
    Weight=@n2.Result
   )
   RigUnit_BeginExecution = &RigUnit_BeginExecution()
   RigVMFunction_Sequence = RigVMAggregateNode(ExecuteContext=@RigUnit_BeginExecution.ExecutePin)
   SequenceExecution_2 = RigVMAggregateNode(ExecuteContext=@RigVMFunction_Sequence.A)
   RigVMVariableNode(
    Value=&RigVMFunction_MathFloatMul(A=RigVMVariableNode() {variable=head_SPR_look_up_alpha}.Value, B=RigVMVariableNode() {variable=upperarm_r_rifle_up_alpha}.Value) {template="Multiply::Execute(in A,in B,out Result)"}.Result
    ExecuteContext=@SequenceExecution_2.E
   ) {variable=look_up_two_hands_rifle_multiplied_alpha}
   > RigVMVariableNode(Value=&RigVMFunction_MathFloatMul(B=RigVMVariableNode() {variable=head_SPR_look_down_alpha}.Value, A=RigVMVariableNode() {variable=upperarm_r_rifle_up_alpha}.Value) {template="Multiply::Execute(in A,in B,out Result)"}.Result) {variable=look_down_two_hands_rifle_multiplied_alpha_0}
   n3 = &RigUnit_SphericalPoseReader(
    ActiveRegionScaleFactors=(PositiveWidth=4,NegativeWidth=1,PositiveHeight=3,NegativeHeight=3)
    ActiveRegionSize=0.07
    Debug=(bDrawDebug=false,bDraw2D=False,bDrawLocalAxes=false,DebugScale=25,DebugSegments=20,DebugThickness=0.25)
    DriverAxis=(X=1,Y=0,Z=0)
    DriverItem=(Type=Bone,Name="head")
    FalloffRegionScaleFactors=(PositiveWidth=1,NegativeWidth=0.8,PositiveHeight=1,NegativeHeight=1)
    FalloffSize=0.4
    FlipHeightScaling=False
    FlipWidthScaling=False
    OptionalParentItem=(Type=Bone,Name="spine_05")
    RotationOffset=(X=70,Y=0,Z=-150)
    ExecutePin=@SequenceExecution_2.D
   )
   > RigVMVariableNode(Value=@n3.OutputParam) {variable=head_SPR_look_up_alpha}
   > n4 = &RigUnit_SphericalPoseReader(
    ActiveRegionScaleFactors=(PositiveWidth=2,NegativeWidth=0.5,PositiveHeight=2,NegativeHeight=2)
    ActiveRegionSize=0.075
    Debug=(bDrawDebug=false,DebugScale=25,DebugSegments=20,DebugThickness=0.25)
    DriverAxis=(X=-1,Y=0,Z=0)
    DriverItem=(Type=Bone,Name="clavicle_r")
    FalloffRegionScaleFactors=(PositiveWidth=1,NegativeWidth=0.4,PositiveHeight=1,NegativeHeight=1)
    FalloffSize=0.2
    FlipHeightScaling=false
    FlipWidthScaling=true
    OptionalParentItem=(Type=Bone,Name="spine_05")
    RotationOffset=(X=-10,Y=120,Z=0)
   )
   > RigVMVariableNode(Value=@n4.OutputParam) {variable=Clavicle r up}
   > n5 = &RigUnit_SphericalPoseReader(
    ActiveRegionScaleFactors=(PositiveWidth=1,NegativeWidth=1,PositiveHeight=1,NegativeHeight=1)
    ActiveRegionSize=0.065
    Debug=(bDrawDebug=false,DebugScale=25,DebugSegments=20,DebugThickness=0.25)
    DriverAxis=(X=-1,Y=0,Z=0)
    DriverItem=(Type=Bone,Name="upperarm_r")
    FalloffRegionScaleFactors=(PositiveWidth=1,NegativeWidth=1,PositiveHeight=1,NegativeHeight=1)
    FalloffSize=0.125
    OptionalParentItem=(Type=Bone)
    RotationOffset=(X=60,Y=0,Z=65)
   )
   > RigVMVariableNode(Value=@n5.OutputParam) {variable=upperarm_r_SPR_back_alpha}
   > n6 = &RigUnit_SphericalPoseReader(
    ActiveRegionScaleFactors=(PositiveWidth=1,NegativeWidth=1,PositiveHeight=1,NegativeHeight=1)
    ActiveRegionSize=0.075
    Debug=(bDrawDebug=false,DebugScale=25,DebugSegments=20,DebugThickness=0.25)
    DriverAxis=(X=-1,Y=0,Z=0)
    DriverItem=(Type=Bone,Name="upperarm_r")
    FalloffRegionScaleFactors=(PositiveWidth=1,NegativeWidth=1,PositiveHeight=1,NegativeHeight=1)
    FalloffSize=0.2
    OptionalParentItem=(Type=Bone)
    RotationOffset=(X=-45,Y=15,Z=-20)
   )
   > RigVMVariableNode(Value=@n6.OutputParam) {variable=upperarm_r_rifle_up_alpha}
   > n7 = &RigUnit_SphericalPoseReader(
    ActiveRegionScaleFactors=(PositiveWidth=2.5,NegativeWidth=3,PositiveHeight=3,NegativeHeight=3)
    ActiveRegionSize=0.1
    Debug=(bDrawDebug=false,DebugScale=25,DebugSegments=20,DebugThickness=0.25)
    DriverAxis=(X=-1,Y=0,Z=0)
    DriverItem=(Type=Bone,Name="upperarm_r")
    FalloffRegionScaleFactors=(PositiveWidth=1,NegativeWidth=1,PositiveHeight=1,NegativeHeight=1)
    FalloffSize=0.5
    FlipHeightScaling=true
    OptionalParentItem=(Type=Bone)
    RotationOffset=(X=180,Y=0,Z=180)
   )
   > RigVMVariableNode(Value=@n7.OutputParam) {variable=upperarm_r_up}
   > n8 = &RigUnit_SphericalPoseReader(
    ActiveRegionScaleFactors=(PositiveWidth=4,NegativeWidth=1,PositiveHeight=3,NegativeHeight=3)
    ActiveRegionSize=0.07
    Debug=(bDrawDebug=true,bDraw2D=False,bDrawLocalAxes=false,DebugScale=25,DebugSegments=20,DebugThickness=0.25)
    DriverAxis=(X=1,Y=0,Z=0)
    DriverItem=(Type=Bone,Name="head")
    FalloffRegionScaleFactors=(PositiveWidth=1,NegativeWidth=0.5,PositiveHeight=1,NegativeHeight=1)
    FalloffSize=0.4
    FlipHeightScaling=false
    FlipWidthScaling=true
    OptionalParentItem=(Type=Bone,Name="spine_05")
    RotationOffset=(X=-80,Y=0,Z=-10)
   )
   > RigVMVariableNode(Value=@n8.OutputParam) {variable=head_SPR_look_up_extreme_alpha}
   n9 = &RigUnit_SphericalPoseReader(
    ActiveRegionScaleFactors=(PositiveWidth=2,NegativeWidth=2,PositiveHeight=2.4,NegativeHeight=0.5)
    ActiveRegionSize=0.1
    Debug=(bDrawDebug=false,DebugScale=25,DebugSegments=20,DebugThickness=0.25)
    DriverAxis=(X=0,Y=1,Z=0)
    DriverItem=(Type=Bone,Name="head")
    FalloffRegionScaleFactors=(PositiveWidth=1,NegativeWidth=1,PositiveHeight=1,NegativeHeight=0.4)
    FalloffSize=0.25
    OptionalParentItem=(Type=Bone)
    RotationOffset=(X=120,Y=0,Z=0)
    ExecutePin=@SequenceExecution_2.B
   )
   > RigVMVariableNode(Value=@n9.OutputParam) {variable=head_SPR_look_left_alpha}
   n10 = &RigUnit_SphericalPoseReader(
    ActiveRegionScaleFactors=(PositiveWidth=2,NegativeWidth=2,PositiveHeight=3,NegativeHeight=0.5)
    ActiveRegionSize=0.1
    Debug=(bDrawDebug=false,DebugScale=25,DebugSegments=20,DebugThickness=0.25)
    DriverAxis=(X=0,Y=1,Z=0)
    DriverItem=(Type=Bone,Name="head")
    FalloffRegionScaleFactors=(PositiveWidth=1,NegativeWidth=1,PositiveHeight=0,NegativeHeight=0.4)
    FalloffSize=0.27
    FlipHeightScaling=true
    FlipWidthScaling=false
    OptionalParentItem=(Type=Bone)
    RotationOffset=(X=60,Y=0,Z=0)
    ExecutePin=@SequenceExecution_2.C
   )
   > RigVMVariableNode(Value=@n10.OutputParam) {variable=head_SPR_look_right_alpha}
   n11 = &RigUnit_SphericalPoseReader(
    ActiveRegionScaleFactors=(PositiveWidth=2,NegativeWidth=0.5,PositiveHeight=2,NegativeHeight=2)
    ActiveRegionSize=0.075
    Debug=(bDrawDebug=true,DebugScale=25,DebugSegments=20,DebugThickness=0.25)
    DriverAxis=(X=0,Y=1,Z=0)
    DriverItem=(Type=Bone,Name="head")
    FalloffRegionScaleFactors=(PositiveWidth=1,NegativeWidth=0.4,PositiveHeight=1,NegativeHeight=1)
    FalloffSize=0.2
    FlipHeightScaling=false
    FlipWidthScaling=true
    OptionalParentItem=(Type=Bone)
    RotationOffset=(X=90,Y=0,Z=12.5)
    ExecutePin=@SequenceExecution_2.A
   )
   > RigVMVariableNode(Value=@n11.OutputParam) {variable=head_SPR_look_down_alpha}
   n12 = RigVMVariableNode() {variable=head_SPR_look_up_alpha}
   &RigUnit_ModifyTransforms(
    ItemToModify=((Item=(Type=Bone,Name="dyn_hair_bk_00"),Transform=(Rotation=(X=0,Y=0,Z=0,W=1),Translation=(X=2,Y=0,Z=4),Scale3D=(X=1,Y=1,Z=1))))
    WeightMinimum=0
    ExecutePin=@RigVMFunction_Sequence.B
    Weight=@n12.Value
   )
   > &RigUnit_ModifyTransforms(
    ItemToModify=((Item=(Type=Bone,Name="dyn_hair_bk_02"),Transform=(Rotation=(X=0,Y=0.0872,Z=0,W=0.9962),Translation=(X=0,Y=0,Z=0),Scale3D=(X=1,Y=1,Z=1))))
    WeightMinimum=0
    Weight=@n12.Value
   )
   &RigUnit_ParentConstraint(
    AdvancedSettings=(InterpolationType=Average,RotationOrderForFilter=XZY)
    Child=(Type=Bone,Name="dyn_hair_l_00")
    Parents=((Item=(Type=Bone,Name="head"),Weight=0.3),(Item=(Type=Bone,Name="dyn_hair_bk_00"),Weight=0.7))
    Weight=1
    ExecutePin=@RigVMFunction_Sequence.C
   )
   > &RigUnit_ParentConstraint(
    AdvancedSettings=(InterpolationType=Average,RotationOrderForFilter=XZY)
    Child=(Type=Bone,Name="dyn_hair_l_01")
    Parents=((Item=(Type=Bone,Name="head"),Weight=0.3),(Item=(Type=Bone,Name="dyn_hair_bk_01"),Weight=0.7))
    Weight=1
   )
   > &RigUnit_ParentConstraint(
    AdvancedSettings=(InterpolationType=Average,RotationOrderForFilter=XZY)
    Child=(Type=Bone,Name="dyn_hair_l_02")
    Parents=((Item=(Type=Bone,Name="clavicle_l"),Weight=0.5),(Item=(Type=Bone,Name="dyn_hair_bk_02"),Weight=0.5))
    Weight=1
   )
   > &RigUnit_ParentConstraint(
    AdvancedSettings=(InterpolationType=Average,RotationOrderForFilter=XZY)
    Child=(Type=Bone,Name="dyn_hair_l_03")
    Parents=((Item=(Type=Bone,Name="clavicle_l"),Weight=0.5),(Item=(Type=Bone,Name="dyn_hair_bk_03"),Weight=0.5))
    Weight=1
   )
   > &RigUnit_ParentConstraint(
    AdvancedSettings=(InterpolationType=Average,RotationOrderForFilter=XZY)
    Child=(Type=Bone,Name="dyn_hair_l_04")
    Parents=((Item=(Type=Bone,Name="clavicle_l"),Weight=0.3),(Item=(Type=Bone,Name="dyn_hair_bk_04"),Weight=0.7))
    Weight=1
   )
   > &RigUnit_ParentConstraint(
    AdvancedSettings=(InterpolationType=Average,RotationOrderForFilter=XZY)
    Child=(Type=Bone,Name="dyn_hair_r_00")
    Parents=((Item=(Type=Bone,Name="head"),Weight=0.3),(Item=(Type=Bone,Name="dyn_hair_bk_00"),Weight=0.7))
    Weight=1
   )
   > &RigUnit_ParentConstraint(
    AdvancedSettings=(InterpolationType=Average,RotationOrderForFilter=XZY)
    Child=(Type=Bone,Name="dyn_hair_r_01")
    Parents=((Item=(Type=Bone,Name="head"),Weight=0.3),(Item=(Type=Bone,Name="dyn_hair_bk_01"),Weight=0.7))
    Weight=1
   )
   > &RigUnit_ParentConstraint(
    AdvancedSettings=(InterpolationType=Average,RotationOrderForFilter=XZY)
    Child=(Type=Bone,Name="dyn_hair_r_02")
    Parents=((Item=(Type=Bone,Name="clavicle_r"),Weight=0.5),(Item=(Type=Bone,Name="dyn_hair_bk_02"),Weight=0.5))
    Weight=1
   )
   > &RigUnit_ParentConstraint(
    AdvancedSettings=(InterpolationType=Average,RotationOrderForFilter=XZY)
    Child=(Type=Bone,Name="dyn_hair_r_03")
    Parents=((Item=(Type=Bone,Name="clavicle_r"),Weight=0.5),(Item=(Type=Bone,Name="dyn_hair_bk_03"),Weight=0.5))
    Weight=1
   )
   > &RigUnit_ParentConstraint(
    AdvancedSettings=(InterpolationType=Average,RotationOrderForFilter=XZY)
    Child=(Type=Bone,Name="dyn_hair_r_04")
    Parents=((Item=(Type=Bone,Name="clavicle_r"),Weight=0.3),(Item=(Type=Bone,Name="dyn_hair_bk_04"),Weight=0.7))
    Weight=1
   )
   > &RigUnit_ParentConstraint(
    AdvancedSettings=(InterpolationType=Average,RotationOrderForFilter=XZY)
    Child=(Type=Bone,Name="dyn_hairBangs_B_01")
    Parents=((Item=(Type=Bone,Name="head"),Weight=0.5),(Item=(Type=Bone,Name="spine_05"),Weight=0.5))
    Weight=1
   )
   > &RigUnit_ParentConstraint(
    AdvancedSettings=(InterpolationType=Average,RotationOrderForFilter=XZY)
    Child=(Type=Bone,Name="dyn_hairBangs_H_01")
    Parents=((Item=(Type=Bone,Name="head"),Weight=0.6),(Item=(Type=Bone,Name="spine_05"),Weight=0.4))
    Weight=1
   )
   > &RigUnit_ParentConstraint(
    AdvancedSettings=(InterpolationType=Average,RotationOrderForFilter=XZY)
    Child=(Type=Bone,Name="dyn_hairBangs_A_01")
    Parents=((Item=(Type=Bone,Name="clavicle_r"),Weight=0.3),(Item=(Type=Bone,Name="spine_05"),Weight=0.2),(Item=(Type=Bone,Name="head"),Weight=0.5))
    Weight=1
   )
   > &RigUnit_ParentConstraint(
    AdvancedSettings=(InterpolationType=Average,RotationOrderForFilter=XZY)
    Child=(Type=Bone,Name="dyn_hairBangs_H_03")
    Parents=((Item=(Type=Bone,Name="clavicle_l"),Weight=0.3),(Item=(Type=Bone,Name="spine_05"),Weight=0.3),(Item=(Type=Bone,Name="head"),Weight=0.4))
    Weight=1
   )
   &RigUnit_ModifyTransforms(
    ItemToModify=((Item=(Type=Bone,Name="dyn_hair_r_00"),Transform=(Rotation=(X=0,Y=-0.0175,Z=0,W=0.9998),Translation=(X=-1,Y=0,Z=1),Scale3D=(X=1,Y=1,Z=1))))
    WeightMinimum=0
    ExecutePin=@RigVMFunction_Sequence.D
    Weight=RigVMVariableNode() {variable=Clavicle r up}.Value
   )
   > &RigUnit_ModifyTransforms(
    ItemToModify=((Item=(Type=Bone,Name="dyn_hair_r_00"),Transform=(Rotation=(X=0,Y=-0.0436,Z=0,W=0.999),Translation=(X=0,Y=0,Z=0),Scale3D=(X=1,Y=1,Z=1))))
    WeightMinimum=0
    Weight=RigVMVariableNode() {variable=head_SPR_look_left_alpha}.Value
   )
   > &RigUnit_ModifyTransforms(
    ItemToModify="((Item=(Type=Bone,Name=\"dyn_hair_r_00\"),Transform=(Rotation=(X=0,Y=-0.1305,Z=0,W=0.9914),Translation=(X=-1,Y=0,Z=0),Scale3D=(X=1,Y=1,Z=1))),(Item=(Type=Bone,Name=\"dyn_hair_r_02\"),Transform=(Rotation=(X=0,Y=0.2588,Z=0,W=0.9659),Translation=(X=0,Y=0,Z=1),Sca ...(truncated)"
    WeightMinimum=0
    Weight=RigVMVariableNode() {variable=upperarm_r_up}.Value
   )
   > &RigUnit_ModifyTransforms(
    ItemToModify="((Item=(Type=Bone,Name=\"dyn_hair_r_00\"),Transform=(Rotation=(X=0,Y=-0.1305,Z=0,W=0.9914),Translation=(X=-1,Y=0,Z=0),Scale3D=(X=1,Y=1,Z=1))),(Item=(Type=Bone,Name=\"dyn_hair_r_02\"),Transform=(Rotation=(X=0,Y=0.2164,Z=0,W=0.9763),Translation=(X=0,Y=0,Z=1),Sca ...(truncated)"
    WeightMinimum=0
    Weight=RigVMVariableNode() {variable=look_up_two_hands_rifle_multiplied_alpha}.Value
   )
   n13 = RigVMVariableNode() {variable=head_SPR_look_up_extreme_alpha}
   > &RigUnit_ModifyTransforms(
    ItemToModify="((Item=(Type=Bone,Name=\"dyn_hair_bk_00\"),Transform=(Rotation=(X=0,Y=-0.1305,Z=0,W=0.9914),Translation=(X=-1,Y=0,Z=0),Scale3D=(X=1,Y=1,Z=1))),(Item=(Type=Bone,Name=\"dyn_hair_bk_02\"),Transform=(Rotation=(X=0,Y=0.0436,Z=0,W=0.999),Translation=(X=0,Y=0,Z=1),Sc ...(truncated)"
    WeightMinimum=0
    Weight=@n13.Value
   )
   > &RigUnit_ModifyTransforms(
    ItemToModify="((Item=(Type=Bone,Name=\"dyn_hair_r_00\"),Transform=(Rotation=(X=0,Y=-0.1305,Z=0,W=0.9914),Translation=(X=-1,Y=0,Z=0),Scale3D=(X=1,Y=1,Z=1))),(Item=(Type=Bone,Name=\"dyn_hair_r_02\"),Transform=(Rotation=(X=0,Y=0.0436,Z=0,W=0.999),Translation=(X=0,Y=0,Z=1),Scal ...(truncated)"
    WeightMinimum=0
    Weight=@n13.Value
   )
   &RigUnit_RotationConstraintLocalSpaceOffset(
    AdvancedSettings=(RotationOrderForFilter=XZY)
    Child=(Type=Bone,Name="dyn_hair_l_MSTR")
    ExecutePin=()
    Filter=(bX=True,bY=True,bZ=True)
    Parents=((Item=(Type=Bone,Name="spine_05"),Weight=0.5),(Item=(Type=Bone,Name="head"),Weight=0.5))
    Weight=1
   )
   > &RigUnit_RotationConstraintLocalSpaceOffset(
    AdvancedSettings=(RotationOrderForFilter=XZY)
    Child=(Type=Bone,Name="dyn_hair_bk_MSTR")
    Filter=(bX=True,bY=True,bZ=True)
    Parents=((Item=(Type=Bone,Name="spine_05"),Weight=0.5),(Item=(Type=Bone,Name="head"),Weight=0.5))
    Weight=1
   )
   > &RigUnit_RotationConstraintLocalSpaceOffset(
    AdvancedSettings=(RotationOrderForFilter=XZY)
    Child=(Type=Bone,Name="dyn_hair_r_MSTR")
    Filter=(bX=True,bY=True,bZ=True)
    Parents=((Item=(Type=Bone,Name="spine_05"),Weight=0.5),(Item=(Type=Bone,Name="head"),Weight=0.5))
    Weight=1
   )
   > n14 = &RigUnit_SphericalPoseReader(
    ActiveRegionScaleFactors=(PositiveWidth=1,NegativeWidth=1,PositiveHeight=1,NegativeHeight=1)
    ActiveRegionSize=0.1
    Debug=(bDrawDebug=false,DebugScale=25,DebugSegments=20,DebugThickness=0.25)
    DriverAxis=(X=0,Y=-1,Z=0)
    DriverItem=(Type=Bone,Name="head")
    FalloffRegionScaleFactors=(PositiveWidth=2,NegativeWidth=2,PositiveHeight=2,NegativeHeight=2)
    FalloffSize=0.2
    FlipWidthScaling=false
    OptionalParentItem=(Type=Bone,Name="neck_01")
    RotationOffset=(X=-90,Y=0,Z=-55)
   )
   n15 = &RigVMFunction_MathDoubleMul(B=1, A=@n14.OutputParam)
   > &RigUnit_ModifyTransforms(
    ItemToModify=((Item=(Type=Bone,Name="dyn_hair_l_MSTR"),Transform=(Rotation=(X=0,Y=0,Z=0,W=1),Translation=(X=0,Y=0,Z=2.5),Scale3D=(X=1,Y=1,Z=1))))
    ItemToModify.0.Transform.Rotation=&RigVMFunction_MathQuaternionFromEuler(Euler=(X=0,Y=-35,Z=-6)).Result
    Weight=@n15.Result
   )
   > &RigUnit_ModifyTransforms(
    ItemToModify=((Item=(Type=Bone,Name="dyn_hair_r_MSTR"),Transform=(Rotation=(X=0,Y=0,Z=0,W=1),Translation=(X=0,Y=0,Z=2.5),Scale3D=(X=1,Y=1,Z=1))))
    ItemToModify.0.Transform.Rotation=&RigVMFunction_MathQuaternionFromEuler(Euler=(X=0,Y=-35,Z=-6)).Result
    Weight=@n15.Result
   )
   > &RigUnit_ModifyTransforms(
    ItemToModify=((Item=(Type=Bone,Name="dyn_hair_bk_MSTR"),Transform=(Rotation=(X=0,Y=0,Z=0,W=1),Translation=(X=0,Y=0,Z=2.5),Scale3D=(X=1,Y=1,Z=1))))
    ItemToModify.0.Transform.Rotation=&RigVMFunction_MathQuaternionFromEuler(Euler=(X=0,Y=-35,Z=-4)).Result
    Weight=@n15.Result
   )
   > n16 = &RigUnit_SphericalPoseReader(
    ActiveRegionScaleFactors=(PositiveWidth=0.25,NegativeWidth=0.25,PositiveHeight=0.25,NegativeHeight=0.25)
    ActiveRegionSize=0.1
    Debug=(bDrawDebug=false,DebugScale=25,DebugSegments=20,DebugThickness=0.25)
    DriverAxis=(X=0,Y=1,Z=0)
    DriverItem=(Type=Bone,Name="head")
    FalloffRegionScaleFactors=(PositiveWidth=0.75,NegativeWidth=0.5,PositiveHeight=4,NegativeHeight=4)
    FalloffSize=0.2
    OptionalParentItem=(Type=Bone,Name="spine_05")
    RotationOffset=(X=90,Y=0,Z=25)
   )
   > &RigUnit_ModifyTransforms(
    ItemToModify=((Item=(Type=Bone,Name="dyn_hair_bk_MSTR"),Transform=(Rotation=(X=0,Y=0,Z=0,W=1),Translation=(X=0,Y=0,Z=0),Scale3D=(X=1,Y=1,Z=1))))
    ItemToModify.0.Transform.Rotation=&RigVMFunction_MathQuaternionFromEuler(Euler=(X=0,Y=-20,Z=-2)).Result
    Weight=@n16.OutputParam
   )
   > &RigUnit_ModifyTransforms(
    ItemToModify=((Item=(Type=Bone,Name="dyn_hair_l_MSTR"),Transform=(Rotation=(X=0,Y=0,Z=0,W=1),Translation=(X=0,Y=0,Z=0),Scale3D=(X=1,Y=1,Z=1))))
    ItemToModify.0.Transform.Rotation=&RigVMFunction_MathQuaternionFromEuler(Euler=(X=0,Y=-20,Z=-3)).Result
    Weight=@n16.OutputParam
   )
   > &RigUnit_ModifyTransforms(
    ItemToModify=((Item=(Type=Bone,Name="dyn_hair_r_MSTR"),Transform=(Rotation=(X=0,Y=0,Z=0,W=1),Translation=(X=0,Y=0,Z=0),Scale3D=(X=1,Y=1,Z=1))))
    ItemToModify.0.Transform.Rotation=&RigVMFunction_MathQuaternionFromEuler(Euler=(X=0,Y=-20,Z=3)).Result
    Weight=@n16.OutputParam
   )
   FN_DeformRig_FNC = RigVMFunctionReferenceNode(
    DeformRig_ Curve="(deformRig_l_upperarm_up_alpha_72_#1=False,deformRig_r_upperarm_up_alpha_73_#2=False,deformRig_l_upperarm_fr_alpha_74_#3=False,deformRig_r_upperarm_fr_alpha_75_CB51B8 ...(truncated)"
    DeformRig_Bool="(upperarm_1_#4=True,elbow_3_#5=True,wrist_7_#6=True,groin_8_#7=True,glute_17_#8=True,knee_9_40C408324CA261 ...(truncated)"
    DeformRig_Debug=(upperarm_debug_18_#4=False,wrist_debug_20_#6=False,lat_debug_26_#9=False,groinGlute_debug_27_#7=False)
    DeformRig_Value="(upperarm_up_18_#10=(Rotation=(X=0,Y=0,Z=0,W=1),Translation=(X=0,Y=0,Z=8),Scale3D=(X=0.7,Y=1,Z=1)),upperarm_fwd_fr_19_#11=(Rotation=(X=0,Y=0,Z=0,W=1),Translation=(X=0,Y=10,Z=0),Scale3D=(X=0.5,Y=1,Z= ...(truncated)"
    ExecuteContext=()
   )
   &RigVMFunction_MathFloatRemap(SourceMaximum=-56, TargetMinimum=0, bClamp=true, Value=@n1.Result.X)
   &RigVMFunction_MathFloatRemap(SourceMaximum=25, TargetMinimum=0, bClamp=true, Value=@n1.Result.X)
   &RigUnit_ModifyTransforms(
    ExecutePin=()
    ItemToModify="((Item=(Type=Bone,Name=\"dyn_hairFrontLeftAimTarget\"),Transform=(Rotation=(X=0,Y=0,Z=0,W=1),Translation=(X=0,Y=0,Z=0),Scale3D=(X=1,Y=1,Z=1))),(Item=(Type=Bone,Name=\"dyn_hairFrontRightAimTarget\"),Transform=(Rotation=(X=0,Y=0,Z=0,W=1),Translation=(X=0,Y=0,Z=0 ...(truncated)"
    Weight=@n2.Result
   )
   &RigUnit_ModifyTransforms(
    ExecutePin=()
    ItemToModify="((Item=(Type=Bone,Name=\"dyn_head\"),Transform=(Rotation=(X=0,Y=0,Z=0,W=1),Translation=(X=-10,Y=0,Z=0),Scale3D=(X=1,Y=1,Z=1))),(Item=(Type=Bone,Name=\"None\"),Transform=(Rotation=(X=0,Y=0,Z=0,W=1),Translation=(X=0,Y=0,Z=-4),Scale3D=(X=1,Y=1,Z=1))),(Item=(Type= ...(truncated)"
    Weight=1
   )
  ]}
]
hierarchy [
 root {loc=<0 0 0>, rot=<0 0 0 1>}
 pelvis {parent=root, loc=<0 -0.5301 98.0571>, rot=<0.0127 0.707 -0.0127 -0.707>}
 spine_01 {parent=pelvis, loc=<6.1208 -0.9328 0>, rot=<0 0 0.034 -0.9994>}
 spine_02 {parent=spine_01, loc=<6.5996 -0.2095 0>, rot=<0 0 0.0238 -0.9997>}
 spine_03 {parent=spine_02, loc=<6.4914 -0.1237 0>, rot=<0 0 0.0149 -0.9999>}
 spine_04 {parent=spine_03, loc=<6.9598 0.0794 0>, rot=<0 0 -0.0132 -0.9999>}
 spine_05 {parent=spine_04, loc=<6.9987 0 0>, rot=<0 0 -0.0015 -1>}
 neck_01 {parent=spine_05, loc=<11.3138 0.1529 0>, rot=<0 0 -0.2003 -0.9797>}
 neck_02 {parent=neck_01, loc=<4.5627 0 0>, rot=<0 0 0.0504 -0.9987>}
 head {parent=neck_02, loc=<5.3914 0 0>, rot=<0 0 0.115 -0.9934>}
 dyn_head {parent=head, loc=<0 0 0>, rot=<0 0 0 -1>}
 dyn_hair_bk_MSTR {parent=dyn_head, loc=<11.1222 -3.6765 0>, rot=<0.2152 -0.6736 0.6736 -0.2152>}
 dyn_hair_bk_00 {parent=dyn_hair_bk_MSTR, loc=<0.5 0 0>, rot=<0 -0.0572 0 -0.9984>}
 dyn_hair_bk_01 {parent=dyn_hair_bk_00, loc=<6.0938 0 0>, rot=<0 -0.0272 0 -0.9996>}
 dyn_hair_bk_02 {parent=dyn_hair_bk_01, loc=<8.7854 0 0>, rot=<0 0.0249 0 -0.9997>}
 dyn_hair_bk_03 {parent=dyn_hair_bk_02, loc=<9.4823 0 0>, rot=<0 -0.1962 0 -0.9806>}
 dyn_hair_bk_04 {parent=dyn_hair_bk_03, loc=<10.3536 0 0>, rot=<0 -0.0751 0 -0.9972>}
 dyn_flyOff_03 {parent=dyn_hair_bk_02, loc=<5.2846 3.8449 0.9109>, rot=<0.0521 0.2261 0.2171 -0.9482>}
 dyn_hair_l_MSTR {parent=dyn_head, loc=<8.8541 0.6916 -6.499>, rot=<0.0522 -0.9061 0.4053 -0.1098>}
 dyn_hair_l_00 {parent=dyn_hair_l_MSTR, loc=<0.5 0 0>, rot=<-0.0873 0.06 0.0268 -0.994>}
 dyn_hair_l_01 {parent=dyn_hair_l_00, loc=<7.0975 0 0>, rot=<-0.1042 0.1249 0.0548 -0.9852>}
 dyn_hair_l_02 {parent=dyn_hair_l_01, loc=<6.5937 0 0>, rot=<-0.023 0.0339 0.0188 -0.999>}
 dyn_hair_l_03 {parent=dyn_hair_l_02, loc=<9.0914 0 0>, rot=<-0.0078 -0.1464 0.0622 -0.9872>}
 dyn_hair_l_04 {parent=dyn_hair_l_03, loc=<8.993 0 0>, rot=<0.0227 -0.1923 0.0255 -0.9808>}
 dyn_hair_r_MSTR {parent=dyn_head, loc=<10.0791 -1.0364 6.0327>, rot=<0.2033 -0.5878 0.7824 -0.0323>}
 dyn_hair_r_00 {parent=dyn_hair_r_MSTR, loc=<0.5 0 0>, rot=<-0.0006 0.0016 -0.0049 -1>}
 dyn_hair_r_01 {parent=dyn_hair_r_00, loc=<6.1151 0 0>, rot=<-0.0678 0.0612 -0.0691 -0.9934>}
 dyn_hair_r_02 {parent=dyn_hair_r_01, loc=<7.2081 0 0>, rot=<0.106 0.0301 -0.0614 -0.992>}
 dyn_hair_r_03 {parent=dyn_hair_r_02, loc=<7.99 0 0>, rot=<0.0129 -0.0997 -0.0063 -0.9949>}
 dyn_hair_r_04 {parent=dyn_hair_r_03, loc=<8.1126 0 0>, rot=<0.0829 -0.1187 -0.002 -0.9895>}
 dyn_hair_r_flyoff {parent=dyn_hair_r_02, loc=<0.5048 -0.8152 0.13>, rot=<-0.5128 0.0218 0.3035 -0.8028>}
 dyn_hairBangs_G_MSTR {parent=head, loc=<4.2297 2.833 -6.1629>, rot=<0.1982 -0.9726 -0.1188 -0.025>}
 dyn_hairBangs_G_01 {parent=dyn_hairBangs_G_MSTR, loc=<0 0 0>, rot=<0 0 0 1>}
 dyn_hairBangs_G_02 {parent=dyn_hairBangs_G_01, loc=<4.044 0.2336 0.1152>, rot=<0.0305 -0.0376 -0.3586 -0.9322>}
 dyn_hairBangs_G_03 {parent=dyn_hairBangs_G_02, loc=<2.7697 0 0>, rot=<0.0352 -0.0538 -0.5846 -0.8088>}
 dyn_hairBangs_F_MSTR {parent=head, loc=<7.9047 8.3639 -3.1326>, rot=<0.323 -0.8531 -0.3355 -0.2351>}
 dyn_hairBangs_F_01 {parent=dyn_hairBangs_F_MSTR, loc=<0 0 0>, rot=<0 0 0 1>}
 dyn_hairBangs_F_02 {parent=dyn_hairBangs_F_01, loc=<5.5744 -0.1416 -0.6227>, rot=<0.0017 -0.2244 0.0491 -0.9733>}
 dyn_hairBangs_F_03 {parent=dyn_hairBangs_F_02, loc=<5.1388 -0.4892 1.0753>, rot=<-0.0193 0.0079 0.0396 -0.999>}
 dyn_hairBangs_E_MSTR {parent=head, loc=<8.9159 9.4183 1.891>, rot=<0.0365 0.6088 0.7726 0.1762>}
 dyn_hairBangs_E_01 {parent=dyn_hairBangs_E_MSTR, loc=<0 0 0>, rot=<0 0 0 1>}
 dyn_hairBangs_E_02 {parent=dyn_hairBangs_E_01, loc=<3.5116 0 0>, rot=<-0.0018 -0.2165 0.1848 -0.9586>}
 dyn_hairBangs_E_03 {parent=dyn_hairBangs_E_02, loc=<3.0514 0 0>, rot=<-0.1878 -0.1575 0.1454 -0.9585>}
 dyn_hairBangs_D_MSTR {parent=head, loc=<6.9037 6.7601 5.3586>, rot=<0.0806 0.1361 0.9866 -0.0407>}
 dyn_hairBangs_D_01 {parent=dyn_hairBangs_D_MSTR, loc=<0 0 0>, rot=<0 0 0 1>}
 dyn_hairBangs_D_02 {parent=dyn_hairBangs_D_01, loc=<4.0807 0 -0.1774>, rot=<0 -0.0761 0 -0.9971>}
 dyn_hairBangs_C_MSTR {parent=head, loc=<1.5865 3.981 7.087>, rot=<-0.0497 0.2624 0.9636 0.0137>}
 dyn_hairBangs_C_01 {parent=dyn_hairBangs_C_MSTR, loc=<0 0 0>, rot=<0 0 0 1>}
 dyn_hairBangs_C_02 {parent=dyn_hairBangs_C_01, loc=<3.6432 0 0>, rot=<-0.0398 -0.0137 0.3004 -0.9529>}
 dyn_hairBangs_B_MSTR {parent=head, loc=<-0.7344 1.2747 7.7689>, rot=<0.0448 0.0856 0.9903 0.0995>}
 dyn_hairBangs_B_01 {parent=dyn_hairBangs_B_MSTR, loc=<0 0 0>, rot=<0 0 0 1>}
 dyn_hairBangs_B_02 {parent=dyn_hairBangs_B_01, loc=<4.743 0.0513 0.2162>, rot=<0.0283 -0.2092 0.2447 -0.9463>}
 dyn_hairBangs_B_03 {parent=dyn_hairBangs_B_01, loc=<4.6963 -2.4117 -0.9989>, rot=<-0.0554 -0.1057 0.459 -0.8804>}
 dyn_hairBangs_A_MSTR {parent=head, loc=<-2.9367 -0.685 8.7978>, rot=<-0.0595 0.0255 0.9976 0.0259>}
 dyn_hairBangs_A_01 {parent=dyn_hairBangs_A_MSTR, loc=<0 0 0>, rot=<0 0 0 1>}
 dyn_hairBangs_A_02 {parent=dyn_hairBangs_A_01, loc=<6.0978 0 0>, rot=<-0.0008 -0.3383 -0.0092 -0.941>}
 dyn_hairBangs_H_MSTR {parent=head, loc=<-0.992 -0.8473 -8.2913>, rot=<-0.2494 -0.9142 -0.3133 0.062>}
 dyn_hairBangs_H_01 {parent=dyn_hairBangs_H_MSTR, loc=<0 0 0>, rot=<0 0 0 1>}
 dyn_hairBangs_H_02 {parent=dyn_hairBangs_H_01, loc=<4.6343 -0.3546 0.1502>, rot=<-0.0812 0.0846 -0.4374 -0.8916>}
 dyn_hairBangs_H_03 {parent=dyn_hairBangs_H_01, loc=<2.2095 -3.1831 -0.618>, rot=<-0.2076 -0.0075 0.0544 -0.9767>}
 dyn_hairBangs_H_04 {parent=dyn_hairBangs_H_03, loc=<3.6776 0 0>, rot=<0.0205 -0.2763 0.1759 -0.9446>}
 dyn_flyOff_01 {parent=head, loc=<-1.3587 -7.6844 9.2577>, rot=<0.3837 -0.2071 0.8243 -0.3611>}
 dyn_flyOff_02 {parent=head, loc=<1.8248 -9.3593 5.886>, rot=<0.2969 -0.516 0.7549 -0.275>}
 clavicle_l {parent=spine_05, loc=<5.4831 -0.1053 -3.2946>, rot=<0.0465 -0.7289 -0.009 -0.6829>}
 upperarm_l {parent=clavicle_l, loc=<9.8658 -1.2136 -0.4105>, rot=<0.0164 -0.3245 -0.0269 -0.9454>}
 clavicle_r {parent=spine_05, loc=<5.4831 -0.1053 3.2946>, rot=<0.7289 0.0465 -0.6829 0.009>}
 upperarm_r {parent=clavicle_r, loc=<-9.8658 1.2136 0.4105>, rot=<0.0164 -0.3245 -0.0269 -0.9454>}
]
variables [
 Clavicle r up {type=double, default=0}
 head_SPR_look_down_alpha {type=double, default=0}
 head_SPR_look_left_alpha {type=double, default=0}
 head_SPR_look_right_alpha {type=double, default=0}
 head_SPR_look_up_alpha {type=double, default=0}
 head_SPR_look_up_extreme_alpha {type=double, default=0}
 look down clav up multiply {type=double, default=0}
 look_down_two_hands_rifle_multiplied_alpha_0 {type=double, default=0}
 look_up_two_hands_rifle_multiplied_alpha {type=double, default=0}
 upperarm_r_SPR_back_alpha {type=double, default=0}
 upperarm_r_rifle_up_alpha {type=double, default=0}
 upperarm_r_up {type=double, default=0}
]
