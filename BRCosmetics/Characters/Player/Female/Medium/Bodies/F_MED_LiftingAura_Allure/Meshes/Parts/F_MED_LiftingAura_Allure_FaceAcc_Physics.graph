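# Physics asset for the LiftingAura Allure female FaceAcc mesh, defining capsule bodies and constraints for its hair.
# Back, left and right five-bone hair chains hang from head-attached master bones, with limited twist and tighter swing on upper-middle links; each tip has an extra capsule.
# Head, neck, spine, clavicle and upper-arm capsules form the body volume, linked by free-angle constraints.
# Eight bangs master bones have only tiny capsules, free-rotating on the head.
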PhysicsAsset /BRCosmetics/Characters/Player/Female/Medium/Bodies/F_MED_LiftingAura_Allure/Meshes/Parts/F_MED_LiftingAura_Allure_FaceAcc_Physics
bodies {
 clavicle_l [{
   shapeType=Capsule
   length=10.076
   radius=9.924
   shapeName=clavicle_l_capsule
   center=<10.5688 3.6024 -6.7545>
   rotation=<rot 66.8729 -0.0074 -0.0083>
  }]
 clavicle_r [{
   shapeType=Capsule
   length=10.076
   radius=9.924
   shapeName=clavicle_r_capsule
   center=<-11.0957 -1.6195 6.2482>
   rotation=<rot 81.8729 -0.025 -0.0255>
  }]
 dyn_hairBangs_A_MSTR [{shapeType=Capsule, length=0.505, radius=0.505, shapeName=dyn_hairBangs_A_MSTR_capsule, rotation=<rot 90 0 0>}]
 dyn_hairBangs_B_MSTR [{shapeType=Capsule, length=0.505, radius=0.505, shapeName=dyn_hairBangs_B_MSTR_capsule, rotation=<rot 90 0 0>}]
 dyn_hairBangs_C_MSTR [{shapeType=Capsule, length=0.505, radius=0.505, shapeName=dyn_hairBangs_C_MSTR_capsule, rotation=<rot 90 0 0>}]
 dyn_hairBangs_D_MSTR [{shapeType=Capsule, length=0.505, radius=0.505, shapeName=dyn_hairBangs_D_MSTR_capsule, rotation=<rot 90 0 0>}]
 dyn_hairBangs_E_MSTR [{shapeType=Capsule, length=0.505, radius=0.505, shapeName=dyn_hairBangs_E_MSTR_capsule, rotation=<rot 90 0 0>}]
 dyn_hairBangs_F_MSTR [{shapeType=Capsule, length=0.505, radius=0.505, shapeName=dyn_hairBangs_F_MSTR_capsule, rotation=<rot 90 0 0>}]
 dyn_hairBangs_G_MSTR [{shapeType=Capsule, length=0.505, radius=0.505, shapeName=dyn_hairBangs_G_MSTR_capsule, rotation=<rot 90 0 0>}]
 dyn_hairBangs_H_MSTR [{shapeType=Capsule, length=0.505, radius=0.505, shapeName=dyn_hairBangs_H_MSTR_capsule, rotation=<rot 90 0 0>}]
 dyn_hair_bk_00 [{shapeType=Capsule, length=2.7364, radius=5.1567, shapeName=dyn_hair_bk_00_capsule, center=<5.0867 0 8>, rotation=<rot 90 0 0>}]
 dyn_hair_bk_01 [{shapeType=Capsule, length=4.2183, radius=5.7817, shapeName=dyn_hair_bk_01_capsule, center=<7.0292 0 5>, rotation=<rot 90 0 0>}]
 dyn_hair_bk_02 [{shapeType=Capsule, length=2.4351, radius=7.5649, shapeName=dyn_hair_bk_02_capsule, center=<7.0292 0 5>, rotation=<rot 90 0 0>}]
 dyn_hair_bk_03 [{shapeType=Capsule, length=2.4351, radius=7.5649, shapeName=dyn_hair_bk_03_capsule, center=<7.0292 0 5>, rotation=<rot 90 0 0>}]
 dyn_hair_bk_04 [
  {shapeType=Capsule, length=2.4351, radius=7.5649, shapeName=dyn_hair_bk_04_capsule, center=<7.0292 0 5>, rotation=<rot 90 0 0>}
  {shapeType=Capsule, length=5, radius=5, shapeName=dyn_hair_bk_04_capsule, center=<13.8981 0 -0.7519>, rotation=<rot 0 0 90>}
 ]
 dyn_hair_bk_MSTR [{shapeType=Capsule, length=0.505, radius=0.505, shapeName=dyn_hair_bk_MSTR_capsule, rotation=<rot 90 0 0>}]
 dyn_hair_l_00 [{shapeType=Capsule, length=2.7364, radius=5.1567, shapeName=dyn_hair_bk_00_capsule, center=<5.0867 0 8>, rotation=<rot 90 0 0>}]
 dyn_hair_l_01 [{shapeType=Capsule, length=4.2183, radius=5.7817, shapeName=dyn_hair_bk_00_capsule, center=<7.0292 0 5>, rotation=<rot 90 0 0>}]
 dyn_hair_l_02 [{shapeType=Capsule, length=2.4351, radius=7.5649, shapeName=dyn_hair_bk_00_capsule, center=<7.0292 0 5>, rotation=<rot 90 0 0>}]
 dyn_hair_l_03 [{shapeType=Capsule, length=2.4351, radius=7.5649, shapeName=dyn_hair_bk_00_capsule, center=<7.0292 0 5>, rotation=<rot 90 0 0>}]
 dyn_hair_l_04 [
  {shapeType=Capsule, length=2.4351, radius=7.5649, shapeName=dyn_hair_bk_00_capsule, center=<7.0292 0 5>, rotation=<rot 90 0 0>}
  {shapeType=Capsule, length=5, radius=5, shapeName=dyn_hair_l_04_capsule, center=<13.8981 0 -0.7519>, rotation=<rot 0 0 90>}
 ]
 dyn_hair_l_MSTR [{shapeType=Capsule, length=0.505, radius=0.505, shapeName=dyn_hair_l_MSTR_capsule, rotation=<rot 90 0 0>}]
 dyn_hair_r_00 [{shapeType=Capsule, length=2.7364, radius=5.1567, shapeName=dyn_hair_bk_00_capsule, center=<5.0867 0 8>, rotation=<rot 90 0 0>}]
 dyn_hair_r_01 [{shapeType=Capsule, length=2.4351, radius=7.5649, shapeName=dyn_hair_bk_00_capsule, center=<7.0292 0 5>, rotation=<rot 90 0 0>}]
 dyn_hair_r_02 [{shapeType=Capsule, length=2.4351, radius=7.5649, shapeName=dyn_hair_bk_00_capsule, center=<7.0292 0 5>, rotation=<rot 90 0 0>}]
 dyn_hair_r_03 [{shapeType=Capsule, length=2.4351, radius=7.5649, shapeName=dyn_hair_bk_00_capsule, center=<7.0292 0 5>, rotation=<rot 90 2.1078 2.1078>}]
 dyn_hair_r_04 [
  {shapeType=Capsule, length=2.4351, radius=7.5649, shapeName=dyn_hair_bk_00_capsule, center=<7.0292 0 5>, rotation=<rot 90 0 0>}
  {shapeType=Capsule, length=5, radius=5, shapeName=dyn_hair_r_04_capsule, center=<13.8981 0 -0.7519>, rotation=<rot 0 0 90>}
 ]
 dyn_hair_r_MSTR [{shapeType=Capsule, length=0.505, radius=0.505, shapeName=dyn_hair_r_MSTR_capsule, rotation=<rot 90 0 0>}]
 head [{shapeType=Capsule, length=3.5019, radius=8.8539, shapeName=head_capsule, rotation=<rot 90 0 -18.8391>}]
 neck_01 [{shapeType=Capsule, length=5, radius=9.4924, shapeName=neck_01_capsule, rotation=<rot 90 0 0>}]
 neck_02 [{shapeType=Capsule, length=5, radius=7, shapeName=neck_02_capsule, rotation=<rot 90 0 0>}]
 spine_03 [{shapeType=Capsule, length=12.4203, radius=9.327, shapeName=spine_03_capsule}]
 spine_05 [{shapeType=Capsule, length=44.6573, radius=25.1184, shapeName=spine_05_capsule, center=<-4.2739 16.5459 0>}]
 upperarm_l [{shapeType=Capsule, length=16.6968, radius=5.6646, shapeName=upperarm_l_capsule, center=<13.0242 4.0463 0.0001>, rotation=<rot 0 -90 90>}]
 upperarm_r [{
   shapeType=Capsule
   length=15.6968
   radius=6.6646
   shapeName=upperarm_r_capsule
   center=<-13.1638 -1.4774 0.0001>
   rotation=<rot -0.0002 -89.9998 90>
  }]
}
constraints [
 {
  bone1Name=head
  bone2Name=neck_02
  swing1LimitDegrees=45
  swing1Motion=Free
  swing2LimitDegrees=45
  swing2Motion=Free
  twistLimitDegrees=45
  twistMotion=Free
 }
 {
  bone1Name=neck_02
  bone2Name=neck_01
  swing1LimitDegrees=45
  swing1Motion=Free
  swing2LimitDegrees=45
  swing2Motion=Free
  twistLimitDegrees=45
  twistMotion=Free
 }
 {
  bone1Name=neck_01
  bone2Name=spine_05
  swing1LimitDegrees=45
  swing1Motion=Free
  swing2LimitDegrees=45
  swing2Motion=Free
  twistLimitDegrees=45
  twistMotion=Free
 }
 {
  bone1Name=spine_05
  bone2Name=spine_03
  swing1LimitDegrees=45
  swing1Motion=Free
  swing2LimitDegrees=45
  swing2Motion=Free
  twistLimitDegrees=45
  twistMotion=Free
 }
 {
  bone1Name=dyn_hairBangs_A_MSTR
  bone2Name=head
  swing1LimitDegrees=45
  swing1Motion=Free
  swing2LimitDegrees=45
  swing2Motion=Free
  twistLimitDegrees=45
  twistMotion=Free
 }
 {
  bone1Name=dyn_hairBangs_B_MSTR
  bone2Name=head
  swing1LimitDegrees=45
  swing1Motion=Free
  swing2LimitDegrees=45
  swing2Motion=Free
  twistLimitDegrees=45
  twistMotion=Free
 }
 {
  bone1Name=dyn_hairBangs_C_MSTR
  bone2Name=head
  swing1LimitDegrees=45
  swing1Motion=Free
  swing2LimitDegrees=45
  swing2Motion=Free
  twistLimitDegrees=45
  twistMotion=Free
 }
 {
  bone1Name=dyn_hairBangs_D_MSTR
  bone2Name=head
  swing1LimitDegrees=45
  swing1Motion=Free
  swing2LimitDegrees=45
  swing2Motion=Free
  twistLimitDegrees=45
  twistMotion=Free
 }
 {
  bone1Name=dyn_hairBangs_E_MSTR
  bone2Name=head
  swing1LimitDegrees=45
  swing1Motion=Free
  swing2LimitDegrees=45
  swing2Motion=Free
  twistLimitDegrees=45
  twistMotion=Free
 }
 {
  bone1Name=dyn_hairBangs_F_MSTR
  bone2Name=head
  swing1LimitDegrees=45
  swing1Motion=Free
  swing2LimitDegrees=45
  swing2Motion=Free
  twistLimitDegrees=45
  twistMotion=Free
 }
 {
  bone1Name=dyn_hairBangs_G_MSTR
  bone2Name=head
  swing1LimitDegrees=45
  swing1Motion=Free
  swing2LimitDegrees=45
  swing2Motion=Free
  twistLimitDegrees=45
  twistMotion=Free
 }
 {
  bone1Name=dyn_hairBangs_H_MSTR
  bone2Name=head
  swing1LimitDegrees=45
  swing1Motion=Free
  swing2LimitDegrees=45
  swing2Motion=Free
  twistLimitDegrees=45
  twistMotion=Free
 }
 {
  bone1Name=dyn_hair_bk_MSTR
  bone2Name=head
  swing1LimitDegrees=45
  swing1Motion=Free
  swing2LimitDegrees=45
  swing2Motion=Free
  twistLimitDegrees=45
  twistMotion=Free
 }
 {
  bone1Name=dyn_hair_l_MSTR
  bone2Name=head
  swing1LimitDegrees=45
  swing1Motion=Free
  swing2LimitDegrees=45
  swing2Motion=Free
  twistLimitDegrees=45
  twistMotion=Free
 }
 {
  bone1Name=dyn_hair_r_MSTR
  bone2Name=head
  swing1LimitDegrees=45
  swing1Motion=Free
  swing2LimitDegrees=45
  swing2Motion=Free
  twistLimitDegrees=45
  twistMotion=Free
 }
 {bone1Name=dyn_hair_bk_00, bone2Name=dyn_hair_bk_MSTR, swing1LimitDegrees=45, swing2LimitDegrees=45, twistLimitDegrees=5}
 {bone1Name=dyn_hair_bk_01, bone2Name=dyn_hair_bk_00, swing1LimitDegrees=5, swing2LimitDegrees=45, twistLimitDegrees=5}
 {bone1Name=dyn_hair_bk_02, bone2Name=dyn_hair_bk_01, swing1LimitDegrees=15, swing2LimitDegrees=45, twistLimitDegrees=5}
 {bone1Name=dyn_hair_bk_03, bone2Name=dyn_hair_bk_02, swing1LimitDegrees=45, swing2LimitDegrees=45, twistLimitDegrees=5}
 {bone1Name=dyn_hair_bk_04, bone2Name=dyn_hair_bk_03, swing1LimitDegrees=45, swing2LimitDegrees=45, twistLimitDegrees=5}
 {bone1Name=dyn_hair_l_00, bone2Name=dyn_hair_l_MSTR, swing1LimitDegrees=45, swing2LimitDegrees=45, twistLimitDegrees=5}
 {bone1Name=dyn_hair_l_01, bone2Name=dyn_hair_l_00, swing1LimitDegrees=5, swing2LimitDegrees=45, twistLimitDegrees=5}
 {bone1Name=dyn_hair_l_02, bone2Name=dyn_hair_l_01, swing1LimitDegrees=15, swing2LimitDegrees=45, twistLimitDegrees=5}
 {bone1Name=dyn_hair_l_03, bone2Name=dyn_hair_l_02, swing1LimitDegrees=45, swing2LimitDegrees=45, twistLimitDegrees=5}
 {bone1Name=dyn_hair_l_04, bone2Name=dyn_hair_l_03, swing1LimitDegrees=45, swing2LimitDegrees=45, twistLimitDegrees=5}
 {bone1Name=dyn_hair_r_00, bone2Name=dyn_hair_r_MSTR, swing1LimitDegrees=45, swing2LimitDegrees=45, twistLimitDegrees=5}
 {bone1Name=dyn_hair_r_01, bone2Name=dyn_hair_r_00, swing1LimitDegrees=5, swing2LimitDegrees=45, twistLimitDegrees=5}
 {bone1Name=dyn_hair_r_02, bone2Name=dyn_hair_r_01, swing1LimitDegrees=15, swing2LimitDegrees=45, twistLimitDegrees=5}
 {bone1Name=dyn_hair_r_03, bone2Name=dyn_hair_r_02, swing1LimitDegrees=45, swing2LimitDegrees=45, twistLimitDegrees=5}
 {bone1Name=dyn_hair_r_04, bone2Name=dyn_hair_r_03, swing1LimitDegrees=45, swing2LimitDegrees=45, twistLimitDegrees=5}
 {
  bone1Name=clavicle_r
  bone2Name=spine_05
  swing1LimitDegrees=45
  swing1Motion=Free
  swing2LimitDegrees=45
  swing2Motion=Free
  twistLimitDegrees=45
  twistMotion=Free
 }
 {
  bone1Name=upperarm_r
  bone2Name=clavicle_r
  swing1LimitDegrees=45
  swing1Motion=Free
  swing2LimitDegrees=45
  swing2Motion=Free
  twistLimitDegrees=45
  twistMotion=Free
 }
 {
  bone1Name=clavicle_l
  bone2Name=spine_05
  swing1LimitDegrees=45
  swing1Motion=Free
  swing2LimitDegrees=45
  swing2Motion=Free
  twistLimitDegrees=45
  twistMotion=Free
 }
 {
  bone1Name=upperarm_l
  bone2Name=clavicle_l
  swing1LimitDegrees=45
  swing1Motion=Free
  swing2LimitDegrees=45
  swing2Motion=Free
  twistLimitDegrees=45
  twistMotion=Free
 }
]
