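# Animation Blueprint for the Allure LiftingAura face accessory, driving secondary motion of the character's long hair and bangs.
# It copies the parent pose, layers upper arms from another mesh copy, then constrains hair to head, neck and spine before a Control Rig.
# Jump and skydive-float states add hair pitch around a RigidBody simulation relative to spine_05.
# Per-bone AnimDynamics with spherical limits moves the bangs, followed by repeated hair constraints.
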AnimBlueprint /BRCosmetics/Characters/Player/Female/Medium/Bodies/F_MED_LiftingAura_Allure/Meshes/Parts/F_MED_LiftingAura_Allure_FaceAcc_AnimBP
target_skeleton=&/F_MED_LiftingAura_Allure_FaceAcc_Skeleton
graphs [
 AnimGraph {nodes [
   n1 = Get bIsTacticalSprinting()
   n2 = "Transform (Modify) Bone - Bone: dyn_hair_bk_00"!(ComponentPose, bAlphaBoolEnabled:bool=@n1, Translation="0,0,0", Rotation="10,0,0", Scale="1,1,1") -> (Pose) {
    alphaInputType=Bool
    rotationMode=BMM_Additive
    rotationSpace=BCS_BoneSpace
    alphaBoolBlend {blendInTime=0.1, blendOption=Cubic, blendOutTime=0.1}
    boneToModify {boneName=dyn_hair_bk_00}
    rotation {pitch=10}
   }
   n3 = "Transform (Modify) Bone - Bone: dyn_hair_l_00"!(ComponentPose=@n2, bAlphaBoolEnabled:bool=@n1, Translation="0,0,0", Rotation="10,0,0", Scale="1,1,1") -> (Pose) {
    alphaInputType=Bool
    rotationMode=BMM_Additive
    rotationSpace=BCS_BoneSpace
    alphaBoolBlend {blendInTime=0.1, blendOption=Cubic, blendOutTime=0.1}
    boneToModify {boneName=dyn_hair_l_00}
    rotation {pitch=10}
   }
   "Transform (Modify) Bone - Bone: dyn_hair_r_00"!(ComponentPose=@n3, bAlphaBoolEnabled:bool=@n1, Translation="0,0,0", Rotation="10,0,0", Scale="1,1,1") -> (Pose) {
    alphaInputType=Bool
    rotationMode=BMM_Additive
    rotationSpace=BCS_BoneSpace
    alphaBoolBlend {blendInTime=0.1, blendOption=Cubic, blendOutTime=0.1}
    boneToModify {boneName=dyn_hair_r_00}
    rotation {pitch=10}
   }
   n4 = Copy Pose From Mesh(SourceMeshComponent=Get Skel_Mesh()) {bUseAttachedParent=true}
   n5 = Copy Pose From Mesh() {bUseAttachedParent=true}
   n6 = "Constraint - dyn_hair_bk_00 (PP)"!(
    ComponentPose=Local To Component(LocalPose=Layered blend per bone(BasePose=@n5, BlendPoses_0=@n4) {layerSetup [{branchFilters [{boneName=upperarm_l}, {boneName=upperarm_r}]}]})
    Alpha=1
    ConstraintWeights_0=0.6
    ConstraintWeights_1:float=0.4
   ) -> (Pose) {
    alphaBoolBlend {blendOption=Linear}
    boneToModify {boneName=dyn_hair_bk_00}
    constraintSetup [{targetBone {boneName=head}}, {targetBone {boneName=spine_05}}]
    constraintWeights [1,1]
   }
   n7 = "Constraint - dyn_hair_bk_01 (PP)"!(ComponentPose=@n6, Alpha=1, ConstraintWeights_0=0.6, ConstraintWeights_1:float=0.4) -> (Pose) {
    alphaBoolBlend {blendOption=Linear}
    boneToModify {boneName=dyn_hair_bk_01}
    constraintSetup [{targetBone {boneName=head}}, {targetBone {boneName=spine_05}}]
    constraintWeights [1,1]
   }
   n8 = "Constraint - dyn_hair_bk_02 (PP)"!(ComponentPose=@n7, Alpha=1, ConstraintWeights_0=0.4, ConstraintWeights_1:float=0.6) -> (Pose) {
    alphaBoolBlend {blendOption=Linear}
    boneToModify {boneName=dyn_hair_bk_02}
    constraintSetup [{targetBone {boneName=head}}, {targetBone {boneName=spine_05}}]
    constraintWeights [1,1]
   }
   n9 = "Constraint - dyn_hair_bk_03 (PP)"!(ComponentPose=@n8, Alpha=1, ConstraintWeights_0=0.6, ConstraintWeights_1:float=0.4) -> (Pose) {
    alphaBoolBlend {blendOption=Linear}
    boneToModify {boneName=dyn_hair_bk_03}
    constraintSetup [{targetBone {boneName=neck_01}}, {targetBone {boneName=spine_05}}]
    constraintWeights [1,1]
   }
   n10 = "Constraint - dyn_hair_bk_04 (PP)"!(ComponentPose=@n9, Alpha=1, ConstraintWeights_0=1, ConstraintWeights_1:float=0) -> (Pose) {
    alphaBoolBlend {blendOption=Linear}
    boneToModify {boneName=dyn_hair_bk_04}
    constraintSetup [{targetBone {boneName=spine_05}}, {targetBone {boneName=spine_05}}]
    constraintWeights [1,1]
   }
   n11 = "Constraint - dyn_hair_l_00 (PP)"!(ComponentPose=@n10, Alpha=1, ConstraintWeights_0=0.3, ConstraintWeights_1:float=0.7) -> (Pose) {
    alphaBoolBlend {blendOption=Linear}
    boneToModify {boneName=dyn_hair_l_00}
    constraintSetup [{targetBone {boneName=head}}, {targetBone {boneName=dyn_hair_bk_00}}]
    constraintWeights [1,1]
   }
   n12 = "Constraint - dyn_hair_l_01 (PP)"!(ComponentPose=@n11, Alpha=1, ConstraintWeights_0=0.3, ConstraintWeights_1:float=0.7) -> (Pose) {
    alphaBoolBlend {blendOption=Linear}
    boneToModify {boneName=dyn_hair_l_01}
    constraintSetup [{targetBone {boneName=head}}, {targetBone {boneName=dyn_hair_bk_01}}]
    constraintWeights [1,1]
   }
   n13 = "Constraint - dyn_hair_l_02 (PP)"!(ComponentPose=@n12, Alpha=1, ConstraintWeights_0=0.3, ConstraintWeights_1:float=0.7) -> (Pose) {
    alphaBoolBlend {blendOption=Linear}
    boneToModify {boneName=dyn_hair_l_02}
    constraintSetup [{targetBone {boneName=head}}, {targetBone {boneName=dyn_hair_bk_02}}]
    constraintWeights [1,1]
   }
   n14 = "Constraint - dyn_hair_l_03 (PP)"!(ComponentPose=@n13, Alpha=1, ConstraintWeights_0=0.5, ConstraintWeights_1:float=0.5) -> (Pose) {
    alphaBoolBlend {blendOption=Linear}
    boneToModify {boneName=dyn_hair_l_03}
    constraintSetup [{targetBone {boneName=spine_05}}, {targetBone {boneName=dyn_hair_bk_03}}]
    constraintWeights [1,1]
   }
   n15 = "Constraint - dyn_hair_l_04 (PP)"!(ComponentPose=@n14, Alpha=1, ConstraintWeights_0=0.8, ConstraintWeights_1:float=0.2) -> (Pose) {
    alphaBoolBlend {blendOption=Linear}
    boneToModify {boneName=dyn_hair_l_04}
    constraintSetup [{targetBone {boneName=spine_05}}, {targetBone {boneName=dyn_hair_bk_04}}]
    constraintWeights [1,1]
   }
   n16 = "Constraint - dyn_hair_r_00 (PP)"!(ComponentPose=@n15, Alpha=1, ConstraintWeights_0=0.3, ConstraintWeights_1:float=0.7) -> (Pose) {
    alphaBoolBlend {blendOption=Linear}
    boneToModify {boneName=dyn_hair_r_00}
    constraintSetup [{targetBone {boneName=head}}, {targetBone {boneName=dyn_hair_bk_00}}]
    constraintWeights [1,1]
   }
   n17 = "Constraint - dyn_hair_r_01 (PP)"!(ComponentPose=@n16, Alpha=1, ConstraintWeights_0=0.3, ConstraintWeights_1:float=0.7) -> (Pose) {
    alphaBoolBlend {blendOption=Linear}
    boneToModify {boneName=dyn_hair_r_01}
    constraintSetup [{targetBone {boneName=head}}, {targetBone {boneName=dyn_hair_bk_01}}]
    constraintWeights [1,1]
   }
   n18 = "Constraint - dyn_hair_r_02 (PP)"!(ComponentPose=@n17, Alpha=1, ConstraintWeights_0=0.3, ConstraintWeights_1:float=0.7) -> (Pose) {
    alphaBoolBlend {blendOption=Linear}
    boneToModify {boneName=dyn_hair_r_02}
    constraintSetup [{targetBone {boneName=head}}, {targetBone {boneName=dyn_hair_bk_02}}]
    constraintWeights [1,1]
   }
   n19 = "Constraint - dyn_hair_r_03 (PP)"!(ComponentPose=@n18, Alpha=1, ConstraintWeights_0=0.5, ConstraintWeights_1:float=0.5) -> (Pose) {
    alphaBoolBlend {blendOption=Linear}
    boneToModify {boneName=dyn_hair_r_03}
    constraintSetup [{targetBone {boneName=spine_05}}, {targetBone {boneName=dyn_hair_bk_03}}]
    constraintWeights [1,1]
   }
   n20 = "Constraint - dyn_hair_r_04 (PP)"!(ComponentPose=@n19, Alpha=1, ConstraintWeights_0=0.8, ConstraintWeights_1:float=0.2) -> (Pose) {
    alphaBoolBlend {blendOption=Linear}
    boneToModify {boneName=dyn_hair_r_04}
    constraintSetup [{targetBone {boneName=spine_05}}, {targetBone {boneName=dyn_hair_bk_04}}]
    constraintWeights [1,1]
   }
   n21 = Control Rig(Source=Component To Local(ComponentPose=@n20)) {
    alphaBoolBlend {blendOption=Linear}
    controlRigAssetReference {blueprintRigClass=&/F_MED_LiftingAura_Allure_FaceAcc_CtrlRig.F_MED_LiftingAura_Allure_FaceAcc_CtrlRig_C}
   }
   n22 = Get Jump() -> (Jump:float)
   n23 = "Transform (Modify) Bone - Bone: dyn_hair_bk_00"!(ComponentPose=Local To Component(LocalPose=@n21), Alpha:float=@n22, Translation="0,0,0", Rotation="20,0,0", Scale="1,1,1") -> (Pose) {
    rotationMode=BMM_Additive
    rotationSpace=BCS_BoneSpace
    alphaBoolBlend {blendOption=Linear}
    alphaScaleBiasClamp {bInterpResult=true, interpSpeedIncreasing=4}
    boneToModify {boneName=dyn_hair_bk_00}
    rotation {pitch=20}
   }
   n24 = "Transform (Modify) Bone - Bone: dyn_hair_l_00"!(ComponentPose=@n23, Alpha:float=@n22, Translation="0,0,0", Rotation="20,0,0", Scale="1,1,1") -> (Pose) {
    rotationMode=BMM_Additive
    rotationSpace=BCS_BoneSpace
    alphaBoolBlend {blendOption=Linear}
    alphaScaleBiasClamp {bInterpResult=true, interpSpeedIncreasing=4}
    boneToModify {boneName=dyn_hair_l_00}
    rotation {pitch=20}
   }
   n25 = "Transform (Modify) Bone - Bone: dyn_hair_r_00"!(ComponentPose=@n24, Alpha:float=@n22, Translation="0,0,0", Rotation="20,0,0", Scale="1,1,1") -> (Pose) {
    rotationMode=BMM_Additive
    rotationSpace=BCS_BoneSpace
    alphaBoolBlend {blendOption=Linear}
    alphaScaleBiasClamp {bInterpResult=true, interpSpeedIncreasing=4}
    boneToModify {boneName=dyn_hair_r_00}
    rotation {pitch=20}
   }
   n26 = Get bIsSkydiveFloating()
   n27 = "Transform (Modify) Bone - Bone: dyn_hair_bk_00"!(ComponentPose=@n25, bAlphaBoolEnabled:bool=@n26, Translation="0,0,0", Rotation="15,0,0", Scale="1,1,1") -> (Pose) {
    alphaInputType=Bool
    rotationMode=BMM_Additive
    rotationSpace=BCS_BoneSpace
    alphaBoolBlend {blendInTime=0.1, blendOption=Cubic, blendOutTime=0.1}
    boneToModify {boneName=dyn_hair_bk_00}
    rotation {pitch=15}
   }
   n28 = "Transform (Modify) Bone - Bone: dyn_hair_l_00"!(ComponentPose=@n27, bAlphaBoolEnabled:bool=@n26, Translation="0,0,0", Rotation="15,0,0", Scale="1,1,1") -> (Pose) {
    alphaInputType=Bool
    rotationMode=BMM_Additive
    rotationSpace=BCS_BoneSpace
    alphaBoolBlend {blendInTime=0.1, blendOption=Cubic, blendOutTime=0.1}
    boneToModify {boneName=dyn_hair_l_00}
    rotation {pitch=15}
   }
   n29 = "Transform (Modify) Bone - Bone: dyn_hair_r_00"!(ComponentPose=@n28, bAlphaBoolEnabled:bool=@n26, Translation="0,0,0", Rotation="15,0,0", Scale="1,1,1") -> (Pose) {
    alphaInputType=Bool
    rotationMode=BMM_Additive
    rotationSpace=BCS_BoneSpace
    alphaBoolBlend {blendInTime=0.1, blendOption=Cubic, blendOutTime=0.1}
    boneToModify {boneName=dyn_hair_r_00}
    rotation {pitch=15}
   }
   n30 = Get backbling rban() -> (backbling rban:Vector)
   n31 = RigidBody!(
    ComponentPose=@n29
    Alpha
    OverrideWorldGravity:Vector="0,0,0"
    ExternalForce
    ComponentLinearAccScale:Vector=@n30
    ComponentLinearVelScale:Vector=@n30
   ) -> (Pose) {
    bClampLinearTranslationLimitToRefPose=true
    bOverrideWorldGravity=true
    overridePhysicsAsset=&/F_MED_LiftingAura_Allure_FaceAcc_Physics
    simulationSpace=BaseBoneSpace
    alphaBoolBlend {blendOption=Linear}
    alphaScaleBiasClamp {bInterpResult=true, interpSpeedIncreasing=2}
    baseBoneRef {boneName=spine_05}
    componentLinearAccScale=<0.4 0.4 0.5>
    componentLinearVelScale=<0.4 0.4 0.5>
    simSpaceSettings {
     maxAngularAcceleration=1000
     maxAngularVelocity=3
     maxLinearAcceleration=700
     maxLinearVelocity=500
     velocityScaleZ=0
     externalLinearDragV=<0.2 0.2>
    }
   }
   n32 = "Transform (Modify) Bone - Bone: dyn_hair_bk_03"(ComponentPose=@n31, Alpha=@n22, Translation="0,0,0", Rotation="35,0,0", Scale="1,1,1") {
    rotationMode=BMM_Additive
    rotationSpace=BCS_BoneSpace
    alphaBoolBlend {blendOption=Linear}
    alphaScaleBiasClamp {bInterpResult=true, interpSpeedIncreasing=0.4}
    boneToModify {boneName=dyn_hair_bk_03}
    rotation {pitch=35}
   }
   n33 = "Transform (Modify) Bone - Bone: dyn_hair_l_03"(ComponentPose=@n32, Alpha=@n22, Translation="0,0,0", Rotation="35,0,0", Scale="1,1,1") {
    rotationMode=BMM_Additive
    rotationSpace=BCS_BoneSpace
    alphaBoolBlend {blendOption=Linear}
    alphaScaleBiasClamp {bInterpResult=true, interpSpeedIncreasing=0.4}
    boneToModify {boneName=dyn_hair_l_03}
    rotation {pitch=35}
   }
   n34 = "Transform (Modify) Bone - Bone: dyn_hair_r_03"(ComponentPose=@n33, Alpha=@n22, Translation="0,0,0", Rotation="35,0,0", Scale="1,1,1") {
    rotationMode=BMM_Additive
    rotationSpace=BCS_BoneSpace
    alphaBoolBlend {blendOption=Linear}
    alphaScaleBiasClamp {bInterpResult=true, interpSpeedIncreasing=0.4}
    boneToModify {boneName=dyn_hair_r_03}
    rotation {pitch=35}
   }
   n35 = "Transform (Modify) Bone - Bone: dyn_hair_bk_04"(ComponentPose=@n34, Alpha=@n22, Translation="0,0,0", Rotation="15,0,0", Scale="1,1,1") {
    rotationMode=BMM_Additive
    rotationSpace=BCS_BoneSpace
    alphaBoolBlend {blendOption=Linear}
    alphaScaleBiasClamp {bInterpResult=true, interpSpeedIncreasing=0.4}
    boneToModify {boneName=dyn_hair_bk_04}
    rotation {pitch=15}
   }
   n36 = "Transform (Modify) Bone - Bone: dyn_hair_l_04"(ComponentPose=@n35, Alpha=@n22, Translation="0,0,0", Rotation="15,0,0", Scale="1,1,1") {
    rotationMode=BMM_Additive
    rotationSpace=BCS_BoneSpace
    alphaBoolBlend {blendOption=Linear}
    alphaScaleBiasClamp {bInterpResult=true, interpSpeedIncreasing=0.4}
    boneToModify {boneName=dyn_hair_l_04}
    rotation {pitch=15}
   }
   n37 = "Transform (Modify) Bone - Bone: dyn_hair_r_04"(ComponentPose=@n36, Alpha=@n22, Translation="0,0,0", Rotation="15,0,0", Scale="1,1,1") {
    rotationMode=BMM_Additive
    rotationSpace=BCS_BoneSpace
    alphaBoolBlend {blendOption=Linear}
    alphaScaleBiasClamp {bInterpResult=true, interpSpeedIncreasing=0.4}
    boneToModify {boneName=dyn_hair_r_04}
    rotation {pitch=15}
   }
   n38 = Anim Dynamics - Bone: dyn_hairBangs_E_01(ComponentPose=@n37, Alpha=0.7, ExternalForce="0,0,0") {
    angularDampingOverride=0.9999
    angularSpringConstant=100
    bAngularSpring=true
    bGravityOverrideInSimSpace=true
    bOverrideAngularDamping=true
    bUseGravityOverride=true
    bUseSphericalLimits=true
    linearSpringConstant=10
    simulationSpace=BoneRelative
    alphaBoolBlend {blendOption=Linear}
    boundBone {boneName=dyn_hairBangs_E_01}
    componentAppliedLinearAccClamp=<3500 3500 3500>
    componentLinearAccScale=<0.2 0.2 0.2>
    componentLinearVelScale=<0.2 0.2 0.2>
    gravityOverride {z=-980}
    physicsBodyDefinitions [{
      sphereCollisionRadius=0
      boundBone {boneName=dyn_hairBangs_E_01}
      boxExtents=<5 25 25>
      constraintSetup {angularLimitsMax=<0.5 -2 0>, angularLimitsMin=<-0.5 -20 0>, angularTarget=<0 0 0>}
      localJointOffset=<3 0 0>
     }]
    relativeSpaceBone {boneName=root}
    sphericalLimits [{limitRadius=9, drivingBone {boneName=head}, sphereLocalOffset=<0 0 0>}]
   }
   n39 = Anim Dynamics - Bone: dyn_hairBangs_E_02(ComponentPose=@n38, Alpha=0.6, ExternalForce="0,0,0") {
    angularDampingOverride=0.9999
    angularSpringConstant=100
    bAngularSpring=true
    bGravityOverrideInSimSpace=true
    bOverrideAngularDamping=true
    bUseGravityOverride=true
    bUseSphericalLimits=true
    linearSpringConstant=10
    simulationSpace=BoneRelative
    alphaBoolBlend {blendOption=Linear}
    boundBone {boneName=dyn_hairBangs_E_02}
    componentAppliedLinearAccClamp=<3500 3500 3500>
    componentLinearAccScale=<0.2 0.2 0.2>
    componentLinearVelScale=<0.2 0.2 0.2>
    gravityOverride {z=-980}
    physicsBodyDefinitions [{
      sphereCollisionRadius=0
      boundBone {boneName=dyn_hairBangs_E_02}
      boxExtents=<5 25 25>
      constraintSetup {angularLimitsMax=<0.5 -5 0>, angularLimitsMin=<-0.5 -20 0>, angularTarget=<0 0 0>}
      localJointOffset=<3 0 0>
     }]
    relativeSpaceBone {boneName=root}
    sphericalLimits [{limitRadius=9, drivingBone {boneName=head}, sphereLocalOffset=<0 0 0>}]
   }
   n40 = Anim Dynamics - Bone: dyn_hairBangs_E_03(ComponentPose=@n39, Alpha=0.5, ExternalForce="0,0,0") {
    angularDampingOverride=0.9999
    angularSpringConstant=100
    bAngularSpring=true
    bGravityOverrideInSimSpace=true
    bOverrideAngularDamping=true
    bUseGravityOverride=true
    bUseSphericalLimits=true
    linearSpringConstant=10
    simulationSpace=BoneRelative
    alphaBoolBlend {blendOption=Linear}
    boundBone {boneName=dyn_hairBangs_E_03}
    componentAppliedLinearAccClamp=<3500 3500 3500>
    componentLinearAccScale=<0.2 0.2 0.2>
    componentLinearVelScale=<0.2 0.2 0.2>
    gravityOverride {z=-980}
    physicsBodyDefinitions [{
      sphereCollisionRadius=0
      boundBone {boneName=dyn_hairBangs_E_03}
      boxExtents=<5 25 25>
      constraintSetup {angularLimitsMax=<0.5 -10 0>, angularLimitsMin=<-0.5 -20 0>, angularTarget=<0 0 0>}
      localJointOffset=<3 0 0>
     }]
    relativeSpaceBone {boneName=root}
    sphericalLimits [{limitRadius=9, drivingBone {boneName=head}, sphereLocalOffset=<0 0 0>}]
   }
   n41 = Anim Dynamics - Bone: dyn_hairBangs_D_01(ComponentPose=@n40, Alpha=0.5, ExternalForce="0,0,0") {
    angularDampingOverride=0.9999
    angularSpringConstant=100
    bAngularSpring=true
    bGravityOverrideInSimSpace=true
    bOverrideAngularDamping=true
    bUseGravityOverride=true
    bUseSphericalLimits=true
    linearSpringConstant=10
    simulationSpace=BoneRelative
    alphaBoolBlend {blendOption=Linear}
    boundBone {boneName=dyn_hairBangs_D_01}
    componentAppliedLinearAccClamp=<3500 3500 3500>
    componentLinearAccScale=<0.2 0.2 0.2>
    componentLinearVelScale=<0.2 0.2 0.2>
    gravityOverride {z=-980}
    physicsBodyDefinitions [{
      sphereCollisionRadius=0
      boundBone {boneName=dyn_hairBangs_D_01}
      boxExtents=<5 25 25>
      constraintSetup {angularLimitsMax=<0.5 -5 0>, angularLimitsMin=<-0.5 -30 0>, angularTarget=<0 0 0>}
      localJointOffset=<3 0 0>
     }]
    relativeSpaceBone {boneName=root}
    sphericalLimits [{limitRadius=8, drivingBone {boneName=head}, sphereLocalOffset=<0 0 0>}]
   }
   n42 = Anim Dynamics - Bone: dyn_hairBangs_D_02(ComponentPose=@n41, Alpha=0.6, ExternalForce="0,0,0") {
    angularDampingOverride=0.9999
    angularSpringConstant=100
    bAngularSpring=true
    bGravityOverrideInSimSpace=true
    bOverrideAngularDamping=true
    bUseGravityOverride=true
    bUseSphericalLimits=true
    linearSpringConstant=10
    simulationSpace=BoneRelative
    alphaBoolBlend {blendOption=Linear}
    boundBone {boneName=dyn_hairBangs_D_02}
    componentAppliedLinearAccClamp=<3500 3500 3500>
    componentLinearAccScale=<0.2 0.2 0.2>
    componentLinearVelScale=<0.2 0.2 0.2>
    gravityOverride {z=-980}
    physicsBodyDefinitions [{
      sphereCollisionRadius=0
      boundBone {boneName=dyn_hairBangs_D_02}
      boxExtents=<5 25 25>
      constraintSetup {angularLimitsMax=<0.5 -5 0>, angularLimitsMin=<-0.5 -30 0>, angularTarget=<0 0 0>}
      localJointOffset=<3 0 0>
     }]
    relativeSpaceBone {boneName=root}
    sphericalLimits [{limitRadius=8, drivingBone {boneName=head}, sphereLocalOffset=<0 0 0>}]
   }
   n43 = Anim Dynamics - Bone: dyn_hairBangs_A_01(ComponentPose=@n42, Alpha=0.7, ExternalForce="0,0,0") {
    angularDampingOverride=0.9999
    angularSpringConstant=100
    bAngularSpring=true
    bGravityOverrideInSimSpace=true
    bOverrideAngularDamping=true
    bUseGravityOverride=true
    bUseSphericalLimits=true
    linearSpringConstant=10
    simulationSpace=BoneRelative
    alphaBoolBlend {blendOption=Linear}
    boundBone {boneName=dyn_hairBangs_A_01}
    componentAppliedLinearAccClamp=<3500 3500 3500>
    componentLinearAccScale=<0.2 0.2 0.2>
    componentLinearVelScale=<0.2 0.2 0.2>
    gravityOverride {z=-980}
    physicsBodyDefinitions [{
      sphereCollisionRadius=0
      boundBone {boneName=dyn_hairBangs_A_01}
      boxExtents=<5 25 25>
      constraintSetup {angularLimitsMax=<0.5 8 8>, angularLimitsMin=<-0.5 -20 -20>, angularTarget=<0 0 0>}
      localJointOffset=<3 0 0>
     }]
    relativeSpaceBone {boneName=root}
    sphericalLimits [{limitRadius=9, drivingBone {boneName=head}, sphereLocalOffset=<0 0 0>}]
   }
   n44 = Anim Dynamics - Bone: dyn_hairBangs_A_02(ComponentPose=@n43, Alpha=0.5, ExternalForce="0,0,0") {
    angularDampingOverride=0.9999
    angularSpringConstant=100
    bAngularSpring=true
    bGravityOverrideInSimSpace=true
    bOverrideAngularDamping=true
    bUseGravityOverride=true
    bUseSphericalLimits=true
    linearSpringConstant=10
    simulationSpace=BoneRelative
    alphaBoolBlend {blendOption=Linear}
    boundBone {boneName=dyn_hairBangs_A_02}
    componentAppliedLinearAccClamp=<3500 3500 3500>
    componentLinearAccScale=<0.2 0.2 0.2>
    componentLinearVelScale=<0.2 0.2 0.2>
    gravityOverride {z=-980}
    physicsBodyDefinitions [{
      sphereCollisionRadius=0
      boundBone {boneName=dyn_hairBangs_A_02}
      boxExtents=<5 25 25>
      constraintSetup {angularLimitsMax=<0.5 8 8>, angularLimitsMin=<-0.5 -20 -20>, angularTarget=<0 0 0>}
      localJointOffset=<3 0 0>
     }]
    relativeSpaceBone {boneName=root}
    sphericalLimits [{limitRadius=9, drivingBone {boneName=head}, sphereLocalOffset=<0 0 0>}]
   }
   n45 = Anim Dynamics - Bone: dyn_hairBangs_B_01(ComponentPose=@n44, Alpha=0.4, ExternalForce="0,0,0") {
    angularDampingOverride=0.9999
    angularSpringConstant=100
    bAngularSpring=true
    bGravityOverrideInSimSpace=true
    bOverrideAngularDamping=true
    bUseGravityOverride=true
    bUseSphericalLimits=true
    linearSpringConstant=10
    simulationSpace=BoneRelative
    alphaBoolBlend {blendOption=Linear}
    boundBone {boneName=dyn_hairBangs_B_01}
    componentAppliedLinearAccClamp=<3500 3500 3500>
    componentLinearAccScale=<0.2 0.2 0.2>
    componentLinearVelScale=<0.2 0.2 0.2>
    gravityOverride {z=-980}
    physicsBodyDefinitions [{
      sphereCollisionRadius=0
      boundBone {boneName=dyn_hairBangs_B_01}
      boxExtents=<5 25 25>
      constraintSetup {angularLimitsMax=<0.5 8 20>, angularLimitsMin=<-0.5 -30 -25>, angularTarget=<0 0 0>}
      localJointOffset=<3 0 0>
     }]
    relativeSpaceBone {boneName=root}
    sphericalLimits [{limitRadius=8, drivingBone {boneName=head}, sphereLocalOffset=<0 0 0>}]
   }
   n46 = Anim Dynamics - Bone: dyn_hairBangs_B_02(ComponentPose=@n45, Alpha=0.3, ExternalForce="0,0,0") {
    angularDampingOverride=0.9999
    angularSpringConstant=100
    bAngularSpring=true
    bGravityOverrideInSimSpace=true
    bOverrideAngularDamping=true
    bUseGravityOverride=true
    bUseSphericalLimits=true
    linearSpringConstant=10
    simulationSpace=BoneRelative
    alphaBoolBlend {blendOption=Linear}
    boundBone {boneName=dyn_hairBangs_B_02}
    componentAppliedLinearAccClamp=<3500 3500 3500>
    componentLinearAccScale=<0.2 0.2 0.2>
    componentLinearVelScale=<0.2 0.2 0.2>
    gravityOverride {z=-980}
    physicsBodyDefinitions [{
      sphereCollisionRadius=0
      boundBone {boneName=dyn_hairBangs_B_02}
      boxExtents=<5 25 25>
      constraintSetup {angularLimitsMax=<0.5 8 20>, angularLimitsMin=<-0.5 -30 -25>, angularTarget=<0 0 0>}
      localJointOffset=<3 0 0>
     }]
    relativeSpaceBone {boneName=root}
    sphericalLimits [{limitRadius=8, drivingBone {boneName=head}, sphereLocalOffset=<0 0 0>}]
   }
   n47 = Anim Dynamics - Bone: dyn_hairBangs_B_03(ComponentPose=@n46, Alpha=0.2, ExternalForce="0,0,0") {
    angularDampingOverride=0.9999
    angularSpringConstant=100
    bAngularSpring=true
    bGravityOverrideInSimSpace=true
    bOverrideAngularDamping=true
    bUseGravityOverride=true
    bUseSphericalLimits=true
    linearSpringConstant=10
    simulationSpace=BoneRelative
    alphaBoolBlend {blendOption=Linear}
    boundBone {boneName=dyn_hairBangs_B_03}
    componentAppliedLinearAccClamp=<3500 3500 3500>
    componentLinearAccScale=<0.2 0.2 0.2>
    componentLinearVelScale=<0.2 0.2 0.2>
    gravityOverride {z=-980}
    physicsBodyDefinitions [{
      sphereCollisionRadius=0
      boundBone {boneName=dyn_hairBangs_B_03}
      boxExtents=<5 25 25>
      constraintSetup {angularLimitsMax=<0.5 8 20>, angularLimitsMin=<-0.5 -30 -25>, angularTarget=<0 0 0>}
      localJointOffset=<3 0 0>
     }]
    relativeSpaceBone {boneName=root}
    sphericalLimits [{limitRadius=8, drivingBone {boneName=head}, sphereLocalOffset=<0 0 0>}]
   }
   n48 = Anim Dynamics - Bone: dyn_hairBangs_C_01(ComponentPose=@n47, Alpha=0.4, ExternalForce="0,0,0") {
    angularDampingOverride=0.9999
    angularSpringConstant=100
    bAngularSpring=true
    bGravityOverrideInSimSpace=true
    bOverrideAngularDamping=true
    bUseGravityOverride=true
    bUseSphericalLimits=true
    linearSpringConstant=10
    simulationSpace=BoneRelative
    alphaBoolBlend {blendOption=Linear}
    boundBone {boneName=dyn_hairBangs_C_01}
    componentAppliedLinearAccClamp=<3500 3500 3500>
    componentLinearAccScale=<0.2 0.2 0.2>
    componentLinearVelScale=<0.2 0.2 0.2>
    gravityOverride {z=-980}
    physicsBodyDefinitions [{
      sphereCollisionRadius=0
      boundBone {boneName=dyn_hairBangs_C_01}
      boxExtents=<5 25 25>
      constraintSetup {angularLimitsMax=<0.5 8 20>, angularLimitsMin=<-0.5 -30 -25>, angularTarget=<0 0 0>}
      localJointOffset=<3 0 0>
     }]
    relativeSpaceBone {boneName=root}
    sphericalLimits [{limitRadius=8, drivingBone {boneName=head}, sphereLocalOffset=<0 0 0>}]
   }
   n49 = Anim Dynamics - Bone: dyn_hairBangs_C_02(ComponentPose=@n48, Alpha=0.3, ExternalForce="0,0,0") {
    angularDampingOverride=0.9999
    angularSpringConstant=100
    bAngularSpring=true
    bGravityOverrideInSimSpace=true
    bOverrideAngularDamping=true
    bUseGravityOverride=true
    bUseSphericalLimits=true
    linearSpringConstant=10
    simulationSpace=BoneRelative
    alphaBoolBlend {blendOption=Linear}
    boundBone {boneName=dyn_hairBangs_C_02}
    componentAppliedLinearAccClamp=<3500 3500 3500>
    componentLinearAccScale=<0.2 0.2 0.2>
    componentLinearVelScale=<0.2 0.2 0.2>
    gravityOverride {z=-980}
    physicsBodyDefinitions [{
      sphereCollisionRadius=0
      boundBone {boneName=dyn_hairBangs_C_02}
      boxExtents=<5 25 25>
      constraintSetup {angularLimitsMax=<0.5 8 20>, angularLimitsMin=<-0.5 -30 -25>, angularTarget=<0 0 0>}
      localJointOffset=<3 0 0>
     }]
    relativeSpaceBone {boneName=root}
    sphericalLimits [{limitRadius=8, drivingBone {boneName=head}, sphereLocalOffset=<0 0 0>}]
   }
   n50 = Anim Dynamics - Bone: dyn_hairBangs_F_01(ComponentPose=@n49, Alpha=0.5, ExternalForce="0,0,0") {
    angularDampingOverride=0.9999
    angularSpringConstant=100
    bAngularSpring=true
    bGravityOverrideInSimSpace=true
    bOverrideAngularDamping=true
    bUseGravityOverride=true
    bUseSphericalLimits=true
    linearSpringConstant=10
    simulationSpace=BoneRelative
    alphaBoolBlend {blendOption=Linear}
    boundBone {boneName=dyn_hairBangs_F_01}
    componentAppliedLinearAccClamp=<3500 3500 3500>
    componentLinearAccScale=<0.2 0.2 0.2>
    componentLinearVelScale=<0.2 0.2 0.2>
    gravityOverride {z=-980}
    physicsBodyDefinitions [{
      sphereCollisionRadius=0
      boundBone {boneName=dyn_hairBangs_F_01}
      boxExtents=<5 25 25>
      constraintSetup {angularLimitsMax=<0.5 1 3>, angularLimitsMin=<-0.5 -20 -20>, angularTarget=<0 0 0>}
      localJointOffset=<3 0 0>
     }]
    relativeSpaceBone {boneName=root}
    sphericalLimits [{limitRadius=6, drivingBone {boneName=head}, sphereLocalOffset=<0 0 0>}]
   }
   n51 = Anim Dynamics - Bone: dyn_hairBangs_F_02(ComponentPose=@n50, Alpha=0.3, ExternalForce="0,0,0") {
    angularDampingOverride=0.9999
    angularSpringConstant=100
    bAngularSpring=true
    bGravityOverrideInSimSpace=true
    bOverrideAngularDamping=true
    bUseGravityOverride=true
    bUseSphericalLimits=true
    linearSpringConstant=10
    simulationSpace=BoneRelative
    alphaBoolBlend {blendOption=Linear}
    boundBone {boneName=dyn_hairBangs_F_02}
    componentAppliedLinearAccClamp=<3500 3500 3500>
    componentLinearAccScale=<0.2 0.2 0.2>
    componentLinearVelScale=<0.2 0.2 0.2>
    gravityOverride {z=-980}
    physicsBodyDefinitions [{
      sphereCollisionRadius=0
      boundBone {boneName=dyn_hairBangs_F_02}
      boxExtents=<5 25 25>
      constraintSetup {angularLimitsMax=<0.5 1 3>, angularLimitsMin=<-0.5 -20 -20>, angularTarget=<0 0 0>}
      localJointOffset=<3 0 0>
     }]
    relativeSpaceBone {boneName=root}
    sphericalLimits [{limitRadius=6, drivingBone {boneName=head}, sphereLocalOffset=<0 0 0>}]
   }
   n52 = Anim Dynamics - Bone: dyn_hairBangs_F_03(ComponentPose=@n51, Alpha=0.1, ExternalForce="0,0,0") {
    angularDampingOverride=0.9999
    angularSpringConstant=100
    bAngularSpring=true
    bGravityOverrideInSimSpace=true
    bOverrideAngularDamping=true
    bUseGravityOverride=true
    bUseSphericalLimits=true
    linearSpringConstant=10
    simulationSpace=BoneRelative
    alphaBoolBlend {blendOption=Linear}
    boundBone {boneName=dyn_hairBangs_F_03}
    componentAppliedLinearAccClamp=<3500 3500 3500>
    componentLinearAccScale=<0.2 0.2 0.2>
    componentLinearVelScale=<0.2 0.2 0.2>
    gravityOverride {z=-980}
    physicsBodyDefinitions [{
      sphereCollisionRadius=0
      boundBone {boneName=dyn_hairBangs_F_03}
      boxExtents=<5 25 25>
      constraintSetup {angularLimitsMax=<0.5 1 3>, angularLimitsMin=<-0.5 -20 -20>, angularTarget=<0 0 0>}
      localJointOffset=<3 0 0>
     }]
    relativeSpaceBone {boneName=root}
    sphericalLimits [{limitRadius=6, drivingBone {boneName=head}, sphereLocalOffset=<0 0 0>}]
   }
   n53 = Anim Dynamics - Bone: dyn_hairBangs_G_01(ComponentPose=@n52, Alpha=0.4, ExternalForce="0,0,0") {
    angularDampingOverride=0.99
    angularSpringConstant=100
    bAngularSpring=true
    bGravityOverrideInSimSpace=true
    bOverrideAngularDamping=true
    bUseGravityOverride=true
    bUseSphericalLimits=true
    simulationSpace=BoneRelative
    alphaBoolBlend {blendOption=Linear}
    boundBone {boneName=dyn_hairBangs_G_01}
    componentAppliedLinearAccClamp=<3000 3000 3000>
    componentLinearAccScale=<0.7 0.7 0.7>
    componentLinearVelScale=<0.7 0.7 0.7>
    gravityOverride {z=-980}
    physicsBodyDefinitions [{
      sphereCollisionRadius=0
      boundBone {boneName=dyn_hairBangs_G_01}
      boxExtents=<10 20 20>
      constraintSetup {angularLimitsMax=<1 5 15>, angularLimitsMin=<-1 -10 -15>, angularTarget=<0 0 0>}
      localJointOffset=<3 0 0>
     }]
    relativeSpaceBone {boneName=root}
   }
   n54 = Anim Dynamics - Bone: dyn_hairBangs_G_02(ComponentPose=@n53, Alpha=0.3, ExternalForce="0,0,0") {
    angularDampingOverride=0.99
    angularSpringConstant=100
    bAngularSpring=true
    bGravityOverrideInSimSpace=true
    bOverrideAngularDamping=true
    bUseGravityOverride=true
    bUseSphericalLimits=true
    simulationSpace=BoneRelative
    alphaBoolBlend {blendOption=Linear}
    boundBone {boneName=dyn_hairBangs_G_02}
    componentAppliedLinearAccClamp=<3000 3000 3000>
    componentLinearAccScale=<0.7 0.7 0.7>
    componentLinearVelScale=<0.7 0.7 0.7>
    gravityOverride {z=-980}
    physicsBodyDefinitions [{
      sphereCollisionRadius=0
      boundBone {boneName=dyn_hairBangs_G_02}
      boxExtents=<10 20 20>
      constraintSetup {angularLimitsMax=<1 5 15>, angularLimitsMin=<-1 -10 -15>, angularTarget=<0 0 0>}
      localJointOffset=<3 0 0>
     }]
    relativeSpaceBone {boneName=root}
    sphericalLimits [
     {limitRadius=6, drivingBone {boneName=neck_02}, sphereLocalOffset=<0 0 0>}
     {limitRadius=5.5, drivingBone {boneName=neck_01}, sphereLocalOffset=<0 0 0>}
    ]
   }
   n55 = Anim Dynamics - Bone: dyn_hairBangs_G_03(ComponentPose=@n54, Alpha=0.2, ExternalForce="0,0,0") {
    angularDampingOverride=0.99
    angularSpringConstant=100
    bAngularSpring=true
    bGravityOverrideInSimSpace=true
    bOverrideAngularDamping=true
    bUseGravityOverride=true
    bUseSphericalLimits=true
    simulationSpace=BoneRelative
    alphaBoolBlend {blendOption=Linear}
    boundBone {boneName=dyn_hairBangs_G_03}
    componentAppliedLinearAccClamp=<3000 3000 3000>
    componentLinearAccScale=<0.7 0.7 0.7>
    componentLinearVelScale=<0.7 0.7 0.7>
    gravityOverride {z=-980}
    physicsBodyDefinitions [{
      sphereCollisionRadius=0
      boundBone {boneName=dyn_hairBangs_G_03}
      boxExtents=<10 20 20>
      constraintSetup {angularLimitsMax=<1 5 15>, angularLimitsMin=<-1 -10 -15>, angularTarget=<0 0 0>}
      localJointOffset=<3 0 0>
     }]
    relativeSpaceBone {boneName=root}
    sphericalLimits [
     {limitRadius=5, drivingBone {boneName=neck_02}, sphereLocalOffset=<0 0 0>}
     {limitRadius=6, drivingBone {boneName=neck_01}, sphereLocalOffset=<0 0 0>}
    ]
   }
   n56 = Anim Dynamics - Bone: dyn_hairBangs_H_02(ComponentPose=@n55, Alpha=0.3, ExternalForce="0,0,0") {
    angularDampingOverride=0.99
    angularSpringConstant=100
    bAngularSpring=true
    bGravityOverrideInSimSpace=true
    bOverrideAngularDamping=true
    bUseGravityOverride=true
    bUseSphericalLimits=true
    simulationSpace=BoneRelative
    alphaBoolBlend {blendOption=Linear}
    boundBone {boneName=dyn_hairBangs_H_02}
    componentAppliedLinearAccClamp=<3000 3000 3000>
    componentLinearAccScale=<0.7 0.7 0.7>
    componentLinearVelScale=<0.7 0.7 0.7>
    gravityOverride {z=-980}
    physicsBodyDefinitions [{
      sphereCollisionRadius=0
      boundBone {boneName=dyn_hairBangs_H_02}
      boxExtents=<10 25 25>
      constraintSetup {angularLimitsMax=<1 8 15>, angularLimitsMin=<-1 -8 -15>, angularTarget=<0 0 0>}
      localJointOffset=<3 0 0>
     }]
    relativeSpaceBone {boneName=root}
   }
   n57 = Anim Dynamics - Bone: dyn_hairBangs_H_03(ComponentPose=@n56, Alpha=0.2, ExternalForce="0,0,0") {
    angularDampingOverride=0.99
    angularSpringConstant=100
    bAngularSpring=true
    bGravityOverrideInSimSpace=true
    bOverrideAngularDamping=true
    bUseGravityOverride=true
    bUseSphericalLimits=true
    simulationSpace=BoneRelative
    alphaBoolBlend {blendOption=Linear}
    boundBone {boneName=dyn_hairBangs_H_03}
    componentAppliedLinearAccClamp=<3000 3000 3000>
    componentLinearAccScale=<0.7 0.7 0.7>
    componentLinearVelScale=<0.7 0.7 0.7>
    gravityOverride {z=-980}
    physicsBodyDefinitions [{
      sphereCollisionRadius=0
      boundBone {boneName=dyn_hairBangs_H_03}
      boxExtents=<10 25 25>
      constraintSetup {angularLimitsMax=<1 8 15>, angularLimitsMin=<-1 -8 -15>, angularTarget=<0 0 0>}
      localJointOffset=<3 0 0>
     }]
    relativeSpaceBone {boneName=root}
    sphericalLimits [
     {limitRadius=6, drivingBone {boneName=neck_02}, sphereLocalOffset=<0 0 0>}
     {limitRadius=5.5, drivingBone {boneName=neck_01}, sphereLocalOffset=<0 0 0>}
    ]
   }
   n58 = Anim Dynamics - Bone: dyn_hairBangs_H_04(ComponentPose=@n57, Alpha=0.1, ExternalForce="0,0,0") {
    angularDampingOverride=0.99
    angularSpringConstant=100
    bAngularSpring=true
    bGravityOverrideInSimSpace=true
    bOverrideAngularDamping=true
    bUseGravityOverride=true
    bUseSphericalLimits=true
    simulationSpace=BoneRelative
    alphaBoolBlend {blendOption=Linear}
    boundBone {boneName=dyn_hairBangs_H_04}
    componentAppliedLinearAccClamp=<3000 3000 3000>
    componentLinearAccScale=<0.7 0.7 0.7>
    componentLinearVelScale=<0.7 0.7 0.7>
    gravityOverride {z=-980}
    physicsBodyDefinitions [{
      sphereCollisionRadius=0
      boundBone {boneName=dyn_hairBangs_H_04}
      boxExtents=<10 25 25>
      constraintSetup {angularLimitsMax=<1 8 15>, angularLimitsMin=<-1 -8 -15>, angularTarget=<0 0 0>}
      localJointOffset=<3 0 0>
     }]
    relativeSpaceBone {boneName=root}
    sphericalLimits [
     {limitRadius=5, drivingBone {boneName=neck_02}, sphereLocalOffset=<0 0 0>}
     {limitRadius=6, drivingBone {boneName=neck_01}, sphereLocalOffset=<0 0 0>}
    ]
   }
   n59 = "Constraint - dyn_hair_bk_00 (PP)"!(ComponentPose=@n58, bAlphaBoolEnabled:bool=True, ConstraintWeights_0=0.6, ConstraintWeights_1=0.4) -> (Pose) {
    alphaInputType=Bool
    alphaBoolBlend {blendOption=Linear}
    boneToModify {boneName=dyn_hair_bk_00}
    constraintSetup [{targetBone {boneName=head}}, {targetBone {boneName=spine_05}}]
    constraintWeights [1,1]
   }
   n60 = "Constraint - dyn_hair_bk_01 (PP)"!(ComponentPose=@n59, bAlphaBoolEnabled:bool=True, ConstraintWeights_0=0.6, ConstraintWeights_1=0.4) -> (Pose) {
    alphaInputType=Bool
    alphaBoolBlend {blendOption=Linear}
    boneToModify {boneName=dyn_hair_bk_01}
    constraintSetup [{targetBone {boneName=head}}, {targetBone {boneName=spine_05}}]
    constraintWeights [1,1]
   }
   n61 = "Constraint - dyn_hair_bk_02 (PP)"!(ComponentPose=@n60, bAlphaBoolEnabled:bool=True, ConstraintWeights_0=0.4, ConstraintWeights_1=0.6) -> (Pose) {
    alphaInputType=Bool
    alphaBoolBlend {blendOption=Linear}
    boneToModify {boneName=dyn_hair_bk_02}
    constraintSetup [{targetBone {boneName=head}}, {targetBone {boneName=spine_05}}]
    constraintWeights [1,1]
   }
   n62 = "Constraint - dyn_hair_bk_03 (PP)"!(ComponentPose=@n61, bAlphaBoolEnabled:bool=True, ConstraintWeights_0=0.6, ConstraintWeights_1=0.4) -> (Pose) {
    alphaInputType=Bool
    alphaBoolBlend {blendOption=Linear}
    boneToModify {boneName=dyn_hair_bk_03}
    constraintSetup [{targetBone {boneName=neck_01}}, {targetBone {boneName=spine_05}}]
    constraintWeights [1,1]
   }
   n63 = "Constraint - dyn_hair_bk_04 (PP)"!(ComponentPose=@n62, bAlphaBoolEnabled:bool=True, ConstraintWeights_0=1, ConstraintWeights_1=0) -> (Pose) {
    alphaInputType=Bool
    alphaBoolBlend {blendOption=Linear}
    boneToModify {boneName=dyn_hair_bk_04}
    constraintSetup [{targetBone {boneName=spine_05}}, {targetBone {boneName=spine_05}}]
    constraintWeights [1,1]
   }
   n64 = "Constraint - dyn_hair_l_00 (PP)"!(ComponentPose=@n63, bAlphaBoolEnabled:bool=True, ConstraintWeights_0=0.3, ConstraintWeights_1=0.7) -> (Pose) {
    alphaInputType=Bool
    alphaBoolBlend {blendOption=Linear}
    boneToModify {boneName=dyn_hair_l_00}
    constraintSetup [{targetBone {boneName=head}}, {targetBone {boneName=dyn_hair_bk_00}}]
    constraintWeights [1,1]
   }
   n65 = "Constraint - dyn_hair_l_01 (PP)"!(ComponentPose=@n64, bAlphaBoolEnabled:bool=True, ConstraintWeights_0=0.3, ConstraintWeights_1=0.7) -> (Pose) {
    alphaInputType=Bool
    alphaBoolBlend {blendOption=Linear}
    boneToModify {boneName=dyn_hair_l_01}
    constraintSetup [{targetBone {boneName=head}}, {targetBone {boneName=dyn_hair_bk_01}}]
    constraintWeights [1,1]
   }
   n66 = "Constraint - dyn_hair_l_02 (PP)"!(ComponentPose=@n65, bAlphaBoolEnabled:bool=True, ConstraintWeights_0=0.3, ConstraintWeights_1=0.7) -> (Pose) {
    alphaInputType=Bool
    alphaBoolBlend {blendOption=Linear}
    boneToModify {boneName=dyn_hair_l_02}
    constraintSetup [{targetBone {boneName=head}}, {targetBone {boneName=dyn_hair_bk_02}}]
    constraintWeights [1,1]
   }
   n67 = "Constraint - dyn_hair_l_03 (PP)"!(ComponentPose=@n66, bAlphaBoolEnabled:bool=True, ConstraintWeights_0=0.5, ConstraintWeights_1=0.5) -> (Pose) {
    alphaInputType=Bool
    alphaBoolBlend {blendOption=Linear}
    boneToModify {boneName=dyn_hair_l_03}
    constraintSetup [{targetBone {boneName=spine_05}}, {targetBone {boneName=dyn_hair_bk_03}}]
    constraintWeights [1,1]
   }
   n68 = "Constraint - dyn_hair_l_04 (PP)"!(ComponentPose=@n67, bAlphaBoolEnabled:bool=True, ConstraintWeights_0=0.8, ConstraintWeights_1=0.2) -> (Pose) {
    alphaInputType=Bool
    alphaBoolBlend {blendOption=Linear}
    boneToModify {boneName=dyn_hair_l_04}
    constraintSetup [{targetBone {boneName=spine_05}}, {targetBone {boneName=dyn_hair_bk_04}}]
    constraintWeights [1,1]
   }
   n69 = "Constraint - dyn_hair_r_00 (PP)"!(ComponentPose=@n68, bAlphaBoolEnabled:bool=True, ConstraintWeights_0=0.3, ConstraintWeights_1=0.7) -> (Pose) {
    alphaInputType=Bool
    alphaBoolBlend {blendOption=Linear}
    boneToModify {boneName=dyn_hair_r_00}
    constraintSetup [{targetBone {boneName=head}}, {targetBone {boneName=dyn_hair_bk_00}}]
    constraintWeights [1,1]
   }
   n70 = "Constraint - dyn_hair_r_01 (PP)"!(ComponentPose=@n69, bAlphaBoolEnabled:bool=True, ConstraintWeights_0=0.3, ConstraintWeights_1=0.7) -> (Pose) {
    alphaInputType=Bool
    alphaBoolBlend {blendOption=Linear}
    boneToModify {boneName=dyn_hair_r_01}
    constraintSetup [{targetBone {boneName=head}}, {targetBone {boneName=dyn_hair_bk_01}}]
    constraintWeights [1,1]
   }
   n71 = "Constraint - dyn_hair_r_02 (PP)"!(ComponentPose=@n70, bAlphaBoolEnabled:bool=True, ConstraintWeights_0=0.3, ConstraintWeights_1=0.7) -> (Pose) {
    alphaInputType=Bool
    alphaBoolBlend {blendOption=Linear}
    boneToModify {boneName=dyn_hair_r_02}
    constraintSetup [{targetBone {boneName=head}}, {targetBone {boneName=dyn_hair_bk_02}}]
    constraintWeights [1,1]
   }
   n72 = "Constraint - dyn_hair_r_03 (PP)"!(ComponentPose=@n71, bAlphaBoolEnabled:bool=True, ConstraintWeights_0=0.5, ConstraintWeights_1=0.5) -> (Pose) {
    alphaInputType=Bool
    alphaBoolBlend {blendOption=Linear}
    boneToModify {boneName=dyn_hair_r_03}
    constraintSetup [{targetBone {boneName=spine_05}}, {targetBone {boneName=dyn_hair_bk_03}}]
    constraintWeights [1,1]
   }
   n73 = "Constraint - dyn_hair_r_04 (PP)"!(ComponentPose=@n72, bAlphaBoolEnabled:bool=True, ConstraintWeights_0=0.8, ConstraintWeights_1=0.2) -> (Pose) {
    alphaInputType=Bool
    alphaBoolBlend {blendOption=Linear}
    boneToModify {boneName=dyn_hair_r_04}
    constraintSetup [{targetBone {boneName=spine_05}}, {targetBone {boneName=dyn_hair_bk_04}}]
    constraintWeights [1,1]
   }
   Output Pose(Result=Component To Local(ComponentPose=@n73)) {layerGroup=None}
  ]}
 EventGraph {nodes [
   Event BlueprintUpdateAnimation()
   > n1 = Sequence!(execute) -> (then_0, then_1, then_2:Exec, then_3:Exec, then_4:Exec, then_5:Exec)
   n2 = Sequence!(execute=@n1.then_1) -> (then_0, then_1)
   > Set Is Melee Float(Is Melee Float:float="To Float (Boolean)"(InBool=Get bIsPlayingMeleeAnim())) -> (Output_Get:float)
   n3 = Branch(execute=@n2.then_1, Condition=Get bIsCrouching())
   > Set RigidVectoronCrouching(RigidVectoronCrouching:Vector="0.2,0.2,0.35") -> (Output_Get:Vector)
   Set RigidVectoronCrouching(execute=@n3.else, RigidVectoronCrouching="0.4,0.4,0.5")
   n4 = Branch(execute=@n1.then_2)
   > n5 = GetRandomWindVector!(
    execute
    GameTime:float=GetGameTimeInSeconds()
    TimeOffset:float=0
    PerlinScale:float=0
    Frequency:Vector="7,6,8"
    Amplitude:Vector="-800,500,0"
    Offset:Vector="200,-300,0"
   ) -> (then, ReturnValue:Vector)
   > Set winds(winds:Vector=@n5) -> (Output_Get:Vector)
   Set winds(execute=@n4.else, winds="0, 0, 0")
   > Set HairRotator_01(HairRotator_01:Rotator="0, 0, 0") -> (Output_Get:Rotator)
   n6 = Get bIsSkyDiving()
   n7 = Get bIsJumping()
   n8 = Get bIsFalling()
   n9 = Branch(execute=@n1.then_3, Condition=AND Boolean(A=OR Boolean(A=@n7, B=@n8), B=NOT Boolean(A=@n6)))
   > Set Jump(Jump:float=1) -> (Output_Get:float)
   Set Jump(execute=@n9.else, Jump=0)
   n10 = CheckForBackbling()
   n11 = Branch(execute=@n1.then_4, Condition=@n10)
   > Set Jump(Jump=0)
   > Set backbling rban(backbling rban:Vector="0, 0, 0") -> (Output_Get:Vector)
   Set backbling rban(execute=@n11.else, backbling rban="0.4,0.4,0.5")
   n12 = Branch(execute=@n1.then_5, Condition=AND Boolean(A=@n10, B=Get bIsSkyDiving()))
   > Set SkydivingWithBackpack(SkydivingWithBackpack:bool=true) -> (Output_Get:bool)
   Set SkydivingWithBackpack(execute=@n12.else, SkydivingWithBackpack=false)
   n13 = Gravity Override State Function DEPRECATED(
    execute=@n1.then_0
    CharacterPartAnimInstance=Self-Reference()
    GameTimeInSeconds=GetGameTimeInSeconds()
    GravityOverrideParams=Get Gravity Override Params()
   )
   > Set Gravity Override(Gravity Override=@n13)
   Event BlueprintInitializeAnimation()
   > Set Skel_Mesh(Skel_Mesh=GetOwnerSkeletalMeshComponent())
  ]}
]
variables [
 backbling rban {type=Vector}
 Gravity Override {type=Vector}
 Gravity Override Params {type=GravityOverrideParamsStruct}
 HairRotator_01 {type=Rotator}
 Is Melee Float {type=number}
 Is Melee Float_0 {type=number}
 Jump {type=number}
 JumpingFalling {type=boolean}
 RigidVectoronCrouching {type=Vector}
 Skel_Mesh {type=SkeletalMeshComponent}
 SkydivingWithBackpack {type=boolean}
 winds {type=Vector}
]
